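# Translucent Niagara ribbon material for the Barium character's kill effect, drawing a smoky trail with a glowing energy fill.
# Panned smoke and noise textures, offset per ribbon, form opacity and a thresholded energy mask driven by dynamic EnergyFillFactor.
# That mask blends background into particle color and scales emissive.
# A swirl-texture eataway, driven by particle alpha and ribbon U, dissolves the combined opacity before depth fade.
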Material /Game/Characters/Player/Female/Medium/Bodies/F_MED_Barium/FX/M_BariumKillRibbon
graph {
 properties {bUsedWithNiagaraRibbons=true, blendMode=BLEND_Translucent, pixelDepthOffsetMode=PDOM_Legacy, usedShadingModels=Default Lit}
 nodes [
  Constant()
  n1 = ScalarParameter() {defaultValue=0.05, parameterName=EnergySmoothMin}
  n2 = MaterialFunctionCall!(
   1=Constant() {r=0.5}
   2=@n1
   3=@n1
   4
   A=Saturate(None=DynamicParameter() {defaultValue=<rgba 0 1 1 550>, paramNames [RandomPerRibbon, EnergyFillFactor, Param3, Param4]}.EnergyFillFactor)
  ) {materialFunction=/Engine/Functions/Engine_MaterialFunctions02/Lerp_Multiple_Float}
  n3 = DynamicParameter() {defaultValue=<rgba 0 1 1 550>, paramNames [RandomPerRibbon, EnergyFillFactor, Param3, Param4]}
  n4 = ComponentMask(None=TextureCoordinate()) {r=true}
  n5 = Saturate(None=Multiply(A=OneMinus(None=@n4)))
  n6 = Constant() {r=1}
  n7 = DynamicParameter() {defaultValue=<rgba 0 1 1 550>, paramNames [RandomPerRibbon, EnergyFillFactor, Param3, Param4]}
  n8 = ChannelMaskParameter(None=TextureSampleParameter2D(UVs=NamedRerouteUsage()) {parameterName=Extra Noise, texture=/Game/Weapons/WeaponSkins/Textures/T_Swirly_Noise_highContrast}.RGBA) {parameterName=NoiseChannel}
  n9 = LinearInterpolate(
   A=ScalarParameter() {defaultValue=0.4, parameterName=NoiseEnergyLow}
   B=ScalarParameter() {defaultValue=1, parameterName=NoiseEnergyHigh}
   Alpha=@n8.R
  )
  n10 = TextureSampleParameter2D(UVs=NamedRerouteUsage()) {parameterName=Opacity Texture, texture=/Game/Effects/Fort_Effects/Textures/T_InkySmokeTrail}
  n11 = ChannelMaskParameter(None=@n10.RGBA) {maskChannel=Green, parameterName=Energy Channel, defaultValue=<rgba 0 1 0 0>}
  n12 = SmoothStep(
   Min=@n2.Lerp 3 Inputs
   Max=MaterialFunctionCall(
    1=Constant() {r=1}
    2=ScalarParameter() {defaultValue=0.5, parameterName=EnergySmoothMax}
    3=Constant() {r=0.1}
    A=Saturate(None=@n3.EnergyFillFactor)
   ) {materialFunction=&Lerp_Multiple_Float}.Lerp 3 Inputs
   Value=Multiply(
    A=MaterialFunctionCall(1=@n5, 2=@n5, 3=@n6, A=@n7.EnergyFillFactor) {materialFunction=&Lerp_Multiple_Float}.Lerp 3 Inputs
    B=Multiply(A=@n9, B=@n11.R)
   )
  ) {constMax=0.51, constMin=0.5}
  n13 = ScalarParameter() {defaultValue=0.7, parameterName=NoiseOpacityLow}
  n14 = ScalarParameter() {defaultValue=1, parameterName=NoiseOpacityHigh}
  n15 = ScalarParameter() {parameterName=OpacitySmoothMin}
  n16 = ScalarParameter() {defaultValue=0.4, parameterName=OpacitySmoothMax}
  n17 = ChannelMaskParameter(None=@n10.RGBA) {maskChannel=Green, parameterName=OpacityChannel, defaultValue=<rgba 0 1 0 0>}
  n18 = Add(
   A=@n12
   B=Multiply(A=LinearInterpolate(A=@n13, B=@n14, Alpha=@n8.R), B=SmoothStep(Min=@n15, Max=@n16, Value=@n17.R) {constMax=0.51, constMin=0.5})
  )
  n19 = TextureCoordinate()
  n20 = ParticleColor()
  n21 = TextureSampleParameter2D(UVs=NamedRerouteUsage()) {parameterName=Eataway Texture, samplerType=SAMPLERTYPE_LinearGrayscale, texture=/Game/Global/Textures/Default/Flowmaps/T_StormSwirlMask}
  n22 = Add(
   A=OneMinus(None=ComponentMask(None=@n19) {r=true})
   B=LinearInterpolate(A=LinearInterpolate(Alpha=@n20.A) {constA=-1, constB=0.125}, B=LinearInterpolate(Alpha=@n20.A) {constA=-1, constB=2}, Alpha=@n21.G)
  )
  n23 = LinearInterpolate(Alpha=@n21.G) {constA=8, constB=2}
  n24 = Multiply(A=ComponentMask(None=TextureCoordinate()) {r=true}) {constB=6}
  n25 = ComponentMask(None=TextureCoordinate()) {r=true}
  n26 = Saturate(None=Multiply(A=OneMinus(None=@n25)) {constB=2})
  n27 = Multiply(A=Saturate(None=Multiply(A=@n22, B=@n23) {constB=8}), B=Multiply(A=Saturate(None=@n24), B=Power(Base=@n26)))
  n28 = ScalarParameter() {defaultValue=0.5, parameterName=Eataway Softness}
  n29 = OneMinus(None=SmoothStep(Min=@n27, Max=Add(A=@n27, B=@n28), Value=ConstantBiasScale(None=@n21.G)))
  DepthFade(
   Opacity=Multiply(A=Saturate(None=@n18), B=Saturate(None=@n29)) {constB=10}
   FadeDistance=ScalarParameter() {defaultValue=5, parameterName=DepthFadeDistance}
  ) {fadeDistanceDefault=30}
  n30 = ParticleColor()
  LinearInterpolate(A=VectorParameter() {parameterName=BackgroundColor}.RGB, B=@n30.RGB, Alpha=@n12)
  MaterialFunctionCall(Emissive=Multiply(A=Multiply(A=@n30.RGB, B=@n12), B=ScalarParameter() {defaultValue=1, parameterName=Emissive Strength})) {materialFunction=&MF_FX_EmissiveAdjustments}
  Multiply(A=MaterialFunctionCall(Number=@n2.Lerp 3 Inputs) {materialFunction=&DebugScalarValues}.Result)
  NamedRerouteDeclaration(None=MaterialFunctionCall!(
    UV=TextureCoordinate()
    Tile U=ScalarParameter() {defaultValue=0.5, group=Noise, parameterName=Opacity Tile U}
    Tile V=ScalarParameter() {defaultValue=1, group=Noise, parameterName=Opacity Tile V}
    Pan U=ScalarParameter() {defaultValue=-0.5, group=Noise, parameterName=Opacity Pan U}
    Pan V
    Offset U=DynamicParameter() {defaultValue=<rgba 0 1 1 550>, paramNames [RandomPerRibbon, EnergyFillFactor, Param3, Param4]}.RandomPerRibbon
    Offset V
    Time
   ) {materialFunction=/Game/Effects/Fort_Effects/Materials/_Functions/MF_PanTileOffset}.UV's) {name=Opacity UVs, nodeColor=<rgba 0 1 0.8359 1>}
  NamedRerouteDeclaration(None=MaterialFunctionCall(
    UV=TextureCoordinate()
    Tile U=ScalarParameter() {defaultValue=0.3, group=Noise, parameterName=Eataway Tile U}
    Tile V=ScalarParameter() {defaultValue=0.4, group=Noise, parameterName=Eataway Tile V}
    Pan U=ScalarParameter() {defaultValue=-0.2, group=Noise, parameterName=Eataway Pan U}
    Pan V=ScalarParameter() {defaultValue=0.5, parameterName=Eataway Pan V}
    Offset U=DynamicParameter() {defaultValue=<rgba 0 1 1 550>, paramNames [RandomPerRibbon, EnergyFillFactor, Param3, Param4]}.RandomPerRibbon
   ) {materialFunction=&MF_PanTileOffset}.UV's) {name=Eataway UVs, nodeColor=<rgba 0.9922 0 1 1>}
  NamedRerouteDeclaration(None=MaterialFunctionCall(
    UV=TextureCoordinate()
    Tile U=ScalarParameter() {defaultValue=0.5, group=Noise, parameterName=Noise Tile U}
    Tile V=ScalarParameter() {defaultValue=1, group=Noise, parameterName=Noise Tile V}
    Pan U=ScalarParameter() {defaultValue=-0.5, group=Noise, parameterName=Noise Pan U}
    Pan V=ScalarParameter() {parameterName=Noise Pan V}
    Offset U=DynamicParameter() {defaultValue=<rgba 0 1 1 550>, paramNames [RandomPerRibbon, EnergyFillFactor, Param3, Param4]}.RandomPerRibbon
    Offset V=ScalarParameter() {parameterName=Noise Offset V}
   ) {materialFunction=&MF_PanTileOffset}.UV's) {name=Noise UVs, nodeColor=<rgba 0.9922 0 1 1>}
  Subtract(A=ScalarParameter() {defaultValue=0.5, parameterName=EnergySmoothMax}) {constB=0.05}
 ]
}
