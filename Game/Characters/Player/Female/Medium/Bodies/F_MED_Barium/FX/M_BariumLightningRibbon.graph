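# Additive unlit Niagara ribbon material for a lightning trail in the Barium player character's FX.
# Particle color drives a panning-noise center streak over a dimmer, saturated secondary layer, shaped by ribbon-edge and lengthwise falloffs and sharpened with SmoothStep.
# Panning swirl noise erodes opacity increasingly along the ribbon, combined with particle alpha, depth fade and camera fade.
# A RandomPerRibbon dynamic parameter offsets each noise lookup, and a CameraOffset function supplies depth offset.
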Material /Game/Characters/Player/Female/Medium/Bodies/F_MED_Barium/FX/M_BariumLightningRibbon
graph {
 properties {bUsedWithNiagaraRibbons=true, blendMode=BLEND_Additive, pixelDepthOffsetMode=PDOM_Legacy, shadingModel=MSM_Unlit, usedShadingModels=Unlit}
 nodes [
  n1 = Multiply(A=VectorParameter() {group=Testing, parameterName=TestColor, defaultValue=<rgba 1 1 1 1>}.RGB, B=ParticleColor().RGB)
  n2 = Max(A=@n1) {constB=0.01}
  n3 = Multiply(
   A=MaterialFunctionCall(Vector 3=@n1) {materialFunction=/Engine/Functions/Engine_MaterialFunctions02/Utility/VectorLength}.V3 Length
   B=Power(Base=Normalize(VectorInput=@n2))
  )
  n4 = ComponentMask(None=TextureCoordinate()) {r=true}
  n5 = Multiply(
   A=ScalarParameter() {defaultValue=0.2, group=CenterEmissive, parameterName=SecondaryDarkness}
   B=Power(Base=OneMinus(None=@n4), Exp=ScalarParameter() {defaultValue=6, group=CenterEmissive, parameterName=SecondaryColorFalloffPower}) {constExponent=10}
  )
  n6 = ScalarParameter() {group=Noise, parameterName=SecondaryNoiseMin}
  n7 = ScalarParameter() {defaultValue=1, group=Noise, parameterName=SecondaryNoiseMax}
  n8 = TextureCoordinate() {coordinateIndex=1}
  n9 = ScalarParameter() {defaultValue=0.4, group=Noise, parameterName=Noise Tile U}
  n10 = ScalarParameter() {defaultValue=1, group=Noise, parameterName=Noise Tile V}
  n11 = ScalarParameter() {defaultValue=3, group=Noise, parameterName=Noise Pan U}
  n12 = DynamicParameter() {defaultValue=<rgba 0.5 0 0 0>, paramNames [RandomPerRibbon, Param2, Param3, Param4]}
  n13 = ChannelMaskParameter(None=TextureSampleParameter2D(UVs=MaterialFunctionCall!(UV=@n8, Tile U=@n9, Tile V=@n10, Pan U=@n11, Pan V, Offset U=@n12.RandomPerRibbon, Offset V, Time) {materialFunction=/Game/Effects/Fort_Effects/Materials/_Functions/MF_PanTileOffset}.UV's) {group=Noise, parameterName=NoiseTexture, texture=/Game/Effects/Fort_Effects/Textures/T_InkySmokeTrail}.RGBA) {group=Noise, maskChannel=Green, parameterName=NoiseMask, defaultValue=<rgba 0 1 0 0>}
  n14 = TextureSampleParameter2D(UVs=MaterialFunctionCall(
    UV=TextureCoordinate() {coordinateIndex=1}
    Tile U=ScalarParameter() {defaultValue=0.4, group=CenterEmissive, parameterName=Center Tile U}
    Tile V=ScalarParameter() {defaultValue=1, group=CenterEmissive, parameterName=Center Tile V}
    Pan U=ScalarParameter() {defaultValue=3, group=CenterEmissive, parameterName=Center Pan U}
    Offset U=DynamicParameter() {defaultValue=<rgba 0.5 0 0 0>, paramNames [RandomPerRibbon, Param2, Param3, Param4]}.RandomPerRibbon
    Offset V=ScalarParameter() {group=CenterEmissive, parameterName=CenterOffsetV}
   ) {materialFunction=&MF_PanTileOffset}.UV's) {group=CenterEmissive, parameterName=CenterTexture, sortPriority=1, texture=&T_InkySmokeTrail}
  n15 = ChannelMaskParameter(None=@n14.RGBA) {group=CenterEmissive, maskChannel=Alpha, parameterName=CenterTextureMask, defaultValue=<rgba 0 0 0 1>}
  n16 = LinearInterpolate(
   A=Multiply(A=Multiply(A=@n3, B=@n5) {constB=0.5}, B=SmoothStep(Min=@n6, Max=@n7, Value=@n13.R))
   B=@n1
   Alpha=SmoothStep(
    Min=ScalarParameter() {group=CenterEmissive, parameterName=CenterNoiseMin}
    Max=ScalarParameter() {defaultValue=1, group=CenterEmissive, parameterName=CenterNoiseMax}
    Value=@n15.R
   )
  )
  n17 = ComponentMask(None=TextureCoordinate()) {g=true}
  n18 = Multiply(A=Power(Base=@n17) {constExponent=2.5}, B=Power(Base=OneMinus(None=@n17)) {constExponent=2.5})
  n19 = OneMinus(None=ComponentMask(None=TextureCoordinate()) {r=true})
  n20 = Saturate(None=Multiply(A=Multiply(A=@n18) {constB=12}, B=Power(Base=@n19) {constExponent=5}))
  MaterialFunctionCall(Emissive=SmoothStep(
    Min=ScalarParameter() {parameterName=SmoothMin}
    Max=ScalarParameter() {defaultValue=0.072, parameterName=SmoothMax}
    Value=Multiply(A=@n16, B=@n20)
   ) {constMax=0.05}) {materialFunction=&MF_FX_EmissiveAdjustments}
  MaterialFunctionCall(Offset Amount=ScalarParameter() {parameterName=CameraOffset}) {materialFunction=/Engine/Functions/Engine_MaterialFunctions02/WorldPositionOffset/CameraOffset}
  n21 = TextureSampleParameter2D(UVs=MaterialFunctionCall(
    UV=TextureCoordinate() {coordinateIndex=1}
    Tile U=ScalarParameter() {defaultValue=0.3, group=EatawayOpacity, parameterName=Tile U Eataway}
    Tile V=ScalarParameter() {defaultValue=1, group=EatawayOpacity, parameterName=Tile V Eataway}
    Pan U=ScalarParameter() {defaultValue=-0.2, group=EatawayOpacity, parameterName=Pan U Eataway}
    Pan V=ScalarParameter() {defaultValue=0.2, group=EatawayOpacity, parameterName=Pan V Eataway}
    Offset U=DynamicParameter() {defaultValue=<rgba 0.5 0 0 0>, paramNames [RandomPerRibbon, Param2, Param3, Param4]}.RandomPerRibbon
   ) {materialFunction=&MF_PanTileOffset}.UV's) {group=EatawayOpacity, parameterName=EatawayNoise, texture=/Game/Weapons/WeaponSkins/Textures/T_Swirly_Noise_highContrast}
  n22 = TextureCoordinate()
  n23 = Power(
   Base=OneMinus(None=ComponentMask(None=@n22) {r=true})
   Exp=ScalarParameter() {defaultValue=3, group=EatawayOpacity, parameterName=BreakupPow, sliderMax=200}
  ) {constExponent=80}
  n24 = Multiply(
   A=SmoothStep(Value=LinearInterpolate(A=@n21.G, Alpha=@n23) {constA=1}) {constMax=0.02}
   B=DepthFade(FadeDistance=ScalarParameter() {defaultValue=10, parameterName=DepthFadeDistance})
  )
  n25 = MaterialFunctionCall(
   Fade Length=ScalarParameter() {defaultValue=50, parameterName=CamDepthFadeLength}
   Fade Offset=ScalarParameter() {defaultValue=50, parameterName=CamDepthFadeOffset}
  ) {materialFunction=&CameraDepthFade}
  Multiply(
   A=Multiply(A=ParticleColor().A, B=ScalarParameter() {defaultValue=1, group=Testing, parameterName=TestAlpha})
   B=Multiply(A=Multiply(A=@n24, B=@n25.Result), B=ScalarParameter() {defaultValue=1, parameterName=OverallOpacity})
  )
 ]
}
