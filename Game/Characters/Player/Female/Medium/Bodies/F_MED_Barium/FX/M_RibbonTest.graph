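# M_RibbonTest is a Default Lit test material for Niagara ribbons, stored in the F_MED_Barium player body's FX folder.
# It multiplies a checker grid texture by a red-to-green gradient driven by the U texture coordinate, which visualizes UV layout along a ribbon.
# A Particle Color node sits unused, and the graph shows no link from the multiply to a material output.
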Material /Game/Characters/Player/Female/Medium/Bodies/F_MED_Barium/FX/M_RibbonTest
graph {
 properties {bUsedWithNiagaraRibbons=true, pixelDepthOffsetMode=PDOM_Legacy, usedShadingModels=Default Lit}
 nodes [
  n1 = TextureCoordinate()
  Multiply(
   A=TextureSample() {texture=/Game/Maps/Textures/checkerGrid}.RGB
   B=LinearInterpolate(A=Constant3Vector() {constant=<rgba 1 0 0 0>}, B=Constant3Vector() {constant=<rgba 0 1 0 0>}, Alpha=ComponentMask(None=@n1) {r=true})
  )
  ParticleColor()
 ]
}
