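# Niagara system for goo-dynamics ribbons on the helmet of the Barium female player character, tinted by the user lightning color.
# A burst spawns grouped ribbon chains, each anchored to its own bone or socket on the skeletal mesh.
# Particles chase goal points along the ribbon axis with per-ribbon random lag, while compression adjusts width and slack; splay force is disabled.
# Kill-boost fade and normalized speed drive a clamped multiplier for ribbon length and width.
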NiagaraSystem /Game/Characters/Player/Female/Medium/Bodies/F_MED_Barium/FX/NS_BariumHelmetFX
dependencies {
 usedDataInterfaces [&NiagaraDataInterfaceArrayInt32, &NiagaraDataInterfaceArrayFloat3, &NiagaraDataInterfaceCurve]
 usedDynamicInputs [&LerpQuaternion, &Divide_Float]
 usedModules [
  &SystemState
  &self:KillBoost
  &self:SystemSpawnScript.SystemScriptSource.SystemScriptGraph.NiagaraNodeAssignment_0.SetVariables_#1
  &self:SystemSpawnScript.SystemScriptSource.SystemScriptGraph.NiagaraNodeAssignment_3.SetVariables_#2
  &self:SystemSpawnScript.SystemScriptSource.SystemScriptGraph.NiagaraNodeAssignment_2.SetVariables_#3
 ]
}
emitters {HelmetRibbons {
  emitterSpawnScript [/Game/Effects/Niagara/Modules/GooDynamics/FN_GooGlobalProperties(
    EndBoneIndexArray=DataInterface() {propertyValues {GpuSyncMode=SyncCpuToGpu, MaxElements=0, IntData [0]}}
    EndBoneOffsetArray=DataInterface() {propertyValues {GpuSyncMode=SyncCpuToGpu, MaxElements=0, FloatData [<0 0 0>]}}
    MaxDistance=15
    NumberOfRibbons=2
    ParticlesPerRibbon=5
    RibbonLength=15
    StartBoneIndexArray=DataInterface() {propertyValues {GpuSyncMode=SyncCpuToGpu, MaxElements=0, IntData [0]}}
    StartBoneOffsetArray=DataInterface() {propertyValues {GpuSyncMode=SyncCpuToGpu, MaxElements=0, FloatData [<8 0 -6>, <-15 0 0>, <-2 0 -3>, <2 0 3>, <15 0 0>, <-8 0 6>]}}
   )]
  emitterUpdateScript [
   &EmitterState(
    Inactive Response=::NewEnumerator1
    Life Cycle Mode=::NewEnumerator1
    Loop Behavior=::NewEnumerator1
    Loop Duration Mode=::NewEnumerator1
    UseLoopDelay=false
   )
   &SpawnBurst_Instantaneous(
    Spawn Count=&Multiply_Int(A=$Emitter.ParticlesPerRibbon, B=$Emitter.NumberOfRibbons)
    Use Spawn Probability=false
    UseLoopCountLimit=false
   )
   /SetVariables(Emitter.LengthMultiplier:NiagaraFloat=&ClampFloat(Float=&Add_Float(
      A=&Multiply_Float(A=$System.KillBoostFade, B=1)
      B=&Multiply_Float(A=$System.NormalizedSpeedMask, B=0.8)
     ))) {isSetParameters=true}
  ]
  particleSpawnScript [
   &FN_SetSpawnGroup(
    Inclusivity=true
    Normalize Spawn Group=true
    Number of Spawn Groups:NiagaraInt32=$Emitter.NumberOfRibbons
    Particles Per Group:NiagaraInt32=$Emitter.ParticlesPerRibbon
    Set Ribbon ID and Ribbon Link Order=true
   )
   /SetVariables(Particles.RibbonRandom:NiagaraFloat=&RandomRangeFloat(
     Minimum=0
     Maximum=1
     Randomness Mode=::NewEnumerator1
     Random Seed=&Add_Integer(A=$Engine.System.RandomSeed, B=$Particles.SpawnGroup)
     Fixed Random Seed=true
    )) {isSetParameters=true}
   &InitializeRibbon(Color=$User.LightningColor_1, Ribbon Width=8, UsePositionOffset=false, Write Material Random=true)
   /SetVariables(Particles.MaterialRandom:NiagaraFloat=$Particles.RibbonRandom) {isSetParameters=true}
   &SkeletalMeshLocation(
    Bone / Socket Index=$Particles.SpawnGroup
    MaskSpawnGroup=false
    Mesh Sampling Type=::NewEnumerator5
    Orientation Sampling=::NewEnumerator1
    Position Sampling=::NewEnumerator2
    Skeletal Mesh=$User.SkeletalMesh
    Socket Sampling Mode=::NewEnumerator5
    Use Sampled Mesh Attribute as Ribbon ID=false
   )
   /Game/Effects/Niagara/Modules/GooDynamics/FN_GooCalculateGoalPosition(
    Axis:Vector3f=$Emitter.RibbonAxis
    Scale:Vector3f=&Multiply_VectorByFloat(
     Vector=&Multiply_VectorByFloat(Vector=&VectorFromFloat(Value=$Emitter.RibbonLength), Float=$Particles.RibbonLinkOrder)
     Float=$Emitter.LengthMultiplier
    )
    StartPosition:Vector3f=$Particles.SkeletalMeshLocation.SampledPosition
   ) {module=FN_CalculateGoalPosition}
   /SetVariables(Particles.Position=$Particles.GoalPosition) {isSetParameters=true}
   /SetVariables(Particles.PreviousRotation:Quat4f=$Engine.Owner.Rotation) {isSetParameters=true}
   &DynamicMaterialParameters(
    Index 0 Param 1=&Lerp_Float(A=0, B=10, Alpha=$Particles.RibbonRandom)
    Module.Param0WriteEnabled=true
    Module.Param1WriteEnabled=true
    Module.Param2WriteEnabled=true
    Module.Param3WriteEnabled=true
   )
   /SetVariables(Particles.SpriteSize=<10 10>) {isSetParameters=true}
  ]
  particleUpdateScript [
   &ParticleState(Kill Particles When Lifetime Has Elapsed=false, Let Infinitely Lived Particles Die When Emitter Deactivates=true)
   &SkeletalMeshLocation(
    Bone / Socket Index=$Particles.SpawnGroup
    MaskSpawnGroup=false
    Mesh Sampling Type=::NewEnumerator5
    Orientation Sampling=::NewEnumerator1
    Position Sampling=::NewEnumerator2
    Skeletal Mesh=$User.SkeletalMesh
    Socket Sampling Mode=::NewEnumerator5
    Use Sampled Mesh Attribute as Ribbon ID=false
   ) {module=SkeletalMeshLocation001}
   /Game/Effects/Niagara/Modules/GooDynamics/FN_GooSplayForce(
    CenterPoint:NiagaraPosition=/Niagara/DynamicInputs/Helpers/SimulationPosition()
    Factor:NiagaraFloat=$Particles.Compression
    PreviousRotation:Quat4f=$Particles.PreviousRotation
    Splay Amount=0
   ) {enabled=false}
   &SolveForcesAndVelocity(Clamp Velocity=false, Limit Acceleration=false)
   &FN_GooCalculateGoalPosition(
    Axis:Vector3f=$Emitter.RibbonAxis
    Scale:Vector3f=&Multiply_VectorByFloat(
     Vector=&Multiply_VectorByFloat(Vector=&VectorFromFloat(Value=$Emitter.RibbonLength), Float=$Particles.RibbonLinkOrder)
     Float=$Emitter.LengthMultiplier
    )
    StartPosition:Vector3f=$Particles.SkeletalMeshLocation001.SampledPosition
   ) {module=FN_CalculateGoalPosition001}
   /Game/Effects/Niagara/Modules/GooDynamics/FN_GooApproachGoalPosition(
    AttractionForce:NiagaraFloat=&Lerp_Float(A=$Emitter.ResolveSpeedLow, B=$Emitter.ResolveSpeedHigh, Alpha=$Particles.RibbonRandom)
    GoalPosition:NiagaraPosition=$Particles.GoalPosition
    MaximumDistance:NiagaraFloat=&Multiply_Float(
     A=&FloatFromCurve(
      FloatCurve=DataInterface() {propertyValues {Curve {keys [
          {interpMode=RCIM_Cubic, tangentMode=RCTM_User, time=0, value=0}
          {arriveTangent=1.3325, interpMode=RCIM_Cubic, leaveTangent=1.3325, tangentMode=RCTM_User, time=1, value=1}
         ]}}}
      CurveIndex=$Particles.RibbonLinkOrder
      Scale Curve=&Lerp_Float(
       A=$Emitter.MaxDistance
       B=$Emitter.MaxDistanceCompressed
       Alpha=&ClampFloat(Float=&Multiply_Float(A=-1, B=$Particles.Compression))
      )
     )
     B=&Lerp_Float(A=$Emitter.MaxDistance_RandFactorLow, B=$Emitter.MaxDistance_RandFactorHigh, Alpha=$Particles.RibbonRandom)
    )
    RibbonAxis:Vector3f=$Particles.RibbonAxis
   )
   /Game/Effects/Niagara/Modules/GooDynamics/FN_GooRibbonWidthCompression(
    Compression Factor:NiagaraFloat=$Particles.Compression
    Ribbon Scale Mult:NiagaraFloat=$Emitter.LengthMultiplier
    Scale Factor when Compressed:NiagaraFloat=$Emitter.Width Factor when Compressed
    Scale Factor when Stretched:NiagaraFloat=$Emitter.Width Factor when Stretched
   )
   /SetVariables(Particles.PreviousRotation:Quat4f=$Engine.Owner.Rotation) {isSetParameters=true}
   &DynamicMaterialParameters(
    Index 0 Param 2=$System.NormalizedOverallMask
    Module.Param0WriteEnabled=false
    Module.Param1WriteEnabled=true
    Module.Param2WriteEnabled=false
    Module.Param3WriteEnabled=false
   ) {module=DynamicMaterialParameters001}
  ]
  rendererClasses [&NiagaraRibbonRendererProperties]
  renderers [{rendererClass=&NiagaraRibbonRendererProperties}]
 }}
summary {
 systemName=NS_BariumHelmetFX
 emitters [{emitterName=HelmetRibbons, rendererClasses [&NiagaraRibbonRendererProperties]}]
 userVariables [
  User.Distance {type=NiagaraFloat, defaultValue=0, description=""}
  User.KillBoostTime {type=NiagaraFloat, defaultValue=-50, description=""}
  User.LightningAcceleration {type=NiagaraFloat, defaultValue=0, description=""}
  User.LightningColor_1 {type=LinearColor, description="", defaultValue {struct=LinearColor, value=<rgba 1 1 1 1>}}
  User.SkeletalMesh {
   type=NiagaraDataInterfaceSkeletalMesh
   defaultValue {
    struct=NiagaraExt_VariableValue_DataInterface
    value {dataInterface=&self:NiagaraDataInterfaceSkeletalMesh_0}
   }
  }
 ]
}
