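# Niagara kill-burst effect for the Barium character, shooting short-lived paint-coloured ribbons from skeletal mesh bones at the helmet, arms and hand.
# Each emitter bursts ribbon chains, with helmet and jacket spawns gated by user probabilities.
# GooDynamics modules pull points toward wind-bent, upward-curving goals that overshoot, then retract over lifetime.
# Colour starts intensely bright and cools while alpha fades in and out.
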NiagaraSystem /Game/Characters/Player/Female/Medium/Bodies/F_MED_Barium/FX/NS_BariumKillBurst
dependencies {
 usedDataInterfaces [
  &NiagaraDataInterfaceArrayInt32
  &NiagaraDataInterfaceArrayFloat3
  &NiagaraDataInterfaceColorCurve
  &NiagaraDataInterfaceCurve
 ]
 usedDynamicInputs []
 usedModules [&SystemState]
}
emitters {
 Helmet {
  emitterSpawnScript [/Game/Effects/Niagara/Modules/GooDynamics/FN_GooGlobalProperties(
    EndBoneIndexArray=DataInterface() {propertyValues {GpuSyncMode=SyncCpuToGpu, MaxElements=0, IntData [0]}}
    EndBoneOffsetArray=DataInterface() {propertyValues {GpuSyncMode=SyncCpuToGpu, MaxElements=0, FloatData [<0 0 0>]}}
    MaxDistance:NiagaraFloat=$User.HipRibbonLength
    MaxDistanceCompressed=10
    MaxDistance_RandFactorHigh=1
    MaxDistance_RandFactorLow=1
    NumberOfRibbons=2
    ParticlesPerRibbon=4
    ResolveSpeedHigh=3
    RibbonLength:NiagaraFloat=$User.HipRibbonLength
    RibbonLengthRandHigh=1
    RibbonLengthRandLow=1
    StartBoneIndexArray=DataInterface() {propertyValues {GpuSyncMode=SyncCpuToGpu, MaxElements=0, IntData [12,12]}}
    StartBoneOffsetArray=DataInterface() {propertyValues {GpuSyncMode=SyncCpuToGpu, MaxElements=0, FloatData [<8 8 -7>, <8 8 7>]}}
   )]
  emitterUpdateScript [
   &EmitterState()
   &SpawnBurst_Instantaneous(
    Spawn Count=&Multiply_Int(A=$Emitter.ParticlesPerRibbon, B=$Emitter.NumberOfRibbons)
    Spawn Probability=$User.SpawnHelmetFX
    Use Spawn Probability=true
    UseLoopCountLimit=false
   )
  ]
  particleSpawnScript [
   /SetVariables(Particles.SpawnGroup=/Niagara/DynamicInputs/Divide/Divide_Int(A:NiagaraInt32=&ReturnExecIndex(), B:NiagaraInt32=$Emitter.ParticlesPerRibbon)) {isSetParameters=true, module=SetVariables_#1}
   /SetVariables(
    Particles.RibbonID=/Niagara/DynamicInputs/Meshes/MakeRibbonIDFromSpawnGroup()
    Particles.RibbonLinkOrder:NiagaraFloat=&Divide_Float(
     A=&MakeFloatFromInt(Int=&ModuloInt(A=&ReturnExecIndex(), B=$Emitter.ParticlesPerRibbon))
     B=&Subtract_Float(A=&MakeFloatFromInt(Int=$Emitter.ParticlesPerRibbon), B=1)
    )
   ) {isSetParameters=true, module=SetVariables_#2}
   /SetVariables(Particles.RibbonRandom:NiagaraFloat=&RandomRangeFloat(
     Minimum=0
     Maximum=1
     Randomness Mode=::NewEnumerator1
     Random Seed=&Add_Integer(A=$Engine.System.RandomSeed, B=$Particles.SpawnGroup)
     Fixed Random Seed=true
    )) {isSetParameters=true, module=SetVariables_#3}
   &InitializeRibbon(
    Color=&MultiplyLinearColor(
     Linear Color=$User.BariumPaintColor
     Scale Factor=&MakeVector4_FromLinearColor(LinearColor=&ColorFromCurve(
       ColorCurve=DataInterface() {propertyValues {
         AlphaCurve {keys [<at 0.574 1 RCIM_Cubic>,<at 0.9977 0>]}
         BlueCurve {keys [<at 0 1 RCIM_Cubic>,<at 1 1 RCIM_Cubic>]}
         GreenCurve {keys [<at 0 1 RCIM_Cubic>,<at 1 1 RCIM_Cubic>]}
         RedCurve {keys [<at 0 1 RCIM_Cubic>,<at 1 1 RCIM_Cubic>]}
        }}
       CurveIndex=$Particles.RibbonLinkOrder
      ))
    )
    Lifetime=&Lerp_Float(A=2, B=3, Alpha=$Particles.RibbonRandom)
    Ribbon Width=$User.HipRibbonWidth
    UsePositionOffset=false
    Write Material Random=true
   )
   /SetVariables(Particles.MaterialRandom:NiagaraFloat=$Particles.RibbonRandom) {isSetParameters=true, module=SetVariables_#4}
   &SkeletalMeshLocation(
    Bone / Socket Index=&SelectIntFromArray(Direct Array Index=$Particles.SpawnGroup, Direct Array Mode=::NewEnumerator1, Int Selection Array=$Emitter.StartBoneIndexArray)
    Bone Sampling Mode=::NewEnumerator4
    MaskSpawnGroup=false
    Orientation Sampling=::NewEnumerator1
    Position Sampling=::NewEnumerator2
    Sampled Position Offset=&SelectVectorFromArray(Direct Array Index=$Particles.SpawnGroup, Direct Array Mode=::NewEnumerator1, Vector Selection Array=$Emitter.StartBoneOffsetArray)
    Skeletal Mesh=$User.SkeletalMesh
    Use Sampled Mesh Attribute as Ribbon ID=false
   )
   /Game/Effects/Niagara/Modules/GooDynamics/FN_GooCalculateGoalPosition(
    Axis:Vector3f=$Emitter.RibbonAxis
    Scale:Vector3f=&Multiply_VectorByFloat(
     Vector=&Multiply_VectorByFloat(
      Vector=&VectorFromFloat(Value=&Multiply_Float(A=$Emitter.RibbonLength, B=0))
      Float=&Lerp_Float(A=$Emitter.RibbonLengthRandMin, B=$Emitter.RibbonLengthRandMax, Alpha=$Particles.RibbonRandom)
     )
     Float=$Particles.RibbonLinkOrder
    )
    StartPosition:Vector3f=$Particles.SkeletalMeshLocation.SampledPosition
   ) {module=FN_CalculateGoalPosition}
   /SetVariables(Particles.Position=$Particles.GoalPosition) {isSetParameters=true, module=SetVariables_#5}
   /SetVariables(Particles.PreviousRotation:Quat4f=$Engine.Owner.Rotation) {isSetParameters=true, module=SetVariables_#6}
   &DynamicMaterialParameters(
    Index 0 Param 1=&Multiply_Float(A=$Particles.RibbonRandom, B=10)
    Module.Param0WriteEnabled=true
    Module.Param1WriteEnabled=true
    Module.Param2WriteEnabled=true
    Module.Param3WriteEnabled=true
   )
   /SetVariables(Particles.SpriteSize=<10 10>) {isSetParameters=true, module=SetVariables_#7}
  ]
  particleUpdateScript [
   &ParticleState()
   &SkeletalMeshLocation(
    Bone / Socket Index=&SelectIntFromArray(Direct Array Index=$Particles.SpawnGroup, Direct Array Mode=::NewEnumerator1, Int Selection Array=$Emitter.StartBoneIndexArray)
    Bone Sampling Mode=::NewEnumerator4
    MaskSpawnGroup=false
    Orientation Sampling=::NewEnumerator1
    Position Sampling=::NewEnumerator2
    Sampled Position Offset=&SelectVectorFromArray(Direct Array Index=$Particles.SpawnGroup, Direct Array Mode=::NewEnumerator1, Vector Selection Array=$Emitter.StartBoneOffsetArray)
    Skeletal Mesh=$User.SkeletalMesh
    Use Sampled Mesh Attribute as Ribbon ID=false
   ) {module=SkeletalMeshLocation001}
   &SolveForcesAndVelocity(Clamp Velocity=false, Limit Acceleration=false)
   &FN_GooCalculateGoalPosition(
    Axis:Vector3f=/Niagara/DynamicInputs/Math/SlerpVector(
     Vector A:Vector3f=&SlerpVector(
      Vector A:Vector3f=&MaskVectorBySpawnGroup(Masked Vector=<1 -1 0>, Passthrough Vector=<-1 -1 0>)
      Vector B:Vector3f=&TransformVector(Vector=$User.WorldSpaceWindVector, Source Space=::World, Destination Space=::Local)
      Alpha:NiagaraFloat=$User.WorldSpaceWindVectorFactor
      Scale:NiagaraFloat=1
     )
     Vector B:Vector3f=<0 0 1>
     Alpha:NiagaraFloat=&FloatFromCurve(
      FloatCurve=DataInterface() {propertyValues {Curve {keys [{tangentMode=RCTM_User, time=0, value=0}, {tangentMode=RCTM_User, time=1, value=0.5}]}}}
      CurveIndex=$Particles.RibbonLinkOrder
     )
     Scale:NiagaraFloat=1
    )
    CustomOrientation:Quat4f=$Engine.Owner.Rotation
    Scale:Vector3f=&Multiply_VectorByFloat(
     Vector=&Multiply_VectorByFloat(
      Vector=&VectorFromFloat(Value=&Multiply_Float(
        A=$Emitter.RibbonLength
        B=&FloatFromCurve(FloatCurve=DataInterface() {propertyValues {Curve {keys [
             {arriveTangent=30.27, interpMode=RCIM_Cubic, leaveTangent=30.27, tangentMode=RCTM_User, time=0, value=0}
             {interpMode=RCIM_Cubic, tangentMode=RCTM_User, time=0.15, value=1.5}
             <at 0.2 1 RCIM_Cubic>
             {arriveTangent=-3.4902, interpMode=RCIM_Cubic, leaveTangent=-3.4902, tangentMode=RCTM_User, time=1, value=0}
            ]}}})
       ))
      Float=&Lerp_Float(A=$Emitter.RibbonLengthRandMin, B=$Emitter.RibbonLengthRandMax, Alpha=$Particles.RibbonRandom)
     )
     Float=$Particles.RibbonLinkOrder
    )
    StartPosition:Vector3f=$Particles.SkeletalMeshLocation001.SampledPosition
    Use Quaternion Rotation=true
   ) {module=FN_CalculateGoalPosition001}
   /Game/Effects/Niagara/Modules/GooDynamics/FN_GooApproachGoalPosition(
    AttractionForce:NiagaraFloat=&Lerp_Float(A=$Emitter.ResolveSpeedLow, B=$Emitter.ResolveSpeedHigh, Alpha=$Particles.RibbonRandom)
    GoalPosition:NiagaraPosition=$Particles.GoalPosition
    MaximumDistance:NiagaraFloat=&Multiply_Float(
     A=&FloatFromCurve(
      FloatCurve=DataInterface() {propertyValues {Curve {keys [
          {arriveTangent=4.1785, interpMode=RCIM_Cubic, time=0, value=0}
          {arriveTangent=2.3315, interpMode=RCIM_Cubic, leaveTangent=2.3315, tangentMode=RCTM_User, time=1, value=1}
         ]}}}
      CurveIndex=$Particles.RibbonLinkOrder
      Scale Curve=&Multiply_Float(
       A=$Emitter.MaxDistance
       B=&FloatFromCurve(FloatCurve=DataInterface() {propertyValues {Curve {keys [<at 0.75 1>,<at 1 0>]}}})
      )
     )
     B=&Lerp_Float(A=$Emitter.MaxDistance_RandFactorLow, B=$Emitter.MaxDistance_RandFactorHigh, Alpha=$Particles.RibbonRandom)
    )
    RibbonAxis:Vector3f=$Particles.RibbonAxis
   )
   /SetVariables(Particles.PreviousRotation:Quat4f=$Engine.Owner.Rotation) {isSetParameters=true, module=SetVariables_#8}
   &ScaleColor(
    Scale Alpha=&FloatFromCurve(FloatCurve=DataInterface() {propertyValues {Curve {keys [<at 0 0>, {arriveTangent=6.6358, leaveTangent=6.6358, time=0.5, value=1}, <at 0.7 1>, <at 1 0>]}}})
    Scale RGB=&VectorFromFloat(Value=&FloatFromCurve(FloatCurve=DataInterface() {propertyValues {Curve {keys [{arriveTangent=-319.691, interpMode=RCIM_Cubic, leaveTangent=-319.691, tangentMode=RCTM_User, time=0, value=800}, <at 0.4 15>, <at 1 1>]}}}))
    ScaleA=true
    ScaleRGB=true
   )
   &DynamicMaterialParameters(
    Index 0 Param 1=&Multiply_Float(A=$Particles.RibbonRandom, B=10)
    Index 0 Param 2=$User.EnergyFillFactor
    Module.Param0WriteEnabled=false
    Module.Param1WriteEnabled=true
    Module.Param2WriteEnabled=false
    Module.Param3WriteEnabled=false
   ) {module=DynamicMaterialParameters001}
  ]
  rendererClasses [&NiagaraRibbonRendererProperties]
  renderers [{rendererClass=&NiagaraRibbonRendererProperties}]
 }
 Helmet001 {
  emitterSpawnScript [&FN_GooGlobalProperties(
    EndBoneIndexArray=DataInterface() {propertyValues {GpuSyncMode=SyncCpuToGpu, MaxElements=0, IntData [0]}}
    EndBoneOffsetArray=DataInterface() {propertyValues {GpuSyncMode=SyncCpuToGpu, MaxElements=0, FloatData [<0 0 0>]}}
    MaxDistance:NiagaraFloat=$User.HipRibbonLength
    MaxDistanceCompressed=10
    MaxDistance_RandFactorHigh=1
    MaxDistance_RandFactorLow=1
    NumberOfRibbons=2
    ParticlesPerRibbon=2
    ResolveSpeedHigh=3
    RibbonLength:NiagaraFloat=$User.HipRibbonLength
    RibbonLengthRandHigh=1
    RibbonLengthRandLow=1
    StartBoneIndexArray=DataInterface() {propertyValues {GpuSyncMode=SyncCpuToGpu, MaxElements=0, IntData [12,12]}}
    StartBoneOffsetArray=DataInterface() {propertyValues {GpuSyncMode=SyncCpuToGpu, MaxElements=0, FloatData [<8 8 -7>, <8 8 7>]}}
   )]
  emitterUpdateScript [
   &EmitterState()
   &SpawnBurst_Instantaneous(
    Spawn Count=&Multiply_Int(A=$Emitter.ParticlesPerRibbon, B=$Emitter.NumberOfRibbons)
    Spawn Probability=$User.SpawnHelmetFX
    Use Spawn Probability=true
    UseLoopCountLimit=false
   )
  ]
  particleSpawnScript [
   /SetVariables(Particles.SpawnGroup=&Divide_Int(A:NiagaraInt32=&ReturnExecIndex(), B:NiagaraInt32=$Emitter.ParticlesPerRibbon)) {isSetParameters=true, module=SetVariables_#1}
   /SetVariables(
    Particles.RibbonID=&MakeRibbonIDFromSpawnGroup()
    Particles.RibbonLinkOrder:NiagaraFloat=&Divide_Float(
     A=&MakeFloatFromInt(Int=&ModuloInt(A=&ReturnExecIndex(), B=$Emitter.ParticlesPerRibbon))
     B=&Subtract_Float(A=&MakeFloatFromInt(Int=$Emitter.ParticlesPerRibbon), B=1)
    )
   ) {isSetParameters=true, module=SetVariables_#2}
   /SetVariables(Particles.RibbonRandom:NiagaraFloat=&RandomRangeFloat(
     Minimum=0
     Maximum=1
     Randomness Mode=::NewEnumerator1
     Random Seed=&Add_Integer(A=$Engine.System.RandomSeed, B=$Particles.SpawnGroup)
     Fixed Random Seed=true
    )) {isSetParameters=true, module=SetVariables_#3}
   &InitializeRibbon(
    Color=&MultiplyLinearColor(
     Linear Color=$User.BariumPaintColor
     Scale Factor=&MakeVector4_FromLinearColor(LinearColor=&ColorFromCurve(
       ColorCurve=DataInterface() {propertyValues {
         AlphaCurve {keys [<at 0.574 1 RCIM_Cubic>,<at 0.9977 0>]}
         BlueCurve {keys [<at 0 1 RCIM_Cubic>,<at 1 1 RCIM_Cubic>]}
         GreenCurve {keys [<at 0 1 RCIM_Cubic>,<at 1 1 RCIM_Cubic>]}
         RedCurve {keys [<at 0 1 RCIM_Cubic>,<at 1 1 RCIM_Cubic>]}
        }}
       CurveIndex=$Particles.RibbonLinkOrder
      ))
    )
    Lifetime=&Lerp_Float(A=2, B=3, Alpha=$Particles.RibbonRandom)
    Ribbon Width=$User.HipRibbonWidth
    UsePositionOffset=false
    Write Material Random=true
   )
   /SetVariables(Particles.MaterialRandom:NiagaraFloat=$Particles.RibbonRandom) {isSetParameters=true, module=SetVariables_#4}
   &SkeletalMeshLocation(
    Bone / Socket Index=&SelectIntFromArray(Direct Array Index=$Particles.SpawnGroup, Direct Array Mode=::NewEnumerator1, Int Selection Array=$Emitter.StartBoneIndexArray)
    Bone Sampling Mode=::NewEnumerator4
    MaskSpawnGroup=false
    Orientation Sampling=::NewEnumerator1
    Position Sampling=::NewEnumerator2
    Sampled Position Offset=&SelectVectorFromArray(Direct Array Index=$Particles.SpawnGroup, Direct Array Mode=::NewEnumerator1, Vector Selection Array=$Emitter.StartBoneOffsetArray)
    Skeletal Mesh=$User.SkeletalMesh
    Use Sampled Mesh Attribute as Ribbon ID=false
   )
   &FN_GooCalculateGoalPosition(
    Axis:Vector3f=$Emitter.RibbonAxis
    Scale:Vector3f=&Multiply_VectorByFloat(
     Vector=&Multiply_VectorByFloat(
      Vector=&VectorFromFloat(Value=&Multiply_Float(A=$Emitter.RibbonLength, B=0))
      Float=&Lerp_Float(A=$Emitter.RibbonLengthRandMin, B=$Emitter.RibbonLengthRandMax, Alpha=$Particles.RibbonRandom)
     )
     Float=$Particles.RibbonLinkOrder
    )
    StartPosition:Vector3f=$Particles.SkeletalMeshLocation.SampledPosition
   ) {module=FN_CalculateGoalPosition}
   /SetVariables(Particles.Position=$Particles.GoalPosition) {isSetParameters=true, module=SetVariables_#5}
   /SetVariables(Particles.PreviousRotation:Quat4f=$Engine.Owner.Rotation) {isSetParameters=true, module=SetVariables_#6}
   &DynamicMaterialParameters(
    Index 0 Param 1=&Multiply_Float(A=$Particles.RibbonRandom, B=10)
    Module.Param0WriteEnabled=true
    Module.Param1WriteEnabled=true
    Module.Param2WriteEnabled=true
    Module.Param3WriteEnabled=true
   )
   /SetVariables(Particles.SpriteSize=<10 10>) {isSetParameters=true, module=SetVariables_#7}
  ]
  particleUpdateScript [
   &ParticleState()
   &SkeletalMeshLocation(
    Bone / Socket Index=&SelectIntFromArray(Direct Array Index=$Particles.SpawnGroup, Direct Array Mode=::NewEnumerator1, Int Selection Array=$Emitter.StartBoneIndexArray)
    Bone Sampling Mode=::NewEnumerator4
    MaskSpawnGroup=false
    Orientation Sampling=::NewEnumerator1
    Position Sampling=::NewEnumerator2
    Sampled Position Offset=&SelectVectorFromArray(Direct Array Index=$Particles.SpawnGroup, Direct Array Mode=::NewEnumerator1, Vector Selection Array=$Emitter.StartBoneOffsetArray)
    Skeletal Mesh=$User.SkeletalMesh
    Use Sampled Mesh Attribute as Ribbon ID=false
   ) {module=SkeletalMeshLocation001}
   &SolveForcesAndVelocity(Clamp Velocity=false, Limit Acceleration=false)
   &FN_GooCalculateGoalPosition(
    Axis:Vector3f=&SlerpVector(
     Vector A:Vector3f=&SlerpVector(
      Vector A:Vector3f=&MaskVectorBySpawnGroup(Masked Vector=<1 -1 0>, Passthrough Vector=<-1 -1 0>)
      Vector B:Vector3f=&TransformVector(Vector=$User.WorldSpaceWindVector, Source Space=::World, Destination Space=::Local)
      Alpha:NiagaraFloat=$User.WorldSpaceWindVectorFactor
      Scale:NiagaraFloat=1
     )
     Vector B:Vector3f=<0 0 1>
     Alpha:NiagaraFloat=&FloatFromCurve(
      FloatCurve=DataInterface() {propertyValues {Curve {keys [{tangentMode=RCTM_User, time=0, value=0}, {tangentMode=RCTM_User, time=1, value=0.5}]}}}
      CurveIndex=$Particles.RibbonLinkOrder
     )
     Scale:NiagaraFloat=1
    )
    CustomOrientation:Quat4f=$Engine.Owner.Rotation
    Scale:Vector3f=&Multiply_VectorByFloat(
     Vector=&Multiply_VectorByFloat(
      Vector=&VectorFromFloat(Value=&Multiply_Float(
        A=$Emitter.RibbonLength
        B=&FloatFromCurve(FloatCurve=DataInterface() {propertyValues {Curve {keys [
             {arriveTangent=30.27, interpMode=RCIM_Cubic, leaveTangent=30.27, tangentMode=RCTM_User, time=0, value=0}
             {interpMode=RCIM_Cubic, tangentMode=RCTM_User, time=0.15, value=1.5}
             <at 0.2 1 RCIM_Cubic>
             {arriveTangent=-3.4902, interpMode=RCIM_Cubic, leaveTangent=-3.4902, tangentMode=RCTM_User, time=1, value=0}
            ]}}})
       ))
      Float=&Lerp_Float(A=$Emitter.RibbonLengthRandMin, B=$Emitter.RibbonLengthRandMax, Alpha=$Particles.RibbonRandom)
     )
     Float=$Particles.RibbonLinkOrder
    )
    StartPosition:Vector3f=$Particles.SkeletalMeshLocation001.SampledPosition
    Use Quaternion Rotation=true
   ) {module=FN_CalculateGoalPosition001}
   &FN_GooApproachGoalPosition(
    AttractionForce:NiagaraFloat=&Lerp_Float(A=$Emitter.ResolveSpeedLow, B=$Emitter.ResolveSpeedHigh, Alpha=$Particles.RibbonRandom)
    GoalPosition:NiagaraPosition=$Particles.GoalPosition
    MaximumDistance:NiagaraFloat=&Multiply_Float(
     A=&FloatFromCurve(
      FloatCurve=DataInterface() {propertyValues {Curve {keys [
          {arriveTangent=4.1785, interpMode=RCIM_Cubic, time=0, value=0}
          {arriveTangent=2.3315, interpMode=RCIM_Cubic, leaveTangent=2.3315, tangentMode=RCTM_User, time=1, value=1}
         ]}}}
      CurveIndex=$Particles.RibbonLinkOrder
      Scale Curve=&Multiply_Float(
       A=$Emitter.MaxDistance
       B=&FloatFromCurve(FloatCurve=DataInterface() {propertyValues {Curve {keys [<at 0.75 1>,<at 1 0>]}}})
      )
     )
     B=&Lerp_Float(A=$Emitter.MaxDistance_RandFactorLow, B=$Emitter.MaxDistance_RandFactorHigh, Alpha=$Particles.RibbonRandom)
    )
    RibbonAxis:Vector3f=$Particles.RibbonAxis
   )
   /SetVariables(Particles.PreviousRotation:Quat4f=$Engine.Owner.Rotation) {isSetParameters=true, module=SetVariables_#8}
   &ScaleColor(
    Scale Alpha=&FloatFromCurve(FloatCurve=DataInterface() {propertyValues {Curve {keys [<at 0 0>, {arriveTangent=6.6358, leaveTangent=6.6358, time=0.5, value=1}, <at 0.7 1>, <at 1 0>]}}})
    Scale RGB=&VectorFromFloat(Value=&FloatFromCurve(FloatCurve=DataInterface() {propertyValues {Curve {keys [{arriveTangent=-319.691, interpMode=RCIM_Cubic, leaveTangent=-319.691, tangentMode=RCTM_User, time=0, value=800}, <at 0.4 15>, <at 1 1>]}}}))
    ScaleA=true
    ScaleRGB=true
   )
   &DynamicMaterialParameters(
    Index 0 Param 1=&Multiply_Float(A=$Particles.RibbonRandom, B=10)
    Index 0 Param 2=$User.EnergyFillFactor
    Module.Param0WriteEnabled=false
    Module.Param1WriteEnabled=true
    Module.Param2WriteEnabled=false
    Module.Param3WriteEnabled=false
   ) {module=DynamicMaterialParameters001}
  ]
  rendererClasses [&NiagaraRibbonRendererProperties]
  renderers [{rendererClass=&NiagaraRibbonRendererProperties}]
 }
 Left_Arm {
  emitterSpawnScript [&FN_GooGlobalProperties(
    EndBoneIndexArray=DataInterface() {propertyValues {GpuSyncMode=SyncCpuToGpu, MaxElements=0, IntData [0]}}
    EndBoneOffsetArray=DataInterface() {propertyValues {GpuSyncMode=SyncCpuToGpu, MaxElements=0, FloatData [<0 0 0>]}}
    MaxDistance:NiagaraFloat=$User.ArmRibbonLength
    MaxDistanceCompressed=10
    MaxDistance_RandFactorHigh=1
    MaxDistance_RandFactorLow=1
    NumberOfRibbons=2
    RibbonLength:NiagaraFloat=$User.ArmRibbonLength
    RibbonLengthRandHigh=1
    RibbonLengthRandLow=1
    StartBoneIndexArray=DataInterface() {propertyValues {GpuSyncMode=SyncCpuToGpu, MaxElements=0, IntData [0,1]}}
    StartBoneOffsetArray=DataInterface() {propertyValues {GpuSyncMode=SyncCpuToGpu, MaxElements=0, FloatData [<10 0 0>, <3 0 0>]}}
   )]
  emitterUpdateScript [
   &EmitterState()
   &SpawnBurst_Instantaneous(
    Spawn Count=&Multiply_Int(A=$Emitter.ParticlesPerRibbon, B=$Emitter.NumberOfRibbons)
    Spawn Probability=$User.SpawnJacketFX
    Use Spawn Probability=true
    UseLoopCountLimit=false
   )
  ]
  particleSpawnScript [
   /SetVariables(Particles.SpawnGroup=&Divide_Int(A:NiagaraInt32=&ReturnExecIndex(), B:NiagaraInt32=$Emitter.ParticlesPerRibbon)) {isSetParameters=true, module=SetVariables_#1}
   /SetVariables(
    Particles.RibbonID=&MakeRibbonIDFromSpawnGroup()
    Particles.RibbonLinkOrder:NiagaraFloat=&Divide_Float(
     A=&MakeFloatFromInt(Int=&ModuloInt(A=&ReturnExecIndex(), B=$Emitter.ParticlesPerRibbon))
     B=&Subtract_Float(A=&MakeFloatFromInt(Int=$Emitter.ParticlesPerRibbon), B=1)
    )
   ) {isSetParameters=true, module=SetVariables_#2}
   /SetVariables(Particles.RibbonRandom:NiagaraFloat=&RandomRangeFloat(
     Minimum=0
     Maximum=1
     Randomness Mode=::NewEnumerator1
     Random Seed=&Add_Integer(A=$Engine.System.RandomSeed, B=$Particles.SpawnGroup)
     Fixed Random Seed=true
    )) {isSetParameters=true, module=SetVariables_#3}
   &InitializeRibbon(
    Color=&MultiplyLinearColor(
     Linear Color=$User.BariumPaintColor
     Scale Factor=&MakeVector4_FromLinearColor(LinearColor=&ColorFromCurve(
       ColorCurve=DataInterface() {propertyValues {
         AlphaCurve {keys [<at 0.574 1 RCIM_Cubic>,<at 0.9977 0>]}
         BlueCurve {keys [<at 0 1 RCIM_Cubic>]}
         GreenCurve {keys [<at 0 1 RCIM_Cubic>]}
         RedCurve {keys [<at 0 1 RCIM_Cubic>]}
        }}
       CurveIndex=$Particles.RibbonLinkOrder
      ))
    )
    Lifetime=&Lerp_Float(A=2, B=3, Alpha=$Particles.RibbonRandom)
    Ribbon Width=&FloatFromCurve(
     FloatCurve=DataInterface() {propertyValues {Curve {keys [<at 0 1>,<at 0.5 0.5>,<at 1 1>]}}}
     CurveIndex=$Particles.RibbonLinkOrder
     Scale Curve=$User.ArmRibbonWidth
    )
    UsePositionOffset=false
    Write Material Random=true
   )
   /SetVariables(Particles.MaterialRandom:NiagaraFloat=$Particles.RibbonRandom) {isSetParameters=true, module=SetVariables_#4}
   &SkeletalMeshLocation(
    Bone / Socket Index=&SelectIntFromArray(Direct Array Index=$Particles.SpawnGroup, Direct Array Mode=::NewEnumerator1, Int Selection Array=$Emitter.StartBoneIndexArray)
    Bone Sampling Mode=::NewEnumerator4
    MaskSpawnGroup=false
    Orientation Sampling=::NewEnumerator1
    Position Sampling=::NewEnumerator2
    Sampled Position Offset=&SelectVectorFromArray(Direct Array Index=$Particles.SpawnGroup, Direct Array Mode=::NewEnumerator1, Vector Selection Array=$Emitter.StartBoneOffsetArray)
    Skeletal Mesh=$User.SkeletalMesh
    Use Sampled Mesh Attribute as Ribbon ID=false
   )
   &FN_GooCalculateGoalPosition(
    Axis:Vector3f=$Emitter.RibbonAxis
    Scale:Vector3f=&Multiply_VectorByFloat(
     Vector=&Multiply_VectorByFloat(
      Vector=&VectorFromFloat(Value=&Multiply_Float(A=$Emitter.RibbonLength, B=0))
      Float=&Lerp_Float(A=$Emitter.RibbonLengthRandMin, B=$Emitter.RibbonLengthRandMax, Alpha=$Particles.RibbonRandom)
     )
     Float=$Particles.RibbonLinkOrder
    )
    StartPosition:Vector3f=$Particles.SkeletalMeshLocation.SampledPosition
   ) {module=FN_CalculateGoalPosition}
   /SetVariables(Particles.Position=$Particles.GoalPosition) {isSetParameters=true, module=SetVariables_#5}
   /SetVariables(Particles.PreviousRotation:Quat4f=$Engine.Owner.Rotation) {isSetParameters=true, module=SetVariables_#6}
   &DynamicMaterialParameters(
    Index 0 Param 1=&Multiply_Float(A=$Particles.RibbonRandom, B=10)
    Module.Param0WriteEnabled=true
    Module.Param1WriteEnabled=true
    Module.Param2WriteEnabled=true
    Module.Param3WriteEnabled=true
   )
   /SetVariables(Particles.SpriteSize=<10 10>) {isSetParameters=true, module=SetVariables_#7}
  ]
  particleUpdateScript [
   &ParticleState()
   &SkeletalMeshLocation(
    Bone / Socket Index=&SelectIntFromArray(Direct Array Index=$Particles.SpawnGroup, Direct Array Mode=::NewEnumerator1, Int Selection Array=$Emitter.StartBoneIndexArray)
    Bone Sampling Mode=::NewEnumerator4
    MaskSpawnGroup=false
    Orientation Sampling=::NewEnumerator1
    Position Sampling=::NewEnumerator2
    Sampled Position Offset=&SelectVectorFromArray(Direct Array Index=$Particles.SpawnGroup, Direct Array Mode=::NewEnumerator1, Vector Selection Array=$Emitter.StartBoneOffsetArray)
    Skeletal Mesh=$User.SkeletalMesh
    Use Sampled Mesh Attribute as Ribbon ID=false
   ) {module=SkeletalMeshLocation001}
   &SolveForcesAndVelocity(Clamp Velocity=false, Limit Acceleration=false)
   &FN_GooCalculateGoalPosition(
    Axis:Vector3f=&SlerpVector(
     Vector A:Vector3f=&SlerpVector(
      Vector A:Vector3f=<1 0 0>
      Vector B:Vector3f=&TransformVector(Vector=$User.WorldSpaceWindVector, Source Space=::World, Destination Space=::Local)
      Alpha:NiagaraFloat=$User.WorldSpaceWindVectorFactor
      Scale:NiagaraFloat=1
     )
     Vector B:Vector3f=<0 0 1>
     Alpha:NiagaraFloat=&FloatFromCurve(
      FloatCurve=DataInterface() {propertyValues {Curve {keys [{tangentMode=RCTM_User, time=0, value=0}, {tangentMode=RCTM_User, time=1, value=1}]}}}
      CurveIndex=$Particles.RibbonLinkOrder
      Scale Curve=0.7
     )
     Scale:NiagaraFloat=1
    )
    CustomOrientation:Quat4f=$Engine.Owner.Rotation
    Scale:Vector3f=&Multiply_VectorByFloat(
     Vector=&Multiply_VectorByFloat(
      Vector=&VectorFromFloat(Value=&Multiply_Float(
        A=$Emitter.RibbonLength
        B=&FloatFromCurve(FloatCurve=DataInterface() {propertyValues {Curve {keys [
             {arriveTangent=17.1487, interpMode=RCIM_Cubic, leaveTangent=17.1487, tangentMode=RCTM_User, time=0, value=1}
             {interpMode=RCIM_Cubic, tangentMode=RCTM_User, time=0.2, value=1.5}
             <at 0.25 1 RCIM_Cubic>
             {arriveTangent=-3.4902, interpMode=RCIM_Cubic, leaveTangent=-3.4902, tangentMode=RCTM_User, time=1, value=0}
            ]}}})
       ))
      Float=&Lerp_Float(A=$Emitter.RibbonLengthRandMin, B=$Emitter.RibbonLengthRandMax, Alpha=$Particles.RibbonRandom)
     )
     Float=$Particles.RibbonLinkOrder
    )
    StartPosition:Vector3f=$Particles.SkeletalMeshLocation001.SampledPosition
    Use Quaternion Rotation=true
   ) {module=FN_CalculateGoalPosition001}
   &FN_GooApproachGoalPosition(
    AttractionForce:NiagaraFloat=&Lerp_Float(A=$Emitter.ResolveSpeedLow, B=$Emitter.ResolveSpeedHigh, Alpha=$Particles.RibbonRandom)
    GoalPosition:NiagaraPosition=$Particles.GoalPosition
    MaximumDistance:NiagaraFloat=&Multiply_Float(
     A=&FloatFromCurve(
      FloatCurve=DataInterface() {propertyValues {Curve {keys [
          {arriveTangent=4.1785, interpMode=RCIM_Cubic, time=0, value=0}
          {arriveTangent=2.3315, interpMode=RCIM_Cubic, leaveTangent=2.3315, tangentMode=RCTM_User, time=1, value=1}
         ]}}}
      CurveIndex=$Particles.RibbonLinkOrder
      Scale Curve=&Multiply_Float(
       A=$Emitter.MaxDistance
       B=&FloatFromCurve(FloatCurve=DataInterface() {propertyValues {Curve {keys [<at 0.75 1>,<at 1 0>]}}})
      )
     )
     B=&Lerp_Float(A=$Emitter.MaxDistance_RandFactorLow, B=$Emitter.MaxDistance_RandFactorHigh, Alpha=$Particles.RibbonRandom)
    )
    RibbonAxis:Vector3f=$Particles.RibbonAxis
   )
   /SetVariables(Particles.PreviousRotation:Quat4f=$Engine.Owner.Rotation) {isSetParameters=true, module=SetVariables_#8}
   &ScaleColor(
    Scale Alpha=&FloatFromCurve(FloatCurve=DataInterface() {propertyValues {Curve {keys [<at 0 0>,<at 0.5 1>,<at 0.7 1>,<at 1 0>]}}})
    Scale RGB=&VectorFromFloat(Value=&FloatFromCurve(FloatCurve=DataInterface() {propertyValues {Curve {keys [{arriveTangent=-319.691, interpMode=RCIM_Cubic, leaveTangent=-319.691, tangentMode=RCTM_User, time=0, value=800}, <at 0.4 15>, <at 1 1>]}}}))
    ScaleA=true
    ScaleRGB=true
   )
   &DynamicMaterialParameters(
    Index 0 Param 1=&Multiply_Float(A=$Particles.RibbonRandom, B=10)
    Index 0 Param 2=$User.EnergyFillFactor
    Module.Param0WriteEnabled=false
    Module.Param1WriteEnabled=true
    Module.Param2WriteEnabled=false
    Module.Param3WriteEnabled=false
   ) {module=DynamicMaterialParameters001}
  ]
  rendererClasses [&NiagaraRibbonRendererProperties]
  renderers [{rendererClass=&NiagaraRibbonRendererProperties}]
 }
 Left_Arm001 {
  emitterSpawnScript [&FN_GooGlobalProperties(
    EndBoneIndexArray=DataInterface() {propertyValues {GpuSyncMode=SyncCpuToGpu, MaxElements=0, IntData [0]}}
    EndBoneOffsetArray=DataInterface() {propertyValues {GpuSyncMode=SyncCpuToGpu, MaxElements=0, FloatData [<0 0 0>]}}
    MaxDistance:NiagaraFloat=$User.ArmRibbonLength
    MaxDistanceCompressed=10
    MaxDistance_RandFactorHigh=1
    MaxDistance_RandFactorLow=1
    NumberOfRibbons=2
    ParticlesPerRibbon=2
    RibbonLength:NiagaraFloat=$User.ArmRibbonLength
    RibbonLengthRandHigh=1
    RibbonLengthRandLow=1
    StartBoneIndexArray=DataInterface() {propertyValues {GpuSyncMode=SyncCpuToGpu, MaxElements=0, IntData [0,1]}}
    StartBoneOffsetArray=DataInterface() {propertyValues {GpuSyncMode=SyncCpuToGpu, MaxElements=0, FloatData [<10 0 0>, <3 0 0>]}}
   )]
  emitterUpdateScript [
   &EmitterState()
   &SpawnBurst_Instantaneous(
    Spawn Count=&Multiply_Int(A=$Emitter.ParticlesPerRibbon, B=$Emitter.NumberOfRibbons)
    Spawn Probability=$User.SpawnJacketFX
    Use Spawn Probability=true
    UseLoopCountLimit=false
   )
  ]
  particleSpawnScript [
   /SetVariables(Particles.SpawnGroup=&Divide_Int(A:NiagaraInt32=&ReturnExecIndex(), B:NiagaraInt32=$Emitter.ParticlesPerRibbon)) {isSetParameters=true, module=SetVariables_#1}
   /SetVariables(
    Particles.RibbonID=&MakeRibbonIDFromSpawnGroup()
    Particles.RibbonLinkOrder:NiagaraFloat=&Divide_Float(
     A=&MakeFloatFromInt(Int=&ModuloInt(A=&ReturnExecIndex(), B=$Emitter.ParticlesPerRibbon))
     B=&Subtract_Float(A=&MakeFloatFromInt(Int=$Emitter.ParticlesPerRibbon), B=1)
    )
   ) {isSetParameters=true, module=SetVariables_#2}
   /SetVariables(Particles.RibbonRandom:NiagaraFloat=&RandomRangeFloat(
     Minimum=0
     Maximum=1
     Randomness Mode=::NewEnumerator1
     Random Seed=&Add_Integer(A=$Engine.System.RandomSeed, B=$Particles.SpawnGroup)
     Fixed Random Seed=true
    )) {isSetParameters=true, module=SetVariables_#3}
   &InitializeRibbon(
    Color=&MultiplyLinearColor(
     Linear Color=$User.BariumPaintColor
     Scale Factor=&MakeVector4_FromLinearColor(LinearColor=&ColorFromCurve(
       ColorCurve=DataInterface() {propertyValues {
         AlphaCurve {keys [<at 0.574 1 RCIM_Cubic>,<at 0.9977 0>]}
         BlueCurve {keys [<at 0 1 RCIM_Cubic>]}
         GreenCurve {keys [<at 0 1 RCIM_Cubic>]}
         RedCurve {keys [<at 0 1 RCIM_Cubic>]}
        }}
       CurveIndex=$Particles.RibbonLinkOrder
      ))
    )
    Lifetime=&Lerp_Float(A=2, B=3, Alpha=$Particles.RibbonRandom)
    Ribbon Width=&FloatFromCurve(
     FloatCurve=DataInterface() {propertyValues {Curve {keys [<at 0 1>,<at 0.5 0.5>,<at 1 1>]}}}
     CurveIndex=$Particles.RibbonLinkOrder
     Scale Curve=$User.ArmRibbonWidth
    )
    UsePositionOffset=false
    Write Material Random=true
   )
   /SetVariables(Particles.MaterialRandom:NiagaraFloat=$Particles.RibbonRandom) {isSetParameters=true, module=SetVariables_#4}
   &SkeletalMeshLocation(
    Bone / Socket Index=&SelectIntFromArray(Direct Array Index=$Particles.SpawnGroup, Direct Array Mode=::NewEnumerator1, Int Selection Array=$Emitter.StartBoneIndexArray)
    Bone Sampling Mode=::NewEnumerator4
    MaskSpawnGroup=false
    Orientation Sampling=::NewEnumerator1
    Position Sampling=::NewEnumerator2
    Sampled Position Offset=&SelectVectorFromArray(Direct Array Index=$Particles.SpawnGroup, Direct Array Mode=::NewEnumerator1, Vector Selection Array=$Emitter.StartBoneOffsetArray)
    Skeletal Mesh=$User.SkeletalMesh
    Use Sampled Mesh Attribute as Ribbon ID=false
   )
   &FN_GooCalculateGoalPosition(
    Axis:Vector3f=$Emitter.RibbonAxis
    Scale:Vector3f=&Multiply_VectorByFloat(
     Vector=&Multiply_VectorByFloat(
      Vector=&VectorFromFloat(Value=&Multiply_Float(A=$Emitter.RibbonLength, B=0))
      Float=&Lerp_Float(A=$Emitter.RibbonLengthRandMin, B=$Emitter.RibbonLengthRandMax, Alpha=$Particles.RibbonRandom)
     )
     Float=$Particles.RibbonLinkOrder
    )
    StartPosition:Vector3f=$Particles.SkeletalMeshLocation.SampledPosition
   ) {module=FN_CalculateGoalPosition}
   /SetVariables(Particles.Position=$Particles.GoalPosition) {isSetParameters=true, module=SetVariables_#5}
   /SetVariables(Particles.PreviousRotation:Quat4f=$Engine.Owner.Rotation) {isSetParameters=true, module=SetVariables_#6}
   &DynamicMaterialParameters(
    Index 0 Param 1=&Multiply_Float(A=$Particles.RibbonRandom, B=10)
    Module.Param0WriteEnabled=true
    Module.Param1WriteEnabled=true
    Module.Param2WriteEnabled=true
    Module.Param3WriteEnabled=true
   )
   /SetVariables(Particles.SpriteSize=<10 10>) {isSetParameters=true, module=SetVariables_#7}
  ]
  particleUpdateScript [
   &ParticleState()
   &SkeletalMeshLocation(
    Bone / Socket Index=&SelectIntFromArray(Direct Array Index=$Particles.SpawnGroup, Direct Array Mode=::NewEnumerator1, Int Selection Array=$Emitter.StartBoneIndexArray)
    Bone Sampling Mode=::NewEnumerator4
    MaskSpawnGroup=false
    Orientation Sampling=::NewEnumerator1
    Position Sampling=::NewEnumerator2
    Sampled Position Offset=&SelectVectorFromArray(Direct Array Index=$Particles.SpawnGroup, Direct Array Mode=::NewEnumerator1, Vector Selection Array=$Emitter.StartBoneOffsetArray)
    Skeletal Mesh=$User.SkeletalMesh
    Use Sampled Mesh Attribute as Ribbon ID=false
   ) {module=SkeletalMeshLocation001}
   &SolveForcesAndVelocity(Clamp Velocity=false, Limit Acceleration=false)
   &FN_GooCalculateGoalPosition(
    Axis:Vector3f=&SlerpVector(
     Vector A:Vector3f=&SlerpVector(
      Vector A:Vector3f=<1 0 0>
      Vector B:Vector3f=&TransformVector(Vector=$User.WorldSpaceWindVector, Source Space=::World, Destination Space=::Local)
      Alpha:NiagaraFloat=$User.WorldSpaceWindVectorFactor
      Scale:NiagaraFloat=1
     )
     Vector B:Vector3f=<0 0 1>
     Alpha:NiagaraFloat=&FloatFromCurve(
      FloatCurve=DataInterface() {propertyValues {Curve {keys [{tangentMode=RCTM_User, time=0, value=0}, {tangentMode=RCTM_User, time=1, value=1}]}}}
      CurveIndex=$Particles.RibbonLinkOrder
      Scale Curve=0.7
     )
     Scale:NiagaraFloat=1
    )
    CustomOrientation:Quat4f=$Engine.Owner.Rotation
    Scale:Vector3f=&Multiply_VectorByFloat(
     Vector=&Multiply_VectorByFloat(
      Vector=&VectorFromFloat(Value=&Multiply_Float(
        A=$Emitter.RibbonLength
        B=&FloatFromCurve(FloatCurve=DataInterface() {propertyValues {Curve {keys [
             {arriveTangent=17.1487, interpMode=RCIM_Cubic, leaveTangent=17.1487, tangentMode=RCTM_User, time=0, value=1}
             {interpMode=RCIM_Cubic, tangentMode=RCTM_User, time=0.2, value=1.5}
             <at 0.25 1 RCIM_Cubic>
             {arriveTangent=-3.4902, interpMode=RCIM_Cubic, leaveTangent=-3.4902, tangentMode=RCTM_User, time=1, value=0}
            ]}}})
       ))
      Float=&Lerp_Float(A=$Emitter.RibbonLengthRandMin, B=$Emitter.RibbonLengthRandMax, Alpha=$Particles.RibbonRandom)
     )
     Float=$Particles.RibbonLinkOrder
    )
    StartPosition:Vector3f=$Particles.SkeletalMeshLocation001.SampledPosition
    Use Quaternion Rotation=true
   ) {module=FN_CalculateGoalPosition001}
   &FN_GooApproachGoalPosition(
    AttractionForce:NiagaraFloat=&Lerp_Float(A=$Emitter.ResolveSpeedLow, B=$Emitter.ResolveSpeedHigh, Alpha=$Particles.RibbonRandom)
    GoalPosition:NiagaraPosition=$Particles.GoalPosition
    MaximumDistance:NiagaraFloat=&Multiply_Float(
     A=&FloatFromCurve(
      FloatCurve=DataInterface() {propertyValues {Curve {keys [
          {arriveTangent=4.1785, interpMode=RCIM_Cubic, time=0, value=0}
          {arriveTangent=2.3315, interpMode=RCIM_Cubic, leaveTangent=2.3315, tangentMode=RCTM_User, time=1, value=1}
         ]}}}
      CurveIndex=$Particles.RibbonLinkOrder
      Scale Curve=&Multiply_Float(
       A=$Emitter.MaxDistance
       B=&FloatFromCurve(FloatCurve=DataInterface() {propertyValues {Curve {keys [<at 0.75 1>,<at 1 0>]}}})
      )
     )
     B=&Lerp_Float(A=$Emitter.MaxDistance_RandFactorLow, B=$Emitter.MaxDistance_RandFactorHigh, Alpha=$Particles.RibbonRandom)
    )
    RibbonAxis:Vector3f=$Particles.RibbonAxis
   )
   /SetVariables(Particles.PreviousRotation:Quat4f=$Engine.Owner.Rotation) {isSetParameters=true, module=SetVariables_#8}
   &ScaleColor(
    Scale Alpha=&FloatFromCurve(FloatCurve=DataInterface() {propertyValues {Curve {keys [<at 0 0>,<at 0.5 1>,<at 0.7 1>,<at 1 0>]}}})
    Scale RGB=&VectorFromFloat(Value=&FloatFromCurve(FloatCurve=DataInterface() {propertyValues {Curve {keys [{arriveTangent=-319.691, interpMode=RCIM_Cubic, leaveTangent=-319.691, tangentMode=RCTM_User, time=0, value=800}, <at 0.4 15>, <at 1 1>]}}}))
    ScaleA=true
    ScaleRGB=true
   )
   &DynamicMaterialParameters(
    Index 0 Param 1=&Multiply_Float(A=$Particles.RibbonRandom, B=10)
    Index 0 Param 2=$User.EnergyFillFactor
    Module.Param0WriteEnabled=false
    Module.Param1WriteEnabled=true
    Module.Param2WriteEnabled=false
    Module.Param3WriteEnabled=false
   ) {module=DynamicMaterialParameters001}
  ]
  rendererClasses [&NiagaraRibbonRendererProperties]
  renderers [{rendererClass=&NiagaraRibbonRendererProperties}]
 }
 Left_Hand {
  emitterSpawnScript [&FN_GooGlobalProperties(
    EndBoneIndexArray=DataInterface() {propertyValues {GpuSyncMode=SyncCpuToGpu, MaxElements=0, IntData [0]}}
    EndBoneOffsetArray=DataInterface() {propertyValues {GpuSyncMode=SyncCpuToGpu, MaxElements=0, FloatData [<0 0 0>]}}
    MaxDistance:NiagaraFloat=$User.ArmRibbonLength
    MaxDistanceCompressed=10
    MaxDistance_RandFactorHigh=1
    MaxDistance_RandFactorLow=1
    RibbonLength:NiagaraFloat=$User.ArmRibbonLength
    RibbonLengthRandHigh=1
    RibbonLengthRandLow=1
    StartBoneIndexArray=DataInterface() {propertyValues {GpuSyncMode=SyncCpuToGpu, MaxElements=0, IntData [2]}}
    StartBoneOffsetArray=DataInterface() {propertyValues {GpuSyncMode=SyncCpuToGpu, MaxElements=0, FloatData [<3 0 0>]}}
   )]
  emitterUpdateScript [
   &EmitterState()
   &SpawnBurst_Instantaneous(
    Spawn Count=&Multiply_Int(A=$Emitter.ParticlesPerRibbon, B=$Emitter.NumberOfRibbons)
    Use Spawn Probability=false
    UseLoopCountLimit=false
   )
  ]
  particleSpawnScript [
   /SetVariables(Particles.SpawnGroup=&Divide_Int(A:NiagaraInt32=&ReturnExecIndex(), B:NiagaraInt32=$Emitter.ParticlesPerRibbon)) {isSetParameters=true, module=SetVariables_#1}
   /SetVariables(
    Particles.RibbonID=&MakeRibbonIDFromSpawnGroup()
    Particles.RibbonLinkOrder:NiagaraFloat=&Divide_Float(
     A=&MakeFloatFromInt(Int=&ModuloInt(A=&ReturnExecIndex(), B=$Emitter.ParticlesPerRibbon))
     B=&Subtract_Float(A=&MakeFloatFromInt(Int=$Emitter.ParticlesPerRibbon), B=1)
    )
   ) {isSetParameters=true, module=SetVariables_#2}
   /SetVariables(Particles.RibbonRandom:NiagaraFloat=&RandomRangeFloat(
     Minimum=0
     Maximum=1
     Randomness Mode=::NewEnumerator1
     Random Seed=&Add_Integer(A=$Engine.System.RandomSeed, B=$Particles.SpawnGroup)
     Fixed Random Seed=true
    )) {isSetParameters=true, module=SetVariables_#3}
   &InitializeRibbon(
    Color=&MultiplyLinearColor(
     Linear Color=$User.BariumPaintColor
     Scale Factor=&MakeVector4_FromLinearColor(LinearColor=&ColorFromCurve(
       ColorCurve=DataInterface() {propertyValues {
         AlphaCurve {keys [<at 0.574 1 RCIM_Cubic>,<at 0.9977 0>]}
         BlueCurve {keys [<at 0 1 RCIM_Cubic>]}
         GreenCurve {keys [<at 0 1 RCIM_Cubic>]}
         RedCurve {keys [<at 0 1 RCIM_Cubic>]}
        }}
       CurveIndex=$Particles.RibbonLinkOrder
      ))
    )
    Lifetime=&Lerp_Float(A=2, B=3, Alpha=$Particles.RibbonRandom)
    Ribbon Width=&FloatFromCurve(
     FloatCurve=DataInterface() {propertyValues {Curve {keys [<at 0 1>,<at 0.5 0.5>,<at 1 1>]}}}
     CurveIndex=$Particles.RibbonLinkOrder
     Scale Curve=15
    )
    UsePositionOffset=false
    Write Material Random=true
   )
   /SetVariables(Particles.MaterialRandom:NiagaraFloat=$Particles.RibbonRandom) {isSetParameters=true, module=SetVariables_#4}
   &SkeletalMeshLocation(
    Bone / Socket Index=&SelectIntFromArray(Direct Array Index=$Particles.SpawnGroup, Direct Array Mode=::NewEnumerator1, Int Selection Array=$Emitter.StartBoneIndexArray)
    Bone Sampling Mode=::NewEnumerator4
    MaskSpawnGroup=false
    Orientation Sampling=::NewEnumerator1
    Position Sampling=::NewEnumerator2
    Sampled Position Offset=&SelectVectorFromArray(Direct Array Index=$Particles.SpawnGroup, Direct Array Mode=::NewEnumerator1, Vector Selection Array=$Emitter.StartBoneOffsetArray)
    Skeletal Mesh=$User.SkeletalMesh
    Use Sampled Mesh Attribute as Ribbon ID=false
   )
   &FN_GooCalculateGoalPosition(
    Axis:Vector3f=$Emitter.RibbonAxis
    Scale:Vector3f=&Multiply_VectorByFloat(
     Vector=&Multiply_VectorByFloat(
      Vector=&VectorFromFloat(Value=&Multiply_Float(A=$Emitter.RibbonLength, B=0))
      Float=&Lerp_Float(A=$Emitter.RibbonLengthRandMin, B=$Emitter.RibbonLengthRandMax, Alpha=$Particles.RibbonRandom)
     )
     Float=$Particles.RibbonLinkOrder
    )
    StartPosition:Vector3f=$Particles.SkeletalMeshLocation.SampledPosition
   ) {module=FN_CalculateGoalPosition}
   /SetVariables(Particles.Position=$Particles.GoalPosition) {isSetParameters=true, module=SetVariables_#5}
   /SetVariables(Particles.PreviousRotation:Quat4f=$Engine.Owner.Rotation) {isSetParameters=true, module=SetVariables_#6}
   &DynamicMaterialParameters(
    Index 0 Param 1=&Multiply_Float(A=$Particles.RibbonRandom, B=10)
    Module.Param0WriteEnabled=true
    Module.Param1WriteEnabled=true
    Module.Param2WriteEnabled=true
    Module.Param3WriteEnabled=true
   )
   /SetVariables(Particles.SpriteSize=<10 10>) {isSetParameters=true, module=SetVariables_#7}
  ]
  particleUpdateScript [
   &ParticleState()
   &SkeletalMeshLocation(
    Bone / Socket Index=&SelectIntFromArray(Direct Array Index=$Particles.SpawnGroup, Direct Array Mode=::NewEnumerator1, Int Selection Array=$Emitter.StartBoneIndexArray)
    Bone Sampling Mode=::NewEnumerator4
    MaskSpawnGroup=false
    Orientation Sampling=::NewEnumerator1
    Position Sampling=::NewEnumerator2
    Sampled Position Offset=&SelectVectorFromArray(Direct Array Index=$Particles.SpawnGroup, Direct Array Mode=::NewEnumerator1, Vector Selection Array=$Emitter.StartBoneOffsetArray)
    Skeletal Mesh=$User.SkeletalMesh
    Use Sampled Mesh Attribute as Ribbon ID=false
   ) {module=SkeletalMeshLocation001}
   &SolveForcesAndVelocity(Clamp Velocity=false, Limit Acceleration=false)
   &FN_GooCalculateGoalPosition(
    Axis:Vector3f=&SlerpVector(
     Vector A:Vector3f=&SlerpVector(
      Vector A:Vector3f=<1 0 0>
      Vector B:Vector3f=&TransformVector(Vector=$User.WorldSpaceWindVector, Source Space=::World, Destination Space=::Local)
      Alpha:NiagaraFloat=$User.WorldSpaceWindVectorFactor
      Scale:NiagaraFloat=1
     )
     Vector B:Vector3f=<0 0 1>
     Alpha:NiagaraFloat=&FloatFromCurve(
      FloatCurve=DataInterface() {propertyValues {Curve {keys [{tangentMode=RCTM_User, time=0, value=0}, {tangentMode=RCTM_User, time=1, value=1}]}}}
      CurveIndex=$Particles.RibbonLinkOrder
      Scale Curve=0.7
     )
     Scale:NiagaraFloat=1
    )
    CustomOrientation:Quat4f=$Engine.Owner.Rotation
    Scale:Vector3f=&Multiply_VectorByFloat(
     Vector=&Multiply_VectorByFloat(
      Vector=&VectorFromFloat(Value=&Multiply_Float(
        A=$Emitter.RibbonLength
        B=&FloatFromCurve(FloatCurve=DataInterface() {propertyValues {Curve {keys [
             {arriveTangent=17.1487, interpMode=RCIM_Cubic, leaveTangent=17.1487, tangentMode=RCTM_User, time=0, value=1}
             {interpMode=RCIM_Cubic, tangentMode=RCTM_User, time=0.2, value=1.5}
             <at 0.25 1 RCIM_Cubic>
             {arriveTangent=-3.4902, interpMode=RCIM_Cubic, leaveTangent=-3.4902, tangentMode=RCTM_User, time=1, value=0}
            ]}}})
       ))
      Float=&Lerp_Float(A=$Emitter.RibbonLengthRandMin, B=$Emitter.RibbonLengthRandMax, Alpha=$Particles.RibbonRandom)
     )
     Float=$Particles.RibbonLinkOrder
    )
    StartPosition:Vector3f=$Particles.SkeletalMeshLocation001.SampledPosition
    Use Quaternion Rotation=true
   ) {module=FN_CalculateGoalPosition001}
   &FN_GooApproachGoalPosition(
    AttractionForce:NiagaraFloat=&Lerp_Float(A=$Emitter.ResolveSpeedLow, B=$Emitter.ResolveSpeedHigh, Alpha=$Particles.RibbonRandom)
    GoalPosition:NiagaraPosition=$Particles.GoalPosition
    MaximumDistance:NiagaraFloat=&Multiply_Float(
     A=&FloatFromCurve(
      FloatCurve=DataInterface() {propertyValues {Curve {keys [
          {arriveTangent=4.1785, interpMode=RCIM_Cubic, time=0, value=0}
          {arriveTangent=2.3315, interpMode=RCIM_Cubic, leaveTangent=2.3315, tangentMode=RCTM_User, time=1, value=1}
         ]}}}
      CurveIndex=$Particles.RibbonLinkOrder
      Scale Curve=&Multiply_Float(
       A=$Emitter.MaxDistance
       B=&FloatFromCurve(FloatCurve=DataInterface() {propertyValues {Curve {keys [<at 0.75 1>,<at 1 0>]}}})
      )
     )
     B=&Lerp_Float(A=$Emitter.MaxDistance_RandFactorLow, B=$Emitter.MaxDistance_RandFactorHigh, Alpha=$Particles.RibbonRandom)
    )
    RibbonAxis:Vector3f=$Particles.RibbonAxis
   )
   /SetVariables(Particles.PreviousRotation:Quat4f=$Engine.Owner.Rotation) {isSetParameters=true, module=SetVariables_#8}
   &ScaleColor(
    Scale Alpha=&FloatFromCurve(FloatCurve=DataInterface() {propertyValues {Curve {keys [<at 0 0>,<at 0.5 1>,<at 0.7 1>,<at 1 0>]}}})
    Scale RGB=&VectorFromFloat(Value=&FloatFromCurve(FloatCurve=DataInterface() {propertyValues {Curve {keys [{arriveTangent=-319.691, interpMode=RCIM_Cubic, leaveTangent=-319.691, tangentMode=RCTM_User, time=0, value=800}, <at 0.4 15>, <at 1 1>]}}}))
    ScaleA=true
    ScaleRGB=true
   )
   &DynamicMaterialParameters(
    Index 0 Param 1=&Multiply_Float(A=$Particles.RibbonRandom, B=10)
    Index 0 Param 2=$User.EnergyFillFactor
    Module.Param0WriteEnabled=false
    Module.Param1WriteEnabled=true
    Module.Param2WriteEnabled=false
    Module.Param3WriteEnabled=false
   ) {module=DynamicMaterialParameters001}
  ]
  rendererClasses [&NiagaraRibbonRendererProperties]
  renderers [{rendererClass=&NiagaraRibbonRendererProperties}]
 }
 Left_Hand001 {
  emitterSpawnScript [&FN_GooGlobalProperties(
    EndBoneIndexArray=DataInterface() {propertyValues {GpuSyncMode=SyncCpuToGpu, MaxElements=0, IntData [0]}}
    EndBoneOffsetArray=DataInterface() {propertyValues {GpuSyncMode=SyncCpuToGpu, MaxElements=0, FloatData [<0 0 0>]}}
    MaxDistance:NiagaraFloat=$User.ArmRibbonLength
    MaxDistanceCompressed=10
    MaxDistance_RandFactorHigh=1
    MaxDistance_RandFactorLow=1
    ParticlesPerRibbon=2
    RibbonLength:NiagaraFloat=$User.ArmRibbonLength
    RibbonLengthRandHigh=1
    RibbonLengthRandLow=1
    StartBoneIndexArray=DataInterface() {propertyValues {GpuSyncMode=SyncCpuToGpu, MaxElements=0, IntData [2]}}
    StartBoneOffsetArray=DataInterface() {propertyValues {GpuSyncMode=SyncCpuToGpu, MaxElements=0, FloatData [<3 0 0>]}}
   )]
  emitterUpdateScript [
   &EmitterState()
   &SpawnBurst_Instantaneous(
    Spawn Count=&Multiply_Int(A=$Emitter.ParticlesPerRibbon, B=$Emitter.NumberOfRibbons)
    Use Spawn Probability=false
    UseLoopCountLimit=false
   )
  ]
  particleSpawnScript [
   /SetVariables(Particles.SpawnGroup=&Divide_Int(A:NiagaraInt32=&ReturnExecIndex(), B:NiagaraInt32=$Emitter.ParticlesPerRibbon)) {isSetParameters=true, module=SetVariables_#1}
   /SetVariables(
    Particles.RibbonID=&MakeRibbonIDFromSpawnGroup()
    Particles.RibbonLinkOrder:NiagaraFloat=&Divide_Float(
     A=&MakeFloatFromInt(Int=&ModuloInt(A=&ReturnExecIndex(), B=$Emitter.ParticlesPerRibbon))
     B=&Subtract_Float(A=&MakeFloatFromInt(Int=$Emitter.ParticlesPerRibbon), B=1)
    )
   ) {isSetParameters=true, module=SetVariables_#2}
   /SetVariables(Particles.RibbonRandom:NiagaraFloat=&RandomRangeFloat(
     Minimum=0
     Maximum=1
     Randomness Mode=::NewEnumerator1
     Random Seed=&Add_Integer(A=$Engine.System.RandomSeed, B=$Particles.SpawnGroup)
     Fixed Random Seed=true
    )) {isSetParameters=true, module=SetVariables_#3}
   &InitializeRibbon(
    Color=&MultiplyLinearColor(
     Linear Color=$User.BariumPaintColor
     Scale Factor=&MakeVector4_FromLinearColor(LinearColor=&ColorFromCurve(
       ColorCurve=DataInterface() {propertyValues {
         AlphaCurve {keys [<at 0.574 1 RCIM_Cubic>,<at 0.9977 0>]}
         BlueCurve {keys [<at 0 1 RCIM_Cubic>]}
         GreenCurve {keys [<at 0 1 RCIM_Cubic>]}
         RedCurve {keys [<at 0 1 RCIM_Cubic>]}
        }}
       CurveIndex=$Particles.RibbonLinkOrder
      ))
    )
    Lifetime=&Lerp_Float(A=2, B=3, Alpha=$Particles.RibbonRandom)
    Ribbon Width=&FloatFromCurve(
     FloatCurve=DataInterface() {propertyValues {Curve {keys [<at 0 1>,<at 0.5 0.5>,<at 1 1>]}}}
     CurveIndex=$Particles.RibbonLinkOrder
     Scale Curve=15
    )
    UsePositionOffset=false
    Write Material Random=true
   )
   /SetVariables(Particles.MaterialRandom:NiagaraFloat=$Particles.RibbonRandom) {isSetParameters=true, module=SetVariables_#4}
   &SkeletalMeshLocation(
    Bone / Socket Index=&SelectIntFromArray(Direct Array Index=$Particles.SpawnGroup, Direct Array Mode=::NewEnumerator1, Int Selection Array=$Emitter.StartBoneIndexArray)
    Bone Sampling Mode=::NewEnumerator4
    MaskSpawnGroup=false
    Orientation Sampling=::NewEnumerator1
    Position Sampling=::NewEnumerator2
    Sampled Position Offset=&SelectVectorFromArray(Direct Array Index=$Particles.SpawnGroup, Direct Array Mode=::NewEnumerator1, Vector Selection Array=$Emitter.StartBoneOffsetArray)
    Skeletal Mesh=$User.SkeletalMesh
    Use Sampled Mesh Attribute as Ribbon ID=false
   )
   &FN_GooCalculateGoalPosition(
    Axis:Vector3f=$Emitter.RibbonAxis
    Scale:Vector3f=&Multiply_VectorByFloat(
     Vector=&Multiply_VectorByFloat(
      Vector=&VectorFromFloat(Value=&Multiply_Float(A=$Emitter.RibbonLength, B=0))
      Float=&Lerp_Float(A=$Emitter.RibbonLengthRandMin, B=$Emitter.RibbonLengthRandMax, Alpha=$Particles.RibbonRandom)
     )
     Float=$Particles.RibbonLinkOrder
    )
    StartPosition:Vector3f=$Particles.SkeletalMeshLocation.SampledPosition
   ) {module=FN_CalculateGoalPosition}
   /SetVariables(Particles.Position=$Particles.GoalPosition) {isSetParameters=true, module=SetVariables_#5}
   /SetVariables(Particles.PreviousRotation:Quat4f=$Engine.Owner.Rotation) {isSetParameters=true, module=SetVariables_#6}
   &DynamicMaterialParameters(
    Index 0 Param 1=&Multiply_Float(A=$Particles.RibbonRandom, B=10)
    Module.Param0WriteEnabled=true
    Module.Param1WriteEnabled=true
    Module.Param2WriteEnabled=true
    Module.Param3WriteEnabled=true
   )
   /SetVariables(Particles.SpriteSize=<10 10>) {isSetParameters=true, module=SetVariables_#7}
  ]
  particleUpdateScript [
   &ParticleState()
   &SkeletalMeshLocation(
    Bone / Socket Index=&SelectIntFromArray(Direct Array Index=$Particles.SpawnGroup, Direct Array Mode=::NewEnumerator1, Int Selection Array=$Emitter.StartBoneIndexArray)
    Bone Sampling Mode=::NewEnumerator4
    MaskSpawnGroup=false
    Orientation Sampling=::NewEnumerator1
    Position Sampling=::NewEnumerator2
    Sampled Position Offset=&SelectVectorFromArray(Direct Array Index=$Particles.SpawnGroup, Direct Array Mode=::NewEnumerator1, Vector Selection Array=$Emitter.StartBoneOffsetArray)
    Skeletal Mesh=$User.SkeletalMesh
    Use Sampled Mesh Attribute as Ribbon ID=false
   ) {module=SkeletalMeshLocation001}
   &SolveForcesAndVelocity(Clamp Velocity=false, Limit Acceleration=false)
   &FN_GooCalculateGoalPosition(
    Axis:Vector3f=&SlerpVector(
     Vector A:Vector3f=&SlerpVector(
      Vector A:Vector3f=<1 0 0>
      Vector B:Vector3f=&TransformVector(Vector=$User.WorldSpaceWindVector, Source Space=::World, Destination Space=::Local)
      Alpha:NiagaraFloat=$User.WorldSpaceWindVectorFactor
      Scale:NiagaraFloat=1
     )
     Vector B:Vector3f=<0 0 1>
     Alpha:NiagaraFloat=&FloatFromCurve(
      FloatCurve=DataInterface() {propertyValues {Curve {keys [{tangentMode=RCTM_User, time=0, value=0}, {tangentMode=RCTM_User, time=1, value=1}]}}}
      CurveIndex=$Particles.RibbonLinkOrder
      Scale Curve=0.7
     )
     Scale:NiagaraFloat=1
    )
    CustomOrientation:Quat4f=$Engine.Owner.Rotation
    Scale:Vector3f=&Multiply_VectorByFloat(
     Vector=&Multiply_VectorByFloat(
      Vector=&VectorFromFloat(Value=&Multiply_Float(
        A=$Emitter.RibbonLength
        B=&FloatFromCurve(FloatCurve=DataInterface() {propertyValues {Curve {keys [
             {arriveTangent=17.1487, interpMode=RCIM_Cubic, leaveTangent=17.1487, tangentMode=RCTM_User, time=0, value=1}
             {interpMode=RCIM_Cubic, tangentMode=RCTM_User, time=0.2, value=1.5}
             <at 0.25 1 RCIM_Cubic>
             {arriveTangent=-3.4902, interpMode=RCIM_Cubic, leaveTangent=-3.4902, tangentMode=RCTM_User, time=1, value=0}
            ]}}})
       ))
      Float=&Lerp_Float(A=$Emitter.RibbonLengthRandMin, B=$Emitter.RibbonLengthRandMax, Alpha=$Particles.RibbonRandom)
     )
     Float=$Particles.RibbonLinkOrder
    )
    StartPosition:Vector3f=$Particles.SkeletalMeshLocation001.SampledPosition
    Use Quaternion Rotation=true
   ) {module=FN_CalculateGoalPosition001}
   &FN_GooApproachGoalPosition(
    AttractionForce:NiagaraFloat=&Lerp_Float(A=$Emitter.ResolveSpeedLow, B=$Emitter.ResolveSpeedHigh, Alpha=$Particles.RibbonRandom)
    GoalPosition:NiagaraPosition=$Particles.GoalPosition
    MaximumDistance:NiagaraFloat=&Multiply_Float(
     A=&FloatFromCurve(
      FloatCurve=DataInterface() {propertyValues {Curve {keys [
          {arriveTangent=4.1785, interpMode=RCIM_Cubic, time=0, value=0}
          {arriveTangent=2.3315, interpMode=RCIM_Cubic, leaveTangent=2.3315, tangentMode=RCTM_User, time=1, value=1}
         ]}}}
      CurveIndex=$Particles.RibbonLinkOrder
      Scale Curve=&Multiply_Float(
       A=$Emitter.MaxDistance
       B=&FloatFromCurve(FloatCurve=DataInterface() {propertyValues {Curve {keys [<at 0.75 1>,<at 1 0>]}}})
      )
     )
     B=&Lerp_Float(A=$Emitter.MaxDistance_RandFactorLow, B=$Emitter.MaxDistance_RandFactorHigh, Alpha=$Particles.RibbonRandom)
    )
    RibbonAxis:Vector3f=$Particles.RibbonAxis
   )
   /SetVariables(Particles.PreviousRotation:Quat4f=$Engine.Owner.Rotation) {isSetParameters=true, module=SetVariables_#8}
   &ScaleColor(
    Scale Alpha=&FloatFromCurve(FloatCurve=DataInterface() {propertyValues {Curve {keys [<at 0 0>,<at 0.5 1>,<at 0.7 1>,<at 1 0>]}}})
    Scale RGB=&VectorFromFloat(Value=&FloatFromCurve(FloatCurve=DataInterface() {propertyValues {Curve {keys [{arriveTangent=-319.691, interpMode=RCIM_Cubic, leaveTangent=-319.691, tangentMode=RCTM_User, time=0, value=800}, <at 0.4 15>, <at 1 1>]}}}))
    ScaleA=true
    ScaleRGB=true
   )
   &DynamicMaterialParameters(
    Index 0 Param 1=&Multiply_Float(A=$Particles.RibbonRandom, B=10)
    Index 0 Param 2=$User.EnergyFillFactor
    Module.Param0WriteEnabled=false
    Module.Param1WriteEnabled=true
    Module.Param2WriteEnabled=false
    Module.Param3WriteEnabled=false
   ) {module=DynamicMaterialParameters001}
  ]
  rendererClasses [&NiagaraRibbonRendererProperties]
  renderers [{rendererClass=&NiagaraRibbonRendererProperties}]
 }
 Left_Hip {
  emitterSpawnScript [&FN_GooGlobalProperties(
    EndBoneIndexArray=DataInterface() {propertyValues {GpuSyncMode=SyncCpuToGpu, MaxElements=0, IntData [0]}}
    EndBoneOffsetArray=DataInterface() {propertyValues {GpuSyncMode=SyncCpuToGpu, MaxElements=0, FloatData [<0 0 0>]}}
    MaxDistance:NiagaraFloat=$User.HipRibbonLength
    MaxDistanceCompressed=10
    MaxDistance_RandFactorHigh=1
    MaxDistance_RandFactorLow=1
    ParticlesPerRibbon=4
    ResolveSpeedHigh=3
    RibbonLength:NiagaraFloat=$User.HipRibbonLength
    RibbonLengthRandHigh=1
    RibbonLengthRandLow=1
    StartBoneIndexArray=DataInterface() {propertyValues {GpuSyncMode=SyncCpuToGpu, MaxElements=0, IntData [10]}}
    StartBoneOffsetArray=DataInterface() {propertyValues {GpuSyncMode=SyncCpuToGpu, MaxElements=0, FloatData [<-2 0 -3>]}}
   )]
  emitterUpdateScript [
   &EmitterState()
   &SpawnBurst_Instantaneous(
    Spawn Count=&Multiply_Int(A=$Emitter.ParticlesPerRibbon, B=$Emitter.NumberOfRibbons)
    Use Spawn Probability=false
    UseLoopCountLimit=false
   )
  ]
  particleSpawnScript [
   /SetVariables(Particles.SpawnGroup=&Divide_Int(A:NiagaraInt32=&ReturnExecIndex(), B:NiagaraInt32=$Emitter.ParticlesPerRibbon)) {isSetParameters=true, module=SetVariables_#1}
   /SetVariables(
    Particles.RibbonID=&MakeRibbonIDFromSpawnGroup()
    Particles.RibbonLinkOrder:NiagaraFloat=&Divide_Float(
     A=&MakeFloatFromInt(Int=&ModuloInt(A=&ReturnExecIndex(), B=$Emitter.ParticlesPerRibbon))
     B=&Subtract_Float(A=&MakeFloatFromInt(Int=$Emitter.ParticlesPerRibbon), B=1)
    )
   ) {isSetParameters=true, module=SetVariables_#2}
   /SetVariables(Particles.RibbonRandom:NiagaraFloat=&RandomRangeFloat(
     Minimum=0
     Maximum=1
     Randomness Mode=::NewEnumerator1
     Random Seed=&Add_Integer(A=$Engine.System.RandomSeed, B=$Particles.SpawnGroup)
     Fixed Random Seed=true
    )) {isSetParameters=true, module=SetVariables_#3}
   &InitializeRibbon(
    Color=&MultiplyLinearColor(
     Linear Color=$User.BariumPaintColor
     Scale Factor=&MakeVector4_FromLinearColor(LinearColor=&ColorFromCurve(
       ColorCurve=DataInterface() {propertyValues {
         AlphaCurve {keys [<at 0.574 1 RCIM_Cubic>,<at 0.9977 0>]}
         BlueCurve {keys [<at 0 1 RCIM_Cubic>,<at 1 1 RCIM_Cubic>]}
         GreenCurve {keys [<at 0 1 RCIM_Cubic>,<at 1 1 RCIM_Cubic>]}
         RedCurve {keys [<at 0 1 RCIM_Cubic>,<at 1 1 RCIM_Cubic>]}
        }}
       CurveIndex=$Particles.RibbonLinkOrder
      ))
    )
    Lifetime=&Lerp_Float(A=2, B=3, Alpha=$Particles.RibbonRandom)
    Ribbon Width=$User.HipRibbonWidth
    UsePositionOffset=false
    Write Material Random=true
   )
   /SetVariables(Particles.MaterialRandom:NiagaraFloat=$Particles.RibbonRandom) {isSetParameters=true, module=SetVariables_#4}
   &SkeletalMeshLocation(
    Bone / Socket Index=&SelectIntFromArray(Direct Array Index=$Particles.SpawnGroup, Direct Array Mode=::NewEnumerator1, Int Selection Array=$Emitter.StartBoneIndexArray)
    Bone Sampling Mode=::NewEnumerator4
    MaskSpawnGroup=false
    Orientation Sampling=::NewEnumerator1
    Position Sampling=::NewEnumerator2
    Sampled Position Offset=&SelectVectorFromArray(Direct Array Index=$Particles.SpawnGroup, Direct Array Mode=::NewEnumerator1, Vector Selection Array=$Emitter.StartBoneOffsetArray)
    Skeletal Mesh=$User.SkeletalMesh
    Use Sampled Mesh Attribute as Ribbon ID=false
   )
   &FN_GooCalculateGoalPosition(
    Axis:Vector3f=$Emitter.RibbonAxis
    Scale:Vector3f=&Multiply_VectorByFloat(
     Vector=&Multiply_VectorByFloat(
      Vector=&VectorFromFloat(Value=&Multiply_Float(A=$Emitter.RibbonLength, B=0))
      Float=&Lerp_Float(A=$Emitter.RibbonLengthRandMin, B=$Emitter.RibbonLengthRandMax, Alpha=$Particles.RibbonRandom)
     )
     Float=$Particles.RibbonLinkOrder
    )
    StartPosition:Vector3f=$Particles.SkeletalMeshLocation.SampledPosition
   ) {module=FN_CalculateGoalPosition}
   /SetVariables(Particles.Position=$Particles.GoalPosition) {isSetParameters=true, module=SetVariables_#5}
   /SetVariables(Particles.PreviousRotation:Quat4f=$Engine.Owner.Rotation) {isSetParameters=true, module=SetVariables_#6}
   &DynamicMaterialParameters(
    Index 0 Param 1=&Multiply_Float(A=$Particles.RibbonRandom, B=10)
    Module.Param0WriteEnabled=true
    Module.Param1WriteEnabled=true
    Module.Param2WriteEnabled=true
    Module.Param3WriteEnabled=true
   )
   /SetVariables(Particles.SpriteSize=<10 10>) {isSetParameters=true, module=SetVariables_#7}
  ]
  particleUpdateScript [
   &ParticleState()
   &SkeletalMeshLocation(
    Bone / Socket Index=&SelectIntFromArray(Direct Array Index=$Particles.SpawnGroup, Direct Array Mode=::NewEnumerator1, Int Selection Array=$Emitter.StartBoneIndexArray)
    Bone Sampling Mode=::NewEnumerator4
    MaskSpawnGroup=false
    Orientation Sampling=::NewEnumerator1
    Position Sampling=::NewEnumerator2
    Sampled Position Offset=&SelectVectorFromArray(Direct Array Index=$Particles.SpawnGroup, Direct Array Mode=::NewEnumerator1, Vector Selection Array=$Emitter.StartBoneOffsetArray)
    Skeletal Mesh=$User.SkeletalMesh
    Use Sampled Mesh Attribute as Ribbon ID=false
   ) {module=SkeletalMeshLocation001}
   &SolveForcesAndVelocity(Clamp Velocity=false, Limit Acceleration=false)
   &FN_GooCalculateGoalPosition(
    Axis:Vector3f=&SlerpVector(
     Vector A:Vector3f=&SlerpVector(
      Vector A:Vector3f=<2 -1 0>
      Vector B:Vector3f=&TransformVector(Vector=$User.WorldSpaceWindVector, Source Space=::World, Destination Space=::Local)
      Alpha:NiagaraFloat=$User.WorldSpaceWindVectorFactor
      Scale:NiagaraFloat=1
     )
     Vector B:Vector3f=<0 0 1>
     Alpha:NiagaraFloat=&FloatFromCurve(
      FloatCurve=DataInterface() {propertyValues {Curve {keys [{tangentMode=RCTM_User, time=0, value=0}, {tangentMode=RCTM_User, time=1, value=0.5}]}}}
      CurveIndex=$Particles.RibbonLinkOrder
     )
     Scale:NiagaraFloat=1
    )
    CustomOrientation:Quat4f=$Engine.Owner.Rotation
    Scale:Vector3f=&Multiply_VectorByFloat(
     Vector=&Multiply_VectorByFloat(
      Vector=&VectorFromFloat(Value=&Multiply_Float(
        A=$Emitter.RibbonLength
        B=&FloatFromCurve(FloatCurve=DataInterface() {propertyValues {Curve {keys [
             {arriveTangent=30.27, interpMode=RCIM_Cubic, leaveTangent=30.27, tangentMode=RCTM_User, time=0, value=0}
             {interpMode=RCIM_Cubic, tangentMode=RCTM_User, time=0.15, value=1.5}
             <at 0.2 1 RCIM_Cubic>
             {arriveTangent=-3.4902, interpMode=RCIM_Cubic, leaveTangent=-3.4902, tangentMode=RCTM_User, time=1, value=0}
            ]}}})
       ))
      Float=&Lerp_Float(A=$Emitter.RibbonLengthRandMin, B=$Emitter.RibbonLengthRandMax, Alpha=$Particles.RibbonRandom)
     )
     Float=$Particles.RibbonLinkOrder
    )
    StartPosition:Vector3f=$Particles.SkeletalMeshLocation001.SampledPosition
    Use Quaternion Rotation=true
   ) {module=FN_CalculateGoalPosition001}
   &FN_GooApproachGoalPosition(
    AttractionForce:NiagaraFloat=&Lerp_Float(A=$Emitter.ResolveSpeedLow, B=$Emitter.ResolveSpeedHigh, Alpha=$Particles.RibbonRandom)
    GoalPosition:NiagaraPosition=$Particles.GoalPosition
    MaximumDistance:NiagaraFloat=&Multiply_Float(
     A=&FloatFromCurve(
      FloatCurve=DataInterface() {propertyValues {Curve {keys [
          {arriveTangent=4.1785, interpMode=RCIM_Cubic, time=0, value=0}
          {arriveTangent=2.3315, interpMode=RCIM_Cubic, leaveTangent=2.3315, tangentMode=RCTM_User, time=1, value=1}
         ]}}}
      CurveIndex=$Particles.RibbonLinkOrder
      Scale Curve=&Multiply_Float(
       A=$Emitter.MaxDistance
       B=&FloatFromCurve(FloatCurve=DataInterface() {propertyValues {Curve {keys [<at 0.75 1>,<at 1 0>]}}})
      )
     )
     B=&Lerp_Float(A=$Emitter.MaxDistance_RandFactorLow, B=$Emitter.MaxDistance_RandFactorHigh, Alpha=$Particles.RibbonRandom)
    )
    RibbonAxis:Vector3f=$Particles.RibbonAxis
   )
   /SetVariables(Particles.PreviousRotation:Quat4f=$Engine.Owner.Rotation) {isSetParameters=true, module=SetVariables_#8}
   &ScaleColor(
    Scale Alpha=&FloatFromCurve(FloatCurve=DataInterface() {propertyValues {Curve {keys [<at 0 0>, {arriveTangent=6.6358, leaveTangent=6.6358, time=0.5, value=1}, <at 0.7 1>, <at 1 0>]}}})
    Scale RGB=&VectorFromFloat(Value=&FloatFromCurve(FloatCurve=DataInterface() {propertyValues {Curve {keys [{arriveTangent=-319.691, interpMode=RCIM_Cubic, leaveTangent=-319.691, tangentMode=RCTM_User, time=0, value=800}, <at 0.4 15>, <at 1 1>]}}}))
    ScaleA=true
    ScaleRGB=true
   )
   &DynamicMaterialParameters(
    Index 0 Param 1=&Multiply_Float(A=$Particles.RibbonRandom, B=10)
    Index 0 Param 2=$User.EnergyFillFactor
    Module.Param0WriteEnabled=false
    Module.Param1WriteEnabled=true
    Module.Param2WriteEnabled=false
    Module.Param3WriteEnabled=false
   ) {module=DynamicMaterialParameters001}
  ]
  rendererClasses [&NiagaraRibbonRendererProperties]
  renderers [{rendererClass=&NiagaraRibbonRendererProperties}]
 }
 Left_Hip001 {
  emitterSpawnScript [&FN_GooGlobalProperties(
    EndBoneIndexArray=DataInterface() {propertyValues {GpuSyncMode=SyncCpuToGpu, MaxElements=0, IntData [0]}}
    EndBoneOffsetArray=DataInterface() {propertyValues {GpuSyncMode=SyncCpuToGpu, MaxElements=0, FloatData [<0 0 0>]}}
    MaxDistance:NiagaraFloat=$User.HipRibbonLength
    MaxDistanceCompressed=10
    MaxDistance_RandFactorHigh=1
    MaxDistance_RandFactorLow=1
    ParticlesPerRibbon=2
    ResolveSpeedHigh=3
    RibbonLength:NiagaraFloat=$User.HipRibbonLength
    RibbonLengthRandHigh=1
    RibbonLengthRandLow=1
    StartBoneIndexArray=DataInterface() {propertyValues {GpuSyncMode=SyncCpuToGpu, MaxElements=0, IntData [10]}}
    StartBoneOffsetArray=DataInterface() {propertyValues {GpuSyncMode=SyncCpuToGpu, MaxElements=0, FloatData [<-2 0 -3>]}}
   )]
  emitterUpdateScript [
   &EmitterState()
   &SpawnBurst_Instantaneous(
    Spawn Count=&Multiply_Int(A=$Emitter.ParticlesPerRibbon, B=$Emitter.NumberOfRibbons)
    Use Spawn Probability=false
    UseLoopCountLimit=false
   )
  ]
  particleSpawnScript [
   /SetVariables(Particles.SpawnGroup=&Divide_Int(A:NiagaraInt32=&ReturnExecIndex(), B:NiagaraInt32=$Emitter.ParticlesPerRibbon)) {isSetParameters=true, module=SetVariables_#1}
   /SetVariables(
    Particles.RibbonID=&MakeRibbonIDFromSpawnGroup()
    Particles.RibbonLinkOrder:NiagaraFloat=&Divide_Float(
     A=&MakeFloatFromInt(Int=&ModuloInt(A=&ReturnExecIndex(), B=$Emitter.ParticlesPerRibbon))
     B=&Subtract_Float(A=&MakeFloatFromInt(Int=$Emitter.ParticlesPerRibbon), B=1)
    )
   ) {isSetParameters=true, module=SetVariables_#2}
   /SetVariables(Particles.RibbonRandom:NiagaraFloat=&RandomRangeFloat(
     Minimum=0
     Maximum=1
     Randomness Mode=::NewEnumerator1
     Random Seed=&Add_Integer(A=$Engine.System.RandomSeed, B=$Particles.SpawnGroup)
     Fixed Random Seed=true
    )) {isSetParameters=true, module=SetVariables_#3}
   &InitializeRibbon(
    Color=&MultiplyLinearColor(
     Linear Color=$User.BariumPaintColor
     Scale Factor=&MakeVector4_FromLinearColor(LinearColor=&ColorFromCurve(
       ColorCurve=DataInterface() {propertyValues {
         AlphaCurve {keys [<at 0.574 1 RCIM_Cubic>,<at 0.9977 0>]}
         BlueCurve {keys [<at 0 1 RCIM_Cubic>,<at 1 1 RCIM_Cubic>]}
         GreenCurve {keys [<at 0 1 RCIM_Cubic>,<at 1 1 RCIM_Cubic>]}
         RedCurve {keys [<at 0 1 RCIM_Cubic>,<at 1 1 RCIM_Cubic>]}
        }}
       CurveIndex=$Particles.RibbonLinkOrder
      ))
    )
    Lifetime=&Lerp_Float(A=2, B=3, Alpha=$Particles.RibbonRandom)
    Ribbon Width=$User.HipRibbonWidth
    UsePositionOffset=false
    Write Material Random=true
   )
   /SetVariables(Particles.MaterialRandom:NiagaraFloat=$Particles.RibbonRandom) {isSetParameters=true, module=SetVariables_#4}
   &SkeletalMeshLocation(
    Bone / Socket Index=&SelectIntFromArray(Direct Array Index=$Particles.SpawnGroup, Direct Array Mode=::NewEnumerator1, Int Selection Array=$Emitter.StartBoneIndexArray)
    Bone Sampling Mode=::NewEnumerator4
    MaskSpawnGroup=false
    Orientation Sampling=::NewEnumerator1
    Position Sampling=::NewEnumerator2
    Sampled Position Offset=&SelectVectorFromArray(Direct Array Index=$Particles.SpawnGroup, Direct Array Mode=::NewEnumerator1, Vector Selection Array=$Emitter.StartBoneOffsetArray)
    Skeletal Mesh=$User.SkeletalMesh
    Use Sampled Mesh Attribute as Ribbon ID=false
   )
   &FN_GooCalculateGoalPosition(
    Axis:Vector3f=$Emitter.RibbonAxis
    Scale:Vector3f=&Multiply_VectorByFloat(
     Vector=&Multiply_VectorByFloat(
      Vector=&VectorFromFloat(Value=&Multiply_Float(A=$Emitter.RibbonLength, B=0))
      Float=&Lerp_Float(A=$Emitter.RibbonLengthRandMin, B=$Emitter.RibbonLengthRandMax, Alpha=$Particles.RibbonRandom)
     )
     Float=$Particles.RibbonLinkOrder
    )
    StartPosition:Vector3f=$Particles.SkeletalMeshLocation.SampledPosition
   ) {module=FN_CalculateGoalPosition}
   /SetVariables(Particles.Position=$Particles.GoalPosition) {isSetParameters=true, module=SetVariables_#5}
   /SetVariables(Particles.PreviousRotation:Quat4f=$Engine.Owner.Rotation) {isSetParameters=true, module=SetVariables_#6}
   &DynamicMaterialParameters(
    Index 0 Param 1=&Multiply_Float(A=$Particles.RibbonRandom, B=10)
    Module.Param0WriteEnabled=true
    Module.Param1WriteEnabled=true
    Module.Param2WriteEnabled=true
    Module.Param3WriteEnabled=true
   )
   /SetVariables(Particles.SpriteSize=<10 10>) {isSetParameters=true, module=SetVariables_#7}
  ]
  particleUpdateScript [
   &ParticleState()
   &SkeletalMeshLocation(
    Bone / Socket Index=&SelectIntFromArray(Direct Array Index=$Particles.SpawnGroup, Direct Array Mode=::NewEnumerator1, Int Selection Array=$Emitter.StartBoneIndexArray)
    Bone Sampling Mode=::NewEnumerator4
    MaskSpawnGroup=false
    Orientation Sampling=::NewEnumerator1
    Position Sampling=::NewEnumerator2
    Sampled Position Offset=&SelectVectorFromArray(Direct Array Index=$Particles.SpawnGroup, Direct Array Mode=::NewEnumerator1, Vector Selection Array=$Emitter.StartBoneOffsetArray)
    Skeletal Mesh=$User.SkeletalMesh
    Use Sampled Mesh Attribute as Ribbon ID=false
   ) {module=SkeletalMeshLocation001}
   &SolveForcesAndVelocity(Clamp Velocity=false, Limit Acceleration=false)
   &FN_GooCalculateGoalPosition(
    Axis:Vector3f=&SlerpVector(
     Vector A:Vector3f=&SlerpVector(
      Vector A:Vector3f=<2 -1 0>
      Vector B:Vector3f=&TransformVector(Vector=$User.WorldSpaceWindVector, Source Space=::World, Destination Space=::Local)
      Alpha:NiagaraFloat=$User.WorldSpaceWindVectorFactor
      Scale:NiagaraFloat=1
     )
     Vector B:Vector3f=<0 0 1>
     Alpha:NiagaraFloat=&FloatFromCurve(
      FloatCurve=DataInterface() {propertyValues {Curve {keys [{tangentMode=RCTM_User, time=0, value=0}, {tangentMode=RCTM_User, time=1, value=0.5}]}}}
      CurveIndex=$Particles.RibbonLinkOrder
     )
     Scale:NiagaraFloat=1
    )
    CustomOrientation:Quat4f=$Engine.Owner.Rotation
    Scale:Vector3f=&Multiply_VectorByFloat(
     Vector=&Multiply_VectorByFloat(
      Vector=&VectorFromFloat(Value=&Multiply_Float(
        A=$Emitter.RibbonLength
        B=&FloatFromCurve(FloatCurve=DataInterface() {propertyValues {Curve {keys [
             {arriveTangent=30.27, interpMode=RCIM_Cubic, leaveTangent=30.27, tangentMode=RCTM_User, time=0, value=0}
             {interpMode=RCIM_Cubic, tangentMode=RCTM_User, time=0.15, value=1.5}
             <at 0.2 1 RCIM_Cubic>
             {arriveTangent=-3.4902, interpMode=RCIM_Cubic, leaveTangent=-3.4902, tangentMode=RCTM_User, time=1, value=0}
            ]}}})
       ))
      Float=&Lerp_Float(A=$Emitter.RibbonLengthRandMin, B=$Emitter.RibbonLengthRandMax, Alpha=$Particles.RibbonRandom)
     )
     Float=$Particles.RibbonLinkOrder
    )
    StartPosition:Vector3f=$Particles.SkeletalMeshLocation001.SampledPosition
    Use Quaternion Rotation=true
   ) {module=FN_CalculateGoalPosition001}
   &FN_GooApproachGoalPosition(
    AttractionForce:NiagaraFloat=&Lerp_Float(A=$Emitter.ResolveSpeedLow, B=$Emitter.ResolveSpeedHigh, Alpha=$Particles.RibbonRandom)
    GoalPosition:NiagaraPosition=$Particles.GoalPosition
    MaximumDistance:NiagaraFloat=&Multiply_Float(
     A=&FloatFromCurve(
      FloatCurve=DataInterface() {propertyValues {Curve {keys [
          {arriveTangent=4.1785, interpMode=RCIM_Cubic, time=0, value=0}
          {arriveTangent=2.3315, interpMode=RCIM_Cubic, leaveTangent=2.3315, tangentMode=RCTM_User, time=1, value=1}
         ]}}}
      CurveIndex=$Particles.RibbonLinkOrder
      Scale Curve=&Multiply_Float(
       A=$Emitter.MaxDistance
       B=&FloatFromCurve(FloatCurve=DataInterface() {propertyValues {Curve {keys [<at 0.75 1>,<at 1 0>]}}})
      )
     )
     B=&Lerp_Float(A=$Emitter.MaxDistance_RandFactorLow, B=$Emitter.MaxDistance_RandFactorHigh, Alpha=$Particles.RibbonRandom)
    )
    RibbonAxis:Vector3f=$Particles.RibbonAxis
   )
   /SetVariables(Particles.PreviousRotation:Quat4f=$Engine.Owner.Rotation) {isSetParameters=true, module=SetVariables_#8}
   &ScaleColor(
    Scale Alpha=&FloatFromCurve(FloatCurve=DataInterface() {propertyValues {Curve {keys [<at 0 0>, {arriveTangent=6.6358, leaveTangent=6.6358, time=0.5, value=1}, <at 0.7 1>, <at 1 0>]}}})
    Scale RGB=&VectorFromFloat(Value=&FloatFromCurve(FloatCurve=DataInterface() {propertyValues {Curve {keys [{arriveTangent=-319.691, interpMode=RCIM_Cubic, leaveTangent=-319.691, tangentMode=RCTM_User, time=0, value=800}, <at 0.4 15>, <at 1 1>]}}}))
    ScaleA=true
    ScaleRGB=true
   )
   &DynamicMaterialParameters(
    Index 0 Param 1=&Multiply_Float(A=$Particles.RibbonRandom, B=10)
    Index 0 Param 2=$User.EnergyFillFactor
    Module.Param0WriteEnabled=false
    Module.Param1WriteEnabled=true
    Module.Param2WriteEnabled=false
    Module.Param3WriteEnabled=false
   ) {module=DynamicMaterialParameters001}
  ]
  rendererClasses [&NiagaraRibbonRendererProperties]
  renderers [{rendererClass=&NiagaraRibbonRendererProperties}]
 }
 Left_Leg {
  emitterSpawnScript [&FN_GooGlobalProperties(
    EndBoneIndexArray=DataInterface() {propertyValues {GpuSyncMode=SyncCpuToGpu, MaxElements=0, IntData [0]}}
    EndBoneOffsetArray=DataInterface() {propertyValues {GpuSyncMode=SyncCpuToGpu, MaxElements=0, FloatData [<0 0 0>]}}
    MaxDistance:NiagaraFloat=$User.LegRibbonLength
    MaxDistanceCompressed=10
    MaxDistance_RandFactorHigh=1
    MaxDistance_RandFactorLow=1
    NumberOfRibbons=2
    ParticlesPerRibbon=3
    ResolveSpeedHigh=3
    RibbonLength:NiagaraFloat=$User.LegRibbonLength
    RibbonLengthRandHigh=1
    RibbonLengthRandLow=1
    StartBoneIndexArray=DataInterface() {propertyValues {GpuSyncMode=SyncCpuToGpu, MaxElements=0, IntData [6,7]}}
    StartBoneOffsetArray=DataInterface() {propertyValues {GpuSyncMode=SyncCpuToGpu, MaxElements=0, FloatData [<-20 0 -5>, <15 0 -5>]}}
   )]
  emitterUpdateScript [
   &EmitterState()
   &SpawnBurst_Instantaneous(
    Spawn Count=&Multiply_Int(A=$Emitter.ParticlesPerRibbon, B=$Emitter.NumberOfRibbons)
    Use Spawn Probability=false
    UseLoopCountLimit=false
   )
  ]
  particleSpawnScript [
   /SetVariables(Particles.SpawnGroup=&Divide_Int(A:NiagaraInt32=&ReturnExecIndex(), B:NiagaraInt32=$Emitter.ParticlesPerRibbon)) {isSetParameters=true, module=SetVariables_#1}
   /SetVariables(
    Particles.RibbonID=&MakeRibbonIDFromSpawnGroup()
    Particles.RibbonLinkOrder:NiagaraFloat=&Divide_Float(
     A=&MakeFloatFromInt(Int=&ModuloInt(A=&ReturnExecIndex(), B=$Emitter.ParticlesPerRibbon))
     B=&Subtract_Float(A=&MakeFloatFromInt(Int=$Emitter.ParticlesPerRibbon), B=1)
    )
   ) {isSetParameters=true, module=SetVariables_#2}
   /SetVariables(Particles.RibbonRandom:NiagaraFloat=&RandomRangeFloat(
     Minimum=0
     Maximum=1
     Randomness Mode=::NewEnumerator1
     Random Seed=&Add_Integer(A=$Engine.System.RandomSeed, B=$Particles.SpawnGroup)
     Fixed Random Seed=true
    )) {isSetParameters=true, module=SetVariables_#3}
   &InitializeRibbon(
    Color=&MultiplyLinearColor(
     Linear Color=$User.BariumPaintColor
     Scale Factor=&MakeVector4_FromLinearColor(LinearColor=&ColorFromCurve(
       ColorCurve=DataInterface() {propertyValues {
         AlphaCurve {keys [<at 0.574 1 RCIM_Cubic>,<at 0.9977 0>]}
         BlueCurve {keys [<at 0 1 RCIM_Cubic>,<at 1 1 RCIM_Cubic>]}
         GreenCurve {keys [<at 0 1 RCIM_Cubic>,<at 1 1 RCIM_Cubic>]}
         RedCurve {keys [<at 0 1 RCIM_Cubic>,<at 1 1 RCIM_Cubic>]}
        }}
       CurveIndex=$Particles.RibbonLinkOrder
      ))
    )
    Lifetime=&Lerp_Float(A=2, B=3, Alpha=$Particles.RibbonRandom)
    Ribbon Width=$User.LegRibbonWidth
    UsePositionOffset=false
    Write Material Random=true
   )
   /SetVariables(Particles.MaterialRandom:NiagaraFloat=$Particles.RibbonRandom) {isSetParameters=true, module=SetVariables_#4}
   &SkeletalMeshLocation(
    Bone / Socket Index=&SelectIntFromArray(Direct Array Index=$Particles.SpawnGroup, Direct Array Mode=::NewEnumerator1, Int Selection Array=$Emitter.StartBoneIndexArray)
    Bone Sampling Mode=::NewEnumerator4
    MaskSpawnGroup=false
    Orientation Sampling=::NewEnumerator1
    Position Sampling=::NewEnumerator2
    Sampled Position Offset=&SelectVectorFromArray(Direct Array Index=$Particles.SpawnGroup, Direct Array Mode=::NewEnumerator1, Vector Selection Array=$Emitter.StartBoneOffsetArray)
    Skeletal Mesh=$User.SkeletalMesh
    Use Sampled Mesh Attribute as Ribbon ID=false
   )
   &FN_GooCalculateGoalPosition(
    Axis:Vector3f=$Emitter.RibbonAxis
    Scale:Vector3f=&Multiply_VectorByFloat(
     Vector=&Multiply_VectorByFloat(
      Vector=&VectorFromFloat(Value=&Multiply_Float(A=$Emitter.RibbonLength, B=0))
      Float=&Lerp_Float(A=$Emitter.RibbonLengthRandMin, B=$Emitter.RibbonLengthRandMax, Alpha=$Particles.RibbonRandom)
     )
     Float=$Particles.RibbonLinkOrder
    )
    StartPosition:Vector3f=$Particles.SkeletalMeshLocation.SampledPosition
   ) {module=FN_CalculateGoalPosition}
   /SetVariables(Particles.Position=$Particles.GoalPosition) {isSetParameters=true, module=SetVariables_#5}
   /SetVariables(Particles.PreviousRotation:Quat4f=$Engine.Owner.Rotation) {isSetParameters=true, module=SetVariables_#6}
   &DynamicMaterialParameters(
    Index 0 Param 1=&Multiply_Float(A=$Particles.RibbonRandom, B=10)
    Module.Param0WriteEnabled=true
    Module.Param1WriteEnabled=true
    Module.Param2WriteEnabled=true
    Module.Param3WriteEnabled=true
   )
   /SetVariables(Particles.SpriteSize=<10 10>) {isSetParameters=true, module=SetVariables_#7}
  ]
  particleUpdateScript [
   &ParticleState()
   &SkeletalMeshLocation(
    Bone / Socket Index=&SelectIntFromArray(Direct Array Index=$Particles.SpawnGroup, Direct Array Mode=::NewEnumerator1, Int Selection Array=$Emitter.StartBoneIndexArray)
    Bone Sampling Mode=::NewEnumerator4
    MaskSpawnGroup=false
    Orientation Sampling=::NewEnumerator1
    Position Sampling=::NewEnumerator2
    Sampled Position Offset=&SelectVectorFromArray(Direct Array Index=$Particles.SpawnGroup, Direct Array Mode=::NewEnumerator1, Vector Selection Array=$Emitter.StartBoneOffsetArray)
    Skeletal Mesh=$User.SkeletalMesh
    Use Sampled Mesh Attribute as Ribbon ID=false
   ) {module=SkeletalMeshLocation001}
   &SolveForcesAndVelocity(Clamp Velocity=false, Limit Acceleration=false)
   &FN_GooCalculateGoalPosition(
    Axis:Vector3f=&SlerpVector(
     Vector A:Vector3f=&SlerpVector(
      Vector A:Vector3f=<1 0 0>
      Vector B:Vector3f=&TransformVector(Vector=$User.WorldSpaceWindVector, Source Space=::World, Destination Space=::Local)
      Alpha:NiagaraFloat=$User.WorldSpaceWindVectorFactor
      Scale:NiagaraFloat=1
     )
     Vector B:Vector3f=<0 0 1>
     Alpha:NiagaraFloat=&FloatFromCurve(
      FloatCurve=DataInterface() {propertyValues {Curve {keys [{tangentMode=RCTM_User, time=0, value=0}, {tangentMode=RCTM_User, time=1, value=0.5}]}}}
      CurveIndex=$Particles.RibbonLinkOrder
     )
     Scale:NiagaraFloat=1
    )
    CustomOrientation:Quat4f=$Engine.Owner.Rotation
    Scale:Vector3f=&Multiply_VectorByFloat(
     Vector=&Multiply_VectorByFloat(
      Vector=&VectorFromFloat(Value=&Multiply_Float(
        A=$Emitter.RibbonLength
        B=&FloatFromCurve(FloatCurve=DataInterface() {propertyValues {Curve {keys [
             {arriveTangent=30.27, interpMode=RCIM_Cubic, leaveTangent=30.27, tangentMode=RCTM_User, time=0, value=0}
             {interpMode=RCIM_Cubic, tangentMode=RCTM_User, time=0.15, value=1.5}
             <at 0.2 1 RCIM_Cubic>
             {arriveTangent=-3.4902, interpMode=RCIM_Cubic, leaveTangent=-3.4902, tangentMode=RCTM_User, time=1, value=0}
            ]}}})
       ))
      Float=&Lerp_Float(A=$Emitter.RibbonLengthRandMin, B=$Emitter.RibbonLengthRandMax, Alpha=$Particles.RibbonRandom)
     )
     Float=$Particles.RibbonLinkOrder
    )
    StartPosition:Vector3f=$Particles.SkeletalMeshLocation001.SampledPosition
    Use Quaternion Rotation=true
   ) {module=FN_CalculateGoalPosition001}
   &FN_GooApproachGoalPosition(
    AttractionForce:NiagaraFloat=&Lerp_Float(A=$Emitter.ResolveSpeedLow, B=$Emitter.ResolveSpeedHigh, Alpha=$Particles.RibbonRandom)
    GoalPosition:NiagaraPosition=$Particles.GoalPosition
    MaximumDistance:NiagaraFloat=&Multiply_Float(
     A=&FloatFromCurve(
      FloatCurve=DataInterface() {propertyValues {Curve {keys [
          {arriveTangent=4.1785, interpMode=RCIM_Cubic, time=0, value=0}
          {arriveTangent=2.3315, interpMode=RCIM_Cubic, leaveTangent=2.3315, tangentMode=RCTM_User, time=1, value=1}
         ]}}}
      CurveIndex=$Particles.RibbonLinkOrder
      Scale Curve=&Multiply_Float(
       A=$Emitter.MaxDistance
       B=&FloatFromCurve(FloatCurve=DataInterface() {propertyValues {Curve {keys [<at 0.75 1>,<at 1 0>]}}})
      )
     )
     B=&Lerp_Float(A=$Emitter.MaxDistance_RandFactorLow, B=$Emitter.MaxDistance_RandFactorHigh, Alpha=$Particles.RibbonRandom)
    )
    RibbonAxis:Vector3f=$Particles.RibbonAxis
   )
   /SetVariables(Particles.PreviousRotation:Quat4f=$Engine.Owner.Rotation) {isSetParameters=true, module=SetVariables_#8}
   &ScaleColor(
    Scale Alpha=&FloatFromCurve(FloatCurve=DataInterface() {propertyValues {Curve {keys [
         {arriveTangent=1.4631, interpMode=RCIM_Cubic, leaveTangent=1.4631, tangentMode=RCTM_User, time=0, value=0}
         {arriveTangent=5.0136, leaveTangent=5.0136, time=0.5, value=1}
         <at 0.7 1>
         <at 1 0>
        ]}}})
    Scale RGB=&VectorFromFloat(Value=&FloatFromCurve(FloatCurve=DataInterface() {propertyValues {Curve {keys [{arriveTangent=-319.691, interpMode=RCIM_Cubic, leaveTangent=-319.691, tangentMode=RCTM_User, time=0, value=800}, <at 0.4 15>, <at 1 1>]}}}))
    ScaleA=true
    ScaleRGB=true
   )
   &DynamicMaterialParameters(
    Index 0 Param 1=&Multiply_Float(A=$Particles.RibbonRandom, B=10)
    Index 0 Param 2=$User.EnergyFillFactor
    Module.Param0WriteEnabled=false
    Module.Param1WriteEnabled=true
    Module.Param2WriteEnabled=false
    Module.Param3WriteEnabled=false
   ) {module=DynamicMaterialParameters001}
  ]
  rendererClasses [&NiagaraRibbonRendererProperties]
  renderers [{rendererClass=&NiagaraRibbonRendererProperties}]
 }
 Left_Leg001 {
  emitterSpawnScript [&FN_GooGlobalProperties(
    EndBoneIndexArray=DataInterface() {propertyValues {GpuSyncMode=SyncCpuToGpu, MaxElements=0, IntData [0]}}
    EndBoneOffsetArray=DataInterface() {propertyValues {GpuSyncMode=SyncCpuToGpu, MaxElements=0, FloatData [<0 0 0>]}}
    MaxDistance:NiagaraFloat=$User.LegRibbonLength
    MaxDistanceCompressed=10
    MaxDistance_RandFactorHigh=1
    MaxDistance_RandFactorLow=1
    NumberOfRibbons=2
    ParticlesPerRibbon=2
    ResolveSpeedHigh=3
    RibbonLength:NiagaraFloat=$User.LegRibbonLength
    RibbonLengthRandHigh=1
    RibbonLengthRandLow=1
    StartBoneIndexArray=DataInterface() {propertyValues {GpuSyncMode=SyncCpuToGpu, MaxElements=0, IntData [6,7]}}
    StartBoneOffsetArray=DataInterface() {propertyValues {GpuSyncMode=SyncCpuToGpu, MaxElements=0, FloatData [<-20 0 -5>, <15 0 -5>]}}
   )]
  emitterUpdateScript [
   &EmitterState()
   &SpawnBurst_Instantaneous(
    Spawn Count=&Multiply_Int(A=$Emitter.ParticlesPerRibbon, B=$Emitter.NumberOfRibbons)
    Use Spawn Probability=false
    UseLoopCountLimit=false
   )
  ]
  particleSpawnScript [
   /SetVariables(Particles.SpawnGroup=&Divide_Int(A:NiagaraInt32=&ReturnExecIndex(), B:NiagaraInt32=$Emitter.ParticlesPerRibbon)) {isSetParameters=true, module=SetVariables_#1}
   /SetVariables(
    Particles.RibbonID=&MakeRibbonIDFromSpawnGroup()
    Particles.RibbonLinkOrder:NiagaraFloat=&Divide_Float(
     A=&MakeFloatFromInt(Int=&ModuloInt(A=&ReturnExecIndex(), B=$Emitter.ParticlesPerRibbon))
     B=&Subtract_Float(A=&MakeFloatFromInt(Int=$Emitter.ParticlesPerRibbon), B=1)
    )
   ) {isSetParameters=true, module=SetVariables_#2}
   /SetVariables(Particles.RibbonRandom:NiagaraFloat=&RandomRangeFloat(
     Minimum=0
     Maximum=1
     Randomness Mode=::NewEnumerator1
     Random Seed=&Add_Integer(A=$Engine.System.RandomSeed, B=$Particles.SpawnGroup)
     Fixed Random Seed=true
    )) {isSetParameters=true, module=SetVariables_#3}
   &InitializeRibbon(
    Color=&MultiplyLinearColor(
     Linear Color=$User.BariumPaintColor
     Scale Factor=&MakeVector4_FromLinearColor(LinearColor=&ColorFromCurve(
       ColorCurve=DataInterface() {propertyValues {
         AlphaCurve {keys [<at 0.574 1 RCIM_Cubic>,<at 0.9977 0>]}
         BlueCurve {keys [<at 0 1 RCIM_Cubic>,<at 1 1 RCIM_Cubic>]}
         GreenCurve {keys [<at 0 1 RCIM_Cubic>,<at 1 1 RCIM_Cubic>]}
         RedCurve {keys [<at 0 1 RCIM_Cubic>,<at 1 1 RCIM_Cubic>]}
        }}
       CurveIndex=$Particles.RibbonLinkOrder
      ))
    )
    Lifetime=&Lerp_Float(A=2, B=3, Alpha=$Particles.RibbonRandom)
    Ribbon Width=$User.LegRibbonWidth
    UsePositionOffset=false
    Write Material Random=true
   )
   /SetVariables(Particles.MaterialRandom:NiagaraFloat=$Particles.RibbonRandom) {isSetParameters=true, module=SetVariables_#4}
   &SkeletalMeshLocation(
    Bone / Socket Index=&SelectIntFromArray(Direct Array Index=$Particles.SpawnGroup, Direct Array Mode=::NewEnumerator1, Int Selection Array=$Emitter.StartBoneIndexArray)
    Bone Sampling Mode=::NewEnumerator4
    MaskSpawnGroup=false
    Orientation Sampling=::NewEnumerator1
    Position Sampling=::NewEnumerator2
    Sampled Position Offset=&SelectVectorFromArray(Direct Array Index=$Particles.SpawnGroup, Direct Array Mode=::NewEnumerator1, Vector Selection Array=$Emitter.StartBoneOffsetArray)
    Skeletal Mesh=$User.SkeletalMesh
    Use Sampled Mesh Attribute as Ribbon ID=false
   )
   &FN_GooCalculateGoalPosition(
    Axis:Vector3f=$Emitter.RibbonAxis
    Scale:Vector3f=&Multiply_VectorByFloat(
     Vector=&Multiply_VectorByFloat(
      Vector=&VectorFromFloat(Value=&Multiply_Float(A=$Emitter.RibbonLength, B=0))
      Float=&Lerp_Float(A=$Emitter.RibbonLengthRandMin, B=$Emitter.RibbonLengthRandMax, Alpha=$Particles.RibbonRandom)
     )
     Float=$Particles.RibbonLinkOrder
    )
    StartPosition:Vector3f=$Particles.SkeletalMeshLocation.SampledPosition
   ) {module=FN_CalculateGoalPosition}
   /SetVariables(Particles.Position=$Particles.GoalPosition) {isSetParameters=true, module=SetVariables_#5}
   /SetVariables(Particles.PreviousRotation:Quat4f=$Engine.Owner.Rotation) {isSetParameters=true, module=SetVariables_#6}
   &DynamicMaterialParameters(
    Index 0 Param 1=&Multiply_Float(A=$Particles.RibbonRandom, B=10)
    Module.Param0WriteEnabled=true
    Module.Param1WriteEnabled=true
    Module.Param2WriteEnabled=true
    Module.Param3WriteEnabled=true
   )
   /SetVariables(Particles.SpriteSize=<10 10>) {isSetParameters=true, module=SetVariables_#7}
  ]
  particleUpdateScript [
   &ParticleState()
   &SkeletalMeshLocation(
    Bone / Socket Index=&SelectIntFromArray(Direct Array Index=$Particles.SpawnGroup, Direct Array Mode=::NewEnumerator1, Int Selection Array=$Emitter.StartBoneIndexArray)
    Bone Sampling Mode=::NewEnumerator4
    MaskSpawnGroup=false
    Orientation Sampling=::NewEnumerator1
    Position Sampling=::NewEnumerator2
    Sampled Position Offset=&SelectVectorFromArray(Direct Array Index=$Particles.SpawnGroup, Direct Array Mode=::NewEnumerator1, Vector Selection Array=$Emitter.StartBoneOffsetArray)
    Skeletal Mesh=$User.SkeletalMesh
    Use Sampled Mesh Attribute as Ribbon ID=false
   ) {module=SkeletalMeshLocation001}
   &SolveForcesAndVelocity(Clamp Velocity=false, Limit Acceleration=false)
   &FN_GooCalculateGoalPosition(
    Axis:Vector3f=&SlerpVector(
     Vector A:Vector3f=&SlerpVector(
      Vector A:Vector3f=<1 0 0>
      Vector B:Vector3f=&TransformVector(Vector=$User.WorldSpaceWindVector, Source Space=::World, Destination Space=::Local)
      Alpha:NiagaraFloat=$User.WorldSpaceWindVectorFactor
      Scale:NiagaraFloat=1
     )
     Vector B:Vector3f=<0 0 1>
     Alpha:NiagaraFloat=&FloatFromCurve(
      FloatCurve=DataInterface() {propertyValues {Curve {keys [{tangentMode=RCTM_User, time=0, value=0}, {tangentMode=RCTM_User, time=1, value=0.5}]}}}
      CurveIndex=$Particles.RibbonLinkOrder
     )
     Scale:NiagaraFloat=1
    )
    CustomOrientation:Quat4f=$Engine.Owner.Rotation
    Scale:Vector3f=&Multiply_VectorByFloat(
     Vector=&Multiply_VectorByFloat(
      Vector=&VectorFromFloat(Value=&Multiply_Float(
        A=$Emitter.RibbonLength
        B=&FloatFromCurve(FloatCurve=DataInterface() {propertyValues {Curve {keys [
             {arriveTangent=30.27, interpMode=RCIM_Cubic, leaveTangent=30.27, tangentMode=RCTM_User, time=0, value=0}
             {interpMode=RCIM_Cubic, tangentMode=RCTM_User, time=0.15, value=1.5}
             <at 0.2 1 RCIM_Cubic>
             {arriveTangent=-3.4902, interpMode=RCIM_Cubic, leaveTangent=-3.4902, tangentMode=RCTM_User, time=1, value=0}
            ]}}})
       ))
      Float=&Lerp_Float(A=$Emitter.RibbonLengthRandMin, B=$Emitter.RibbonLengthRandMax, Alpha=$Particles.RibbonRandom)
     )
     Float=$Particles.RibbonLinkOrder
    )
    StartPosition:Vector3f=$Particles.SkeletalMeshLocation001.SampledPosition
    Use Quaternion Rotation=true
   ) {module=FN_CalculateGoalPosition001}
   &FN_GooApproachGoalPosition(
    AttractionForce:NiagaraFloat=&Lerp_Float(A=$Emitter.ResolveSpeedLow, B=$Emitter.ResolveSpeedHigh, Alpha=$Particles.RibbonRandom)
    GoalPosition:NiagaraPosition=$Particles.GoalPosition
    MaximumDistance:NiagaraFloat=&Multiply_Float(
     A=&FloatFromCurve(
      FloatCurve=DataInterface() {propertyValues {Curve {keys [
          {arriveTangent=4.1785, interpMode=RCIM_Cubic, time=0, value=0}
          {arriveTangent=2.3315, interpMode=RCIM_Cubic, leaveTangent=2.3315, tangentMode=RCTM_User, time=1, value=1}
         ]}}}
      CurveIndex=$Particles.RibbonLinkOrder
      Scale Curve=&Multiply_Float(
       A=$Emitter.MaxDistance
       B=&FloatFromCurve(FloatCurve=DataInterface() {propertyValues {Curve {keys [<at 0.75 1>,<at 1 0>]}}})
      )
     )
     B=&Lerp_Float(A=$Emitter.MaxDistance_RandFactorLow, B=$Emitter.MaxDistance_RandFactorHigh, Alpha=$Particles.RibbonRandom)
    )
    RibbonAxis:Vector3f=$Particles.RibbonAxis
   )
   /SetVariables(Particles.PreviousRotation:Quat4f=$Engine.Owner.Rotation) {isSetParameters=true, module=SetVariables_#8}
   &ScaleColor(
    Scale Alpha=&FloatFromCurve(FloatCurve=DataInterface() {propertyValues {Curve {keys [
         {arriveTangent=1.4631, interpMode=RCIM_Cubic, leaveTangent=1.4631, tangentMode=RCTM_User, time=0, value=0}
         {arriveTangent=5.0136, leaveTangent=5.0136, time=0.5, value=1}
         <at 0.7 1>
         <at 1 0>
        ]}}})
    Scale RGB=&VectorFromFloat(Value=&FloatFromCurve(FloatCurve=DataInterface() {propertyValues {Curve {keys [{arriveTangent=-319.691, interpMode=RCIM_Cubic, leaveTangent=-319.691, tangentMode=RCTM_User, time=0, value=800}, <at 0.4 15>, <at 1 1>]}}}))
    ScaleA=true
    ScaleRGB=true
   )
   &DynamicMaterialParameters(
    Index 0 Param 1=&Multiply_Float(A=$Particles.RibbonRandom, B=10)
    Index 0 Param 2=$User.EnergyFillFactor
    Module.Param0WriteEnabled=false
    Module.Param1WriteEnabled=true
    Module.Param2WriteEnabled=false
    Module.Param3WriteEnabled=false
   ) {module=DynamicMaterialParameters001}
  ]
  rendererClasses [&NiagaraRibbonRendererProperties]
  renderers [{rendererClass=&NiagaraRibbonRendererProperties}]
 }
 Right_Arm {
  emitterSpawnScript [&FN_GooGlobalProperties(
    EndBoneIndexArray=DataInterface() {propertyValues {GpuSyncMode=SyncCpuToGpu, MaxElements=0, IntData [0]}}
    EndBoneOffsetArray=DataInterface() {propertyValues {GpuSyncMode=SyncCpuToGpu, MaxElements=0, FloatData [<0 0 0>]}}
    MaxDistance:NiagaraFloat=$User.ArmRibbonLength
    MaxDistanceCompressed=10
    MaxDistance_RandFactorHigh=1
    MaxDistance_RandFactorLow=1
    NumberOfRibbons=2
    RibbonLength:NiagaraFloat=$User.ArmRibbonLength
    RibbonLengthRandHigh=1
    RibbonLengthRandLow=1
    StartBoneIndexArray=DataInterface() {propertyValues {GpuSyncMode=SyncCpuToGpu, MaxElements=0, IntData [3,4]}}
    StartBoneOffsetArray=DataInterface() {propertyValues {GpuSyncMode=SyncCpuToGpu, MaxElements=0, FloatData [<-10 0 0>, <-3 0 0>]}}
   )]
  emitterUpdateScript [
   &EmitterState()
   &SpawnBurst_Instantaneous(
    Spawn Count=&Multiply_Int(A=$Emitter.ParticlesPerRibbon, B=$Emitter.NumberOfRibbons)
    Spawn Probability=$User.SpawnJacketFX
    Use Spawn Probability=true
    UseLoopCountLimit=false
   )
  ]
  particleSpawnScript [
   /SetVariables(Particles.SpawnGroup=&Divide_Int(A:NiagaraInt32=&ReturnExecIndex(), B:NiagaraInt32=$Emitter.ParticlesPerRibbon)) {isSetParameters=true, module=SetVariables_#9}
   /SetVariables(
    Particles.RibbonID=&MakeRibbonIDFromSpawnGroup()
    Particles.RibbonLinkOrder:NiagaraFloat=&Divide_Float(
     A=&MakeFloatFromInt(Int=&ModuloInt(A=&ReturnExecIndex(), B=$Emitter.ParticlesPerRibbon))
     B=&Subtract_Float(A=&MakeFloatFromInt(Int=$Emitter.ParticlesPerRibbon), B=1)
    )
   ) {isSetParameters=true, module=SetVariables_#10}
   /SetVariables(Particles.RibbonRandom:NiagaraFloat=&RandomRangeFloat(
     Minimum=0
     Maximum=1
     Randomness Mode=::NewEnumerator1
     Random Seed=&Add_Integer(A=$Engine.System.RandomSeed, B=$Particles.SpawnGroup)
     Fixed Random Seed=true
    )) {isSetParameters=true, module=SetVariables_#3}
   &InitializeRibbon(
    Color=&MultiplyLinearColor(
     Linear Color=$User.BariumPaintColor
     Scale Factor=&MakeVector4_FromLinearColor(LinearColor=&ColorFromCurve(
       ColorCurve=DataInterface() {propertyValues {
         AlphaCurve {keys [<at 0.574 1 RCIM_Cubic>,<at 0.9977 0>]}
         BlueCurve {keys [<at 0 1 RCIM_Cubic>,<at 1 1 RCIM_Cubic>]}
         GreenCurve {keys [<at 0 1 RCIM_Cubic>,<at 1 1 RCIM_Cubic>]}
         RedCurve {keys [<at 0 1 RCIM_Cubic>,<at 1 1 RCIM_Cubic>]}
        }}
       CurveIndex=$Particles.RibbonLinkOrder
      ))
    )
    Lifetime=&Lerp_Float(A=2, B=3, Alpha=$Particles.RibbonRandom)
    Ribbon Width=&FloatFromCurve(
     FloatCurve=DataInterface() {propertyValues {Curve {keys [<at 0 1>,<at 0.5 0.5>,<at 1 1>]}}}
     CurveIndex=$Particles.RibbonLinkOrder
     Scale Curve=$User.ArmRibbonWidth
    )
    UsePositionOffset=false
    Write Material Random=true
   )
   /SetVariables(Particles.MaterialRandom:NiagaraFloat=$Particles.RibbonRandom) {isSetParameters=true, module=SetVariables_#4}
   &SkeletalMeshLocation(
    Bone / Socket Index=&SelectIntFromArray(Direct Array Index=$Particles.SpawnGroup, Direct Array Mode=::NewEnumerator1, Int Selection Array=$Emitter.StartBoneIndexArray)
    Bone Sampling Mode=::NewEnumerator4
    MaskSpawnGroup=false
    Orientation Sampling=::NewEnumerator1
    Position Sampling=::NewEnumerator2
    Sampled Position Offset=&SelectVectorFromArray(Direct Array Index=$Particles.SpawnGroup, Direct Array Mode=::NewEnumerator1, Vector Selection Array=$Emitter.StartBoneOffsetArray)
    Skeletal Mesh=$User.SkeletalMesh
    Use Sampled Mesh Attribute as Ribbon ID=false
   )
   &FN_GooCalculateGoalPosition(
    Axis:Vector3f=$Emitter.RibbonAxis
    Scale:Vector3f=&Multiply_VectorByFloat(
     Vector=&Multiply_VectorByFloat(
      Vector=&VectorFromFloat(Value=&Multiply_Float(A=$Emitter.RibbonLength, B=0))
      Float=&Lerp_Float(A=$Emitter.RibbonLengthRandMin, B=$Emitter.RibbonLengthRandMax, Alpha=$Particles.RibbonRandom)
     )
     Float=$Particles.RibbonLinkOrder
    )
    StartPosition:Vector3f=$Particles.SkeletalMeshLocation.SampledPosition
   ) {module=FN_CalculateGoalPosition}
   /SetVariables(Particles.Position=$Particles.GoalPosition) {isSetParameters=true, module=SetVariables_#5}
   /SetVariables(Particles.PreviousRotation:Quat4f=$Engine.Owner.Rotation) {isSetParameters=true, module=SetVariables_#6}
   &DynamicMaterialParameters(
    Index 0 Param 1=&Multiply_Float(A=$Particles.RibbonRandom, B=10)
    Module.Param0WriteEnabled=true
    Module.Param1WriteEnabled=true
    Module.Param2WriteEnabled=true
    Module.Param3WriteEnabled=true
   )
   /SetVariables(Particles.SpriteSize=<10 10>) {isSetParameters=true, module=SetVariables_#7}
  ]
  particleUpdateScript [
   &ParticleState()
   &SkeletalMeshLocation(
    Bone / Socket Index=&SelectIntFromArray(Direct Array Index=$Particles.SpawnGroup, Direct Array Mode=::NewEnumerator1, Int Selection Array=$Emitter.StartBoneIndexArray)
    Bone Sampling Mode=::NewEnumerator4
    MaskSpawnGroup=false
    Orientation Sampling=::NewEnumerator1
    Position Sampling=::NewEnumerator2
    Sampled Position Offset=&SelectVectorFromArray(Direct Array Index=$Particles.SpawnGroup, Direct Array Mode=::NewEnumerator1, Vector Selection Array=$Emitter.StartBoneOffsetArray)
    Skeletal Mesh=$User.SkeletalMesh
    Use Sampled Mesh Attribute as Ribbon ID=false
   ) {module=SkeletalMeshLocation001}
   &SolveForcesAndVelocity(Clamp Velocity=false, Limit Acceleration=false)
   &FN_GooCalculateGoalPosition(
    Axis:Vector3f=&SlerpVector(
     Vector A:Vector3f=&SlerpVector(
      Vector A:Vector3f=<-1 0 0>
      Vector B:Vector3f=&TransformVector(Vector=$User.WorldSpaceWindVector, Source Space=::World, Destination Space=::Local)
      Alpha:NiagaraFloat=$User.WorldSpaceWindVectorFactor
      Scale:NiagaraFloat=1
     )
     Vector B:Vector3f=<0 0 1>
     Alpha:NiagaraFloat=&FloatFromCurve(
      FloatCurve=DataInterface() {propertyValues {Curve {keys [{tangentMode=RCTM_User, time=0, value=0}, {tangentMode=RCTM_User, time=1, value=1}]}}}
      CurveIndex=$Particles.RibbonLinkOrder
      Scale Curve=0.7
     )
     Scale:NiagaraFloat=1
    )
    CustomOrientation:Quat4f=$Engine.Owner.Rotation
    Scale:Vector3f=&Multiply_VectorByFloat(
     Vector=&Multiply_VectorByFloat(
      Vector=&VectorFromFloat(Value=&Multiply_Float(
        A=$Emitter.RibbonLength
        B=&FloatFromCurve(FloatCurve=DataInterface() {propertyValues {Curve {keys [
             {arriveTangent=17.1487, interpMode=RCIM_Cubic, leaveTangent=17.1487, tangentMode=RCTM_User, time=0, value=0}
             {interpMode=RCIM_Cubic, tangentMode=RCTM_User, time=0.2, value=1.5}
             <at 0.25 1 RCIM_Cubic>
             {arriveTangent=-3.4902, interpMode=RCIM_Cubic, leaveTangent=-3.4902, tangentMode=RCTM_User, time=1, value=0}
            ]}}})
       ))
      Float=&Lerp_Float(A=$Emitter.RibbonLengthRandMin, B=$Emitter.RibbonLengthRandMax, Alpha=$Particles.RibbonRandom)
     )
     Float=$Particles.RibbonLinkOrder
    )
    StartPosition:Vector3f=$Particles.SkeletalMeshLocation001.SampledPosition
    Use Quaternion Rotation=true
   ) {module=FN_CalculateGoalPosition001}
   &FN_GooApproachGoalPosition(
    AttractionForce:NiagaraFloat=&Lerp_Float(A=$Emitter.ResolveSpeedLow, B=$Emitter.ResolveSpeedHigh, Alpha=$Particles.RibbonRandom)
    GoalPosition:NiagaraPosition=$Particles.GoalPosition
    MaximumDistance:NiagaraFloat=&Multiply_Float(
     A=&FloatFromCurve(
      FloatCurve=DataInterface() {propertyValues {Curve {keys [
          {arriveTangent=4.1785, interpMode=RCIM_Cubic, time=0, value=0}
          {arriveTangent=2.3315, interpMode=RCIM_Cubic, leaveTangent=2.3315, tangentMode=RCTM_User, time=1, value=1}
         ]}}}
      CurveIndex=$Particles.RibbonLinkOrder
      Scale Curve=&Multiply_Float(
       A=$Emitter.MaxDistance
       B=&FloatFromCurve(FloatCurve=DataInterface() {propertyValues {Curve {keys [<at 0.75 1>,<at 1 0>]}}})
      )
     )
     B=&Lerp_Float(A=$Emitter.MaxDistance_RandFactorLow, B=$Emitter.MaxDistance_RandFactorHigh, Alpha=$Particles.RibbonRandom)
    )
    RibbonAxis:Vector3f=$Particles.RibbonAxis
   )
   /SetVariables(Particles.PreviousRotation:Quat4f=$Engine.Owner.Rotation) {isSetParameters=true, module=SetVariables_#8}
   &ScaleColor(
    Scale Alpha=&FloatFromCurve(FloatCurve=DataInterface() {propertyValues {Curve {keys [<at 0 0>, {arriveTangent=6.6358, leaveTangent=6.6358, time=0.5, value=1}, <at 0.7 1>, <at 1 0>]}}})
    Scale RGB=&VectorFromFloat(Value=&FloatFromCurve(FloatCurve=DataInterface() {propertyValues {Curve {keys [{arriveTangent=-319.691, interpMode=RCIM_Cubic, leaveTangent=-319.691, tangentMode=RCTM_User, time=0, value=800}, <at 0.4 15>, <at 1 1>]}}}))
    ScaleA=true
    ScaleRGB=true
   )
   &DynamicMaterialParameters(
    Index 0 Param 1=&Multiply_Float(A=$Particles.RibbonRandom, B=10)
    Index 0 Param 2=$User.EnergyFillFactor
    Module.Param0WriteEnabled=false
    Module.Param1WriteEnabled=true
    Module.Param2WriteEnabled=false
    Module.Param3WriteEnabled=false
   ) {module=DynamicMaterialParameters001}
  ]
  rendererClasses [&NiagaraRibbonRendererProperties]
  renderers [{rendererClass=&NiagaraRibbonRendererProperties}]
 }
 Right_Arm001 {
  emitterSpawnScript [&FN_GooGlobalProperties(
    EndBoneIndexArray=DataInterface() {propertyValues {GpuSyncMode=SyncCpuToGpu, MaxElements=0, IntData [0]}}
    EndBoneOffsetArray=DataInterface() {propertyValues {GpuSyncMode=SyncCpuToGpu, MaxElements=0, FloatData [<0 0 0>]}}
    MaxDistance:NiagaraFloat=$User.ArmRibbonLength
    MaxDistanceCompressed=10
    MaxDistance_RandFactorHigh=1
    MaxDistance_RandFactorLow=1
    NumberOfRibbons=2
    ParticlesPerRibbon=2
    RibbonLength:NiagaraFloat=$User.ArmRibbonLength
    RibbonLengthRandHigh=1
    RibbonLengthRandLow=1
    StartBoneIndexArray=DataInterface() {propertyValues {GpuSyncMode=SyncCpuToGpu, MaxElements=0, IntData [3,4]}}
    StartBoneOffsetArray=DataInterface() {propertyValues {GpuSyncMode=SyncCpuToGpu, MaxElements=0, FloatData [<-10 0 0>, <-3 0 0>]}}
   )]
  emitterUpdateScript [
   &EmitterState()
   &SpawnBurst_Instantaneous(
    Spawn Count=&Multiply_Int(A=$Emitter.ParticlesPerRibbon, B=$Emitter.NumberOfRibbons)
    Spawn Probability=$User.SpawnJacketFX
    Use Spawn Probability=true
    UseLoopCountLimit=false
   )
  ]
  particleSpawnScript [
   /SetVariables(Particles.SpawnGroup=&Divide_Int(A:NiagaraInt32=&ReturnExecIndex(), B:NiagaraInt32=$Emitter.ParticlesPerRibbon)) {isSetParameters=true, module=SetVariables_#9}
   /SetVariables(
    Particles.RibbonID=&MakeRibbonIDFromSpawnGroup()
    Particles.RibbonLinkOrder:NiagaraFloat=&Divide_Float(
     A=&MakeFloatFromInt(Int=&ModuloInt(A=&ReturnExecIndex(), B=$Emitter.ParticlesPerRibbon))
     B=&Subtract_Float(A=&MakeFloatFromInt(Int=$Emitter.ParticlesPerRibbon), B=1)
    )
   ) {isSetParameters=true, module=SetVariables_#10}
   /SetVariables(Particles.RibbonRandom:NiagaraFloat=&RandomRangeFloat(
     Minimum=0
     Maximum=1
     Randomness Mode=::NewEnumerator1
     Random Seed=&Add_Integer(A=$Engine.System.RandomSeed, B=$Particles.SpawnGroup)
     Fixed Random Seed=true
    )) {isSetParameters=true, module=SetVariables_#3}
   &InitializeRibbon(
    Color=&MultiplyLinearColor(
     Linear Color=$User.BariumPaintColor
     Scale Factor=&MakeVector4_FromLinearColor(LinearColor=&ColorFromCurve(
       ColorCurve=DataInterface() {propertyValues {
         AlphaCurve {keys [<at 0.574 1 RCIM_Cubic>,<at 0.9977 0>]}
         BlueCurve {keys [<at 0 1 RCIM_Cubic>,<at 1 1 RCIM_Cubic>]}
         GreenCurve {keys [<at 0 1 RCIM_Cubic>,<at 1 1 RCIM_Cubic>]}
         RedCurve {keys [<at 0 1 RCIM_Cubic>,<at 1 1 RCIM_Cubic>]}
        }}
       CurveIndex=$Particles.RibbonLinkOrder
      ))
    )
    Lifetime=&Lerp_Float(A=2, B=3, Alpha=$Particles.RibbonRandom)
    Ribbon Width=&FloatFromCurve(
     FloatCurve=DataInterface() {propertyValues {Curve {keys [<at 0 1>,<at 0.5 0.5>,<at 1 1>]}}}
     CurveIndex=$Particles.RibbonLinkOrder
     Scale Curve=$User.ArmRibbonWidth
    )
    UsePositionOffset=false
    Write Material Random=true
   )
   /SetVariables(Particles.MaterialRandom:NiagaraFloat=$Particles.RibbonRandom) {isSetParameters=true, module=SetVariables_#4}
   &SkeletalMeshLocation(
    Bone / Socket Index=&SelectIntFromArray(Direct Array Index=$Particles.SpawnGroup, Direct Array Mode=::NewEnumerator1, Int Selection Array=$Emitter.StartBoneIndexArray)
    Bone Sampling Mode=::NewEnumerator4
    MaskSpawnGroup=false
    Orientation Sampling=::NewEnumerator1
    Position Sampling=::NewEnumerator2
    Sampled Position Offset=&SelectVectorFromArray(Direct Array Index=$Particles.SpawnGroup, Direct Array Mode=::NewEnumerator1, Vector Selection Array=$Emitter.StartBoneOffsetArray)
    Skeletal Mesh=$User.SkeletalMesh
    Use Sampled Mesh Attribute as Ribbon ID=false
   )
   &FN_GooCalculateGoalPosition(
    Axis:Vector3f=$Emitter.RibbonAxis
    Scale:Vector3f=&Multiply_VectorByFloat(
     Vector=&Multiply_VectorByFloat(
      Vector=&VectorFromFloat(Value=&Multiply_Float(A=$Emitter.RibbonLength, B=0))
      Float=&Lerp_Float(A=$Emitter.RibbonLengthRandMin, B=$Emitter.RibbonLengthRandMax, Alpha=$Particles.RibbonRandom)
     )
     Float=$Particles.RibbonLinkOrder
    )
    StartPosition:Vector3f=$Particles.SkeletalMeshLocation.SampledPosition
   ) {module=FN_CalculateGoalPosition}
   /SetVariables(Particles.Position=$Particles.GoalPosition) {isSetParameters=true, module=SetVariables_#5}
   /SetVariables(Particles.PreviousRotation:Quat4f=$Engine.Owner.Rotation) {isSetParameters=true, module=SetVariables_#6}
   &DynamicMaterialParameters(
    Index 0 Param 1=&Multiply_Float(A=$Particles.RibbonRandom, B=10)
    Module.Param0WriteEnabled=true
    Module.Param1WriteEnabled=true
    Module.Param2WriteEnabled=true
    Module.Param3WriteEnabled=true
   )
   /SetVariables(Particles.SpriteSize=<10 10>) {isSetParameters=true, module=SetVariables_#7}
  ]
  particleUpdateScript [
   &ParticleState()
   &SkeletalMeshLocation(
    Bone / Socket Index=&SelectIntFromArray(Direct Array Index=$Particles.SpawnGroup, Direct Array Mode=::NewEnumerator1, Int Selection Array=$Emitter.StartBoneIndexArray)
    Bone Sampling Mode=::NewEnumerator4
    MaskSpawnGroup=false
    Orientation Sampling=::NewEnumerator1
    Position Sampling=::NewEnumerator2
    Sampled Position Offset=&SelectVectorFromArray(Direct Array Index=$Particles.SpawnGroup, Direct Array Mode=::NewEnumerator1, Vector Selection Array=$Emitter.StartBoneOffsetArray)
    Skeletal Mesh=$User.SkeletalMesh
    Use Sampled Mesh Attribute as Ribbon ID=false
   ) {module=SkeletalMeshLocation001}
   &SolveForcesAndVelocity(Clamp Velocity=false, Limit Acceleration=false)
   &FN_GooCalculateGoalPosition(
    Axis:Vector3f=&SlerpVector(
     Vector A:Vector3f=&SlerpVector(
      Vector A:Vector3f=<-1 0 0>
      Vector B:Vector3f=&TransformVector(Vector=$User.WorldSpaceWindVector, Source Space=::World, Destination Space=::Local)
      Alpha:NiagaraFloat=$User.WorldSpaceWindVectorFactor
      Scale:NiagaraFloat=1
     )
     Vector B:Vector3f=<0 0 1>
     Alpha:NiagaraFloat=&FloatFromCurve(
      FloatCurve=DataInterface() {propertyValues {Curve {keys [{tangentMode=RCTM_User, time=0, value=0}, {tangentMode=RCTM_User, time=1, value=1}]}}}
      CurveIndex=$Particles.RibbonLinkOrder
      Scale Curve=0.7
     )
     Scale:NiagaraFloat=1
    )
    CustomOrientation:Quat4f=$Engine.Owner.Rotation
    Scale:Vector3f=&Multiply_VectorByFloat(
     Vector=&Multiply_VectorByFloat(
      Vector=&VectorFromFloat(Value=&Multiply_Float(
        A=$Emitter.RibbonLength
        B=&FloatFromCurve(FloatCurve=DataInterface() {propertyValues {Curve {keys [
             {arriveTangent=17.1487, interpMode=RCIM_Cubic, leaveTangent=17.1487, tangentMode=RCTM_User, time=0, value=0}
             {interpMode=RCIM_Cubic, tangentMode=RCTM_User, time=0.2, value=1.5}
             <at 0.25 1 RCIM_Cubic>
             {arriveTangent=-3.4902, interpMode=RCIM_Cubic, leaveTangent=-3.4902, tangentMode=RCTM_User, time=1, value=0}
            ]}}})
       ))
      Float=&Lerp_Float(A=$Emitter.RibbonLengthRandMin, B=$Emitter.RibbonLengthRandMax, Alpha=$Particles.RibbonRandom)
     )
     Float=$Particles.RibbonLinkOrder
    )
    StartPosition:Vector3f=$Particles.SkeletalMeshLocation001.SampledPosition
    Use Quaternion Rotation=true
   ) {module=FN_CalculateGoalPosition001}
   &FN_GooApproachGoalPosition(
    AttractionForce:NiagaraFloat=&Lerp_Float(A=$Emitter.ResolveSpeedLow, B=$Emitter.ResolveSpeedHigh, Alpha=$Particles.RibbonRandom)
    GoalPosition:NiagaraPosition=$Particles.GoalPosition
    MaximumDistance:NiagaraFloat=&Multiply_Float(
     A=&FloatFromCurve(
      FloatCurve=DataInterface() {propertyValues {Curve {keys [
          {arriveTangent=4.1785, interpMode=RCIM_Cubic, time=0, value=0}
          {arriveTangent=2.3315, interpMode=RCIM_Cubic, leaveTangent=2.3315, tangentMode=RCTM_User, time=1, value=1}
         ]}}}
      CurveIndex=$Particles.RibbonLinkOrder
      Scale Curve=&Multiply_Float(
       A=$Emitter.MaxDistance
       B=&FloatFromCurve(FloatCurve=DataInterface() {propertyValues {Curve {keys [<at 0.75 1>,<at 1 0>]}}})
      )
     )
     B=&Lerp_Float(A=$Emitter.MaxDistance_RandFactorLow, B=$Emitter.MaxDistance_RandFactorHigh, Alpha=$Particles.RibbonRandom)
    )
    RibbonAxis:Vector3f=$Particles.RibbonAxis
   )
   /SetVariables(Particles.PreviousRotation:Quat4f=$Engine.Owner.Rotation) {isSetParameters=true, module=SetVariables_#8}
   &ScaleColor(
    Scale Alpha=&FloatFromCurve(FloatCurve=DataInterface() {propertyValues {Curve {keys [<at 0 0>, {arriveTangent=6.6358, leaveTangent=6.6358, time=0.5, value=1}, <at 0.7 1>, <at 1 0>]}}})
    Scale RGB=&VectorFromFloat(Value=&FloatFromCurve(FloatCurve=DataInterface() {propertyValues {Curve {keys [{arriveTangent=-319.691, interpMode=RCIM_Cubic, leaveTangent=-319.691, tangentMode=RCTM_User, time=0, value=800}, <at 0.4 15>, <at 1 1>]}}}))
    ScaleA=true
    ScaleRGB=true
   )
   &DynamicMaterialParameters(
    Index 0 Param 1=&Multiply_Float(A=$Particles.RibbonRandom, B=10)
    Index 0 Param 2=$User.EnergyFillFactor
    Module.Param0WriteEnabled=false
    Module.Param1WriteEnabled=true
    Module.Param2WriteEnabled=false
    Module.Param3WriteEnabled=false
   ) {module=DynamicMaterialParameters001}
  ]
  rendererClasses [&NiagaraRibbonRendererProperties]
  renderers [{rendererClass=&NiagaraRibbonRendererProperties}]
 }
 Right_Hand {
  emitterSpawnScript [&FN_GooGlobalProperties(
    EndBoneIndexArray=DataInterface() {propertyValues {GpuSyncMode=SyncCpuToGpu, MaxElements=0, IntData [0]}}
    EndBoneOffsetArray=DataInterface() {propertyValues {GpuSyncMode=SyncCpuToGpu, MaxElements=0, FloatData [<0 0 0>]}}
    MaxDistance:NiagaraFloat=$User.ArmRibbonLength
    MaxDistanceCompressed=10
    MaxDistance_RandFactorHigh=1
    MaxDistance_RandFactorLow=1
    RibbonLength:NiagaraFloat=$User.ArmRibbonLength
    RibbonLengthRandHigh=1
    RibbonLengthRandLow=1
    StartBoneIndexArray=DataInterface() {propertyValues {GpuSyncMode=SyncCpuToGpu, MaxElements=0, IntData [5]}}
    StartBoneOffsetArray=DataInterface() {propertyValues {GpuSyncMode=SyncCpuToGpu, MaxElements=0, FloatData [<-3 0 0>]}}
   )]
  emitterUpdateScript [
   &EmitterState()
   &SpawnBurst_Instantaneous(
    Spawn Count=&Multiply_Int(A=$Emitter.ParticlesPerRibbon, B=$Emitter.NumberOfRibbons)
    Use Spawn Probability=false
    UseLoopCountLimit=false
   )
  ]
  particleSpawnScript [
   /SetVariables(Particles.SpawnGroup=&Divide_Int(A:NiagaraInt32=&ReturnExecIndex(), B:NiagaraInt32=$Emitter.ParticlesPerRibbon)) {isSetParameters=true, module=SetVariables_#9}
   /SetVariables(
    Particles.RibbonID=&MakeRibbonIDFromSpawnGroup()
    Particles.RibbonLinkOrder:NiagaraFloat=&Divide_Float(
     A=&MakeFloatFromInt(Int=&ModuloInt(A=&ReturnExecIndex(), B=$Emitter.ParticlesPerRibbon))
     B=&Subtract_Float(A=&MakeFloatFromInt(Int=$Emitter.ParticlesPerRibbon), B=1)
    )
   ) {isSetParameters=true, module=SetVariables_#10}
   /SetVariables(Particles.RibbonRandom:NiagaraFloat=&RandomRangeFloat(
     Minimum=0
     Maximum=1
     Randomness Mode=::NewEnumerator1
     Random Seed=&Add_Integer(A=$Engine.System.RandomSeed, B=$Particles.SpawnGroup)
     Fixed Random Seed=true
    )) {isSetParameters=true, module=SetVariables_#3}
   &InitializeRibbon(
    Color=&MultiplyLinearColor(
     Linear Color=$User.BariumPaintColor
     Scale Factor=&MakeVector4_FromLinearColor(LinearColor=&ColorFromCurve(
       ColorCurve=DataInterface() {propertyValues {
         AlphaCurve {keys [<at 0.574 1 RCIM_Cubic>,<at 0.9977 0>]}
         BlueCurve {keys [<at 0 1 RCIM_Cubic>,<at 1 1 RCIM_Cubic>]}
         GreenCurve {keys [<at 0 1 RCIM_Cubic>,<at 1 1 RCIM_Cubic>]}
         RedCurve {keys [<at 0 1 RCIM_Cubic>,<at 1 1 RCIM_Cubic>]}
        }}
       CurveIndex=$Particles.RibbonLinkOrder
      ))
    )
    Lifetime=&Lerp_Float(A=2, B=3, Alpha=$Particles.RibbonRandom)
    Ribbon Width=&FloatFromCurve(
     FloatCurve=DataInterface() {propertyValues {Curve {keys [<at 0 1>,<at 0.5 0.5>,<at 1 1>]}}}
     CurveIndex=$Particles.RibbonLinkOrder
     Scale Curve=15
    )
    UsePositionOffset=false
    Write Material Random=true
   )
   /SetVariables(Particles.MaterialRandom:NiagaraFloat=$Particles.RibbonRandom) {isSetParameters=true, module=SetVariables_#4}
   &SkeletalMeshLocation(
    Bone / Socket Index=&SelectIntFromArray(Direct Array Index=$Particles.SpawnGroup, Direct Array Mode=::NewEnumerator1, Int Selection Array=$Emitter.StartBoneIndexArray)
    Bone Sampling Mode=::NewEnumerator4
    MaskSpawnGroup=false
    Orientation Sampling=::NewEnumerator1
    Position Sampling=::NewEnumerator2
    Sampled Position Offset=&SelectVectorFromArray(Direct Array Index=$Particles.SpawnGroup, Direct Array Mode=::NewEnumerator1, Vector Selection Array=$Emitter.StartBoneOffsetArray)
    Skeletal Mesh=$User.SkeletalMesh
    Use Sampled Mesh Attribute as Ribbon ID=false
   )
   &FN_GooCalculateGoalPosition(
    Axis:Vector3f=$Emitter.RibbonAxis
    Scale:Vector3f=&Multiply_VectorByFloat(
     Vector=&Multiply_VectorByFloat(
      Vector=&VectorFromFloat(Value=&Multiply_Float(A=$Emitter.RibbonLength, B=0))
      Float=&Lerp_Float(A=$Emitter.RibbonLengthRandMin, B=$Emitter.RibbonLengthRandMax, Alpha=$Particles.RibbonRandom)
     )
     Float=$Particles.RibbonLinkOrder
    )
    StartPosition:Vector3f=$Particles.SkeletalMeshLocation.SampledPosition
   ) {module=FN_CalculateGoalPosition}
   /SetVariables(Particles.Position=$Particles.GoalPosition) {isSetParameters=true, module=SetVariables_#5}
   /SetVariables(Particles.PreviousRotation:Quat4f=$Engine.Owner.Rotation) {isSetParameters=true, module=SetVariables_#6}
   &DynamicMaterialParameters(
    Index 0 Param 1=&Multiply_Float(A=$Particles.RibbonRandom, B=10)
    Module.Param0WriteEnabled=true
    Module.Param1WriteEnabled=true
    Module.Param2WriteEnabled=true
    Module.Param3WriteEnabled=true
   )
   /SetVariables(Particles.SpriteSize=<10 10>) {isSetParameters=true, module=SetVariables_#7}
  ]
  particleUpdateScript [
   &ParticleState()
   &SkeletalMeshLocation(
    Bone / Socket Index=&SelectIntFromArray(Direct Array Index=$Particles.SpawnGroup, Direct Array Mode=::NewEnumerator1, Int Selection Array=$Emitter.StartBoneIndexArray)
    Bone Sampling Mode=::NewEnumerator4
    MaskSpawnGroup=false
    Orientation Sampling=::NewEnumerator1
    Position Sampling=::NewEnumerator2
    Sampled Position Offset=&SelectVectorFromArray(Direct Array Index=$Particles.SpawnGroup, Direct Array Mode=::NewEnumerator1, Vector Selection Array=$Emitter.StartBoneOffsetArray)
    Skeletal Mesh=$User.SkeletalMesh
    Use Sampled Mesh Attribute as Ribbon ID=false
   ) {module=SkeletalMeshLocation001}
   &SolveForcesAndVelocity(Clamp Velocity=false, Limit Acceleration=false)
   &FN_GooCalculateGoalPosition(
    Axis:Vector3f=&SlerpVector(
     Vector A:Vector3f=&SlerpVector(
      Vector A:Vector3f=<-1 0 0>
      Vector B:Vector3f=&TransformVector(Vector=$User.WorldSpaceWindVector, Source Space=::World, Destination Space=::Local)
      Alpha:NiagaraFloat=$User.WorldSpaceWindVectorFactor
      Scale:NiagaraFloat=1
     )
     Vector B:Vector3f=<0 0 1>
     Alpha:NiagaraFloat=&FloatFromCurve(
      FloatCurve=DataInterface() {propertyValues {Curve {keys [{tangentMode=RCTM_User, time=0, value=0}, {tangentMode=RCTM_User, time=1, value=1}]}}}
      CurveIndex=$Particles.RibbonLinkOrder
      Scale Curve=0.7
     )
     Scale:NiagaraFloat=1
    )
    CustomOrientation:Quat4f=$Engine.Owner.Rotation
    Scale:Vector3f=&Multiply_VectorByFloat(
     Vector=&Multiply_VectorByFloat(
      Vector=&VectorFromFloat(Value=&Multiply_Float(
        A=$Emitter.RibbonLength
        B=&FloatFromCurve(FloatCurve=DataInterface() {propertyValues {Curve {keys [
             {arriveTangent=17.1487, interpMode=RCIM_Cubic, leaveTangent=17.1487, tangentMode=RCTM_User, time=0, value=0}
             {interpMode=RCIM_Cubic, tangentMode=RCTM_User, time=0.2, value=1.5}
             <at 0.25 1 RCIM_Cubic>
             {arriveTangent=-3.4902, interpMode=RCIM_Cubic, leaveTangent=-3.4902, tangentMode=RCTM_User, time=1, value=0}
            ]}}})
       ))
      Float=&Lerp_Float(A=$Emitter.RibbonLengthRandMin, B=$Emitter.RibbonLengthRandMax, Alpha=$Particles.RibbonRandom)
     )
     Float=$Particles.RibbonLinkOrder
    )
    StartPosition:Vector3f=$Particles.SkeletalMeshLocation001.SampledPosition
    Use Quaternion Rotation=true
   ) {module=FN_CalculateGoalPosition001}
   &FN_GooApproachGoalPosition(
    AttractionForce:NiagaraFloat=&Lerp_Float(A=$Emitter.ResolveSpeedLow, B=$Emitter.ResolveSpeedHigh, Alpha=$Particles.RibbonRandom)
    GoalPosition:NiagaraPosition=$Particles.GoalPosition
    MaximumDistance:NiagaraFloat=&Multiply_Float(
     A=&FloatFromCurve(
      FloatCurve=DataInterface() {propertyValues {Curve {keys [
          {arriveTangent=4.1785, interpMode=RCIM_Cubic, time=0, value=0}
          {arriveTangent=2.3315, interpMode=RCIM_Cubic, leaveTangent=2.3315, tangentMode=RCTM_User, time=1, value=1}
         ]}}}
      CurveIndex=$Particles.RibbonLinkOrder
      Scale Curve=&Multiply_Float(
       A=$Emitter.MaxDistance
       B=&FloatFromCurve(FloatCurve=DataInterface() {propertyValues {Curve {keys [<at 0.75 1>,<at 1 0>]}}})
      )
     )
     B=&Lerp_Float(A=$Emitter.MaxDistance_RandFactorLow, B=$Emitter.MaxDistance_RandFactorHigh, Alpha=$Particles.RibbonRandom)
    )
    RibbonAxis:Vector3f=$Particles.RibbonAxis
   )
   /SetVariables(Particles.PreviousRotation:Quat4f=$Engine.Owner.Rotation) {isSetParameters=true, module=SetVariables_#8}
   &ScaleColor(
    Scale Alpha=&FloatFromCurve(FloatCurve=DataInterface() {propertyValues {Curve {keys [<at 0 0>, {arriveTangent=6.6358, leaveTangent=6.6358, time=0.5, value=1}, <at 0.7 1>, <at 1 0>]}}})
    Scale RGB=&VectorFromFloat(Value=&FloatFromCurve(FloatCurve=DataInterface() {propertyValues {Curve {keys [{arriveTangent=-319.691, interpMode=RCIM_Cubic, leaveTangent=-319.691, tangentMode=RCTM_User, time=0, value=800}, <at 0.4 15>, <at 1 1>]}}}))
    ScaleA=true
    ScaleRGB=true
   )
   &DynamicMaterialParameters(
    Index 0 Param 1=&Multiply_Float(A=$Particles.RibbonRandom, B=10)
    Index 0 Param 2=$User.EnergyFillFactor
    Module.Param0WriteEnabled=false
    Module.Param1WriteEnabled=true
    Module.Param2WriteEnabled=false
    Module.Param3WriteEnabled=false
   ) {module=DynamicMaterialParameters001}
  ]
  rendererClasses [&NiagaraRibbonRendererProperties]
  renderers [{rendererClass=&NiagaraRibbonRendererProperties}]
 }
 Right_Hand001 {
  emitterSpawnScript [&FN_GooGlobalProperties(
    EndBoneIndexArray=DataInterface() {propertyValues {GpuSyncMode=SyncCpuToGpu, MaxElements=0, IntData [0]}}
    EndBoneOffsetArray=DataInterface() {propertyValues {GpuSyncMode=SyncCpuToGpu, MaxElements=0, FloatData [<0 0 0>]}}
    MaxDistance:NiagaraFloat=$User.ArmRibbonLength
    MaxDistanceCompressed=10
    MaxDistance_RandFactorHigh=1
    MaxDistance_RandFactorLow=1
    ParticlesPerRibbon=2
    RibbonLength:NiagaraFloat=$User.ArmRibbonLength
    RibbonLengthRandHigh=1
    RibbonLengthRandLow=1
    StartBoneIndexArray=DataInterface() {propertyValues {GpuSyncMode=SyncCpuToGpu, MaxElements=0, IntData [5]}}
    StartBoneOffsetArray=DataInterface() {propertyValues {GpuSyncMode=SyncCpuToGpu, MaxElements=0, FloatData [<-3 0 0>]}}
   )]
  emitterUpdateScript [
   &EmitterState()
   &SpawnBurst_Instantaneous(
    Spawn Count=&Multiply_Int(A=$Emitter.ParticlesPerRibbon, B=$Emitter.NumberOfRibbons)
    Use Spawn Probability=false
    UseLoopCountLimit=false
   )
  ]
  particleSpawnScript [
   /SetVariables(Particles.SpawnGroup=&Divide_Int(A:NiagaraInt32=&ReturnExecIndex(), B:NiagaraInt32=$Emitter.ParticlesPerRibbon)) {isSetParameters=true, module=SetVariables_#9}
   /SetVariables(
    Particles.RibbonID=&MakeRibbonIDFromSpawnGroup()
    Particles.RibbonLinkOrder:NiagaraFloat=&Divide_Float(
     A=&MakeFloatFromInt(Int=&ModuloInt(A=&ReturnExecIndex(), B=$Emitter.ParticlesPerRibbon))
     B=&Subtract_Float(A=&MakeFloatFromInt(Int=$Emitter.ParticlesPerRibbon), B=1)
    )
   ) {isSetParameters=true, module=SetVariables_#10}
   /SetVariables(Particles.RibbonRandom:NiagaraFloat=&RandomRangeFloat(
     Minimum=0
     Maximum=1
     Randomness Mode=::NewEnumerator1
     Random Seed=&Add_Integer(A=$Engine.System.RandomSeed, B=$Particles.SpawnGroup)
     Fixed Random Seed=true
    )) {isSetParameters=true, module=SetVariables_#3}
   &InitializeRibbon(
    Color=&MultiplyLinearColor(
     Linear Color=$User.BariumPaintColor
     Scale Factor=&MakeVector4_FromLinearColor(LinearColor=&ColorFromCurve(
       ColorCurve=DataInterface() {propertyValues {
         AlphaCurve {keys [<at 0.574 1 RCIM_Cubic>,<at 0.9977 0>]}
         BlueCurve {keys [<at 0 1 RCIM_Cubic>,<at 1 1 RCIM_Cubic>]}
         GreenCurve {keys [<at 0 1 RCIM_Cubic>,<at 1 1 RCIM_Cubic>]}
         RedCurve {keys [<at 0 1 RCIM_Cubic>,<at 1 1 RCIM_Cubic>]}
        }}
       CurveIndex=$Particles.RibbonLinkOrder
      ))
    )
    Lifetime=&Lerp_Float(A=2, B=3, Alpha=$Particles.RibbonRandom)
    Ribbon Width=&FloatFromCurve(
     FloatCurve=DataInterface() {propertyValues {Curve {keys [<at 0 1>,<at 0.5 0.5>,<at 1 1>]}}}
     CurveIndex=$Particles.RibbonLinkOrder
     Scale Curve=15
    )
    UsePositionOffset=false
    Write Material Random=true
   )
   /SetVariables(Particles.MaterialRandom:NiagaraFloat=$Particles.RibbonRandom) {isSetParameters=true, module=SetVariables_#4}
   &SkeletalMeshLocation(
    Bone / Socket Index=&SelectIntFromArray(Direct Array Index=$Particles.SpawnGroup, Direct Array Mode=::NewEnumerator1, Int Selection Array=$Emitter.StartBoneIndexArray)
    Bone Sampling Mode=::NewEnumerator4
    MaskSpawnGroup=false
    Orientation Sampling=::NewEnumerator1
    Position Sampling=::NewEnumerator2
    Sampled Position Offset=&SelectVectorFromArray(Direct Array Index=$Particles.SpawnGroup, Direct Array Mode=::NewEnumerator1, Vector Selection Array=$Emitter.StartBoneOffsetArray)
    Skeletal Mesh=$User.SkeletalMesh
    Use Sampled Mesh Attribute as Ribbon ID=false
   )
   &FN_GooCalculateGoalPosition(
    Axis:Vector3f=$Emitter.RibbonAxis
    Scale:Vector3f=&Multiply_VectorByFloat(
     Vector=&Multiply_VectorByFloat(
      Vector=&VectorFromFloat(Value=&Multiply_Float(A=$Emitter.RibbonLength, B=0))
      Float=&Lerp_Float(A=$Emitter.RibbonLengthRandMin, B=$Emitter.RibbonLengthRandMax, Alpha=$Particles.RibbonRandom)
     )
     Float=$Particles.RibbonLinkOrder
    )
    StartPosition:Vector3f=$Particles.SkeletalMeshLocation.SampledPosition
   ) {module=FN_CalculateGoalPosition}
   /SetVariables(Particles.Position=$Particles.GoalPosition) {isSetParameters=true, module=SetVariables_#5}
   /SetVariables(Particles.PreviousRotation:Quat4f=$Engine.Owner.Rotation) {isSetParameters=true, module=SetVariables_#6}
   &DynamicMaterialParameters(
    Index 0 Param 1=&Multiply_Float(A=$Particles.RibbonRandom, B=10)
    Module.Param0WriteEnabled=true
    Module.Param1WriteEnabled=true
    Module.Param2WriteEnabled=true
    Module.Param3WriteEnabled=true
   )
   /SetVariables(Particles.SpriteSize=<10 10>) {isSetParameters=true, module=SetVariables_#7}
  ]
  particleUpdateScript [
   &ParticleState()
   &SkeletalMeshLocation(
    Bone / Socket Index=&SelectIntFromArray(Direct Array Index=$Particles.SpawnGroup, Direct Array Mode=::NewEnumerator1, Int Selection Array=$Emitter.StartBoneIndexArray)
    Bone Sampling Mode=::NewEnumerator4
    MaskSpawnGroup=false
    Orientation Sampling=::NewEnumerator1
    Position Sampling=::NewEnumerator2
    Sampled Position Offset=&SelectVectorFromArray(Direct Array Index=$Particles.SpawnGroup, Direct Array Mode=::NewEnumerator1, Vector Selection Array=$Emitter.StartBoneOffsetArray)
    Skeletal Mesh=$User.SkeletalMesh
    Use Sampled Mesh Attribute as Ribbon ID=false
   ) {module=SkeletalMeshLocation001}
   &SolveForcesAndVelocity(Clamp Velocity=false, Limit Acceleration=false)
   &FN_GooCalculateGoalPosition(
    Axis:Vector3f=&SlerpVector(
     Vector A:Vector3f=&SlerpVector(
      Vector A:Vector3f=<-1 0 0>
      Vector B:Vector3f=&TransformVector(Vector=$User.WorldSpaceWindVector, Source Space=::World, Destination Space=::Local)
      Alpha:NiagaraFloat=$User.WorldSpaceWindVectorFactor
      Scale:NiagaraFloat=1
     )
     Vector B:Vector3f=<0 0 1>
     Alpha:NiagaraFloat=&FloatFromCurve(
      FloatCurve=DataInterface() {propertyValues {Curve {keys [{tangentMode=RCTM_User, time=0, value=0}, {tangentMode=RCTM_User, time=1, value=1}]}}}
      CurveIndex=$Particles.RibbonLinkOrder
      Scale Curve=0.7
     )
     Scale:NiagaraFloat=1
    )
    CustomOrientation:Quat4f=$Engine.Owner.Rotation
    Scale:Vector3f=&Multiply_VectorByFloat(
     Vector=&Multiply_VectorByFloat(
      Vector=&VectorFromFloat(Value=&Multiply_Float(
        A=$Emitter.RibbonLength
        B=&FloatFromCurve(FloatCurve=DataInterface() {propertyValues {Curve {keys [
             {arriveTangent=17.1487, interpMode=RCIM_Cubic, leaveTangent=17.1487, tangentMode=RCTM_User, time=0, value=0}
             {interpMode=RCIM_Cubic, tangentMode=RCTM_User, time=0.2, value=1.5}
             <at 0.25 1 RCIM_Cubic>
             {arriveTangent=-3.4902, interpMode=RCIM_Cubic, leaveTangent=-3.4902, tangentMode=RCTM_User, time=1, value=0}
            ]}}})
       ))
      Float=&Lerp_Float(A=$Emitter.RibbonLengthRandMin, B=$Emitter.RibbonLengthRandMax, Alpha=$Particles.RibbonRandom)
     )
     Float=$Particles.RibbonLinkOrder
    )
    StartPosition:Vector3f=$Particles.SkeletalMeshLocation001.SampledPosition
    Use Quaternion Rotation=true
   ) {module=FN_CalculateGoalPosition001}
   &FN_GooApproachGoalPosition(
    AttractionForce:NiagaraFloat=&Lerp_Float(A=$Emitter.ResolveSpeedLow, B=$Emitter.ResolveSpeedHigh, Alpha=$Particles.RibbonRandom)
    GoalPosition:NiagaraPosition=$Particles.GoalPosition
    MaximumDistance:NiagaraFloat=&Multiply_Float(
     A=&FloatFromCurve(
      FloatCurve=DataInterface() {propertyValues {Curve {keys [
          {arriveTangent=4.1785, interpMode=RCIM_Cubic, time=0, value=0}
          {arriveTangent=2.3315, interpMode=RCIM_Cubic, leaveTangent=2.3315, tangentMode=RCTM_User, time=1, value=1}
         ]}}}
      CurveIndex=$Particles.RibbonLinkOrder
      Scale Curve=&Multiply_Float(
       A=$Emitter.MaxDistance
       B=&FloatFromCurve(FloatCurve=DataInterface() {propertyValues {Curve {keys [<at 0.75 1>,<at 1 0>]}}})
      )
     )
     B=&Lerp_Float(A=$Emitter.MaxDistance_RandFactorLow, B=$Emitter.MaxDistance_RandFactorHigh, Alpha=$Particles.RibbonRandom)
    )
    RibbonAxis:Vector3f=$Particles.RibbonAxis
   )
   /SetVariables(Particles.PreviousRotation:Quat4f=$Engine.Owner.Rotation) {isSetParameters=true, module=SetVariables_#8}
   &ScaleColor(
    Scale Alpha=&FloatFromCurve(FloatCurve=DataInterface() {propertyValues {Curve {keys [<at 0 0>, {arriveTangent=6.6358, leaveTangent=6.6358, time=0.5, value=1}, <at 0.7 1>, <at 1 0>]}}})
    Scale RGB=&VectorFromFloat(Value=&FloatFromCurve(FloatCurve=DataInterface() {propertyValues {Curve {keys [{arriveTangent=-319.691, interpMode=RCIM_Cubic, leaveTangent=-319.691, tangentMode=RCTM_User, time=0, value=800}, <at 0.4 15>, <at 1 1>]}}}))
    ScaleA=true
    ScaleRGB=true
   )
   &DynamicMaterialParameters(
    Index 0 Param 1=&Multiply_Float(A=$Particles.RibbonRandom, B=10)
    Index 0 Param 2=$User.EnergyFillFactor
    Module.Param0WriteEnabled=false
    Module.Param1WriteEnabled=true
    Module.Param2WriteEnabled=false
    Module.Param3WriteEnabled=false
   ) {module=DynamicMaterialParameters001}
  ]
  rendererClasses [&NiagaraRibbonRendererProperties]
  renderers [{rendererClass=&NiagaraRibbonRendererProperties}]
 }
 Right_Hip {
  emitterSpawnScript [&FN_GooGlobalProperties(
    EndBoneIndexArray=DataInterface() {propertyValues {GpuSyncMode=SyncCpuToGpu, MaxElements=0, IntData [0]}}
    EndBoneOffsetArray=DataInterface() {propertyValues {GpuSyncMode=SyncCpuToGpu, MaxElements=0, FloatData [<0 0 0>]}}
    MaxDistance:NiagaraFloat=$User.HipRibbonLength
    MaxDistanceCompressed=10
    MaxDistance_RandFactorHigh=1
    MaxDistance_RandFactorLow=1
    ParticlesPerRibbon=4
    ResolveSpeedHigh=3
    RibbonLength:NiagaraFloat=$User.HipRibbonLength
    RibbonLengthRandHigh=1
    RibbonLengthRandLow=1
    StartBoneIndexArray=DataInterface() {propertyValues {GpuSyncMode=SyncCpuToGpu, MaxElements=0, IntData [11]}}
    StartBoneOffsetArray=DataInterface() {propertyValues {GpuSyncMode=SyncCpuToGpu, MaxElements=0, FloatData [<2 0 3>]}}
   )]
  emitterUpdateScript [
   &EmitterState()
   &SpawnBurst_Instantaneous(
    Spawn Count=&Multiply_Int(A=$Emitter.ParticlesPerRibbon, B=$Emitter.NumberOfRibbons)
    Use Spawn Probability=false
    UseLoopCountLimit=false
   )
  ]
  particleSpawnScript [
   /SetVariables(Particles.SpawnGroup=&Divide_Int(A:NiagaraInt32=&ReturnExecIndex(), B:NiagaraInt32=$Emitter.ParticlesPerRibbon)) {isSetParameters=true, module=SetVariables_#9}
   /SetVariables(
    Particles.RibbonID=&MakeRibbonIDFromSpawnGroup()
    Particles.RibbonLinkOrder:NiagaraFloat=&Divide_Float(
     A=&MakeFloatFromInt(Int=&ModuloInt(A=&ReturnExecIndex(), B=$Emitter.ParticlesPerRibbon))
     B=&Subtract_Float(A=&MakeFloatFromInt(Int=$Emitter.ParticlesPerRibbon), B=1)
    )
   ) {isSetParameters=true, module=SetVariables_#10}
   /SetVariables(Particles.RibbonRandom:NiagaraFloat=&RandomRangeFloat(
     Minimum=0
     Maximum=1
     Randomness Mode=::NewEnumerator1
     Random Seed=&Add_Integer(A=$Engine.System.RandomSeed, B=$Particles.SpawnGroup)
     Fixed Random Seed=true
    )) {isSetParameters=true, module=SetVariables_#3}
   &InitializeRibbon(
    Color=&MultiplyLinearColor(
     Linear Color=$User.BariumPaintColor
     Scale Factor=&MakeVector4_FromLinearColor(LinearColor=&ColorFromCurve(
       ColorCurve=DataInterface() {propertyValues {
         AlphaCurve {keys [<at 0.574 1 RCIM_Cubic>,<at 0.9977 0>]}
         BlueCurve {keys [<at 0 1 RCIM_Cubic>,<at 1 1 RCIM_Cubic>]}
         GreenCurve {keys [<at 0 1 RCIM_Cubic>,<at 1 1 RCIM_Cubic>]}
         RedCurve {keys [<at 0 1 RCIM_Cubic>,<at 1 1 RCIM_Cubic>]}
        }}
       CurveIndex=$Particles.RibbonLinkOrder
      ))
    )
    Lifetime=&Lerp_Float(A=2, B=3, Alpha=$Particles.RibbonRandom)
    Ribbon Width=$User.HipRibbonWidth
    UsePositionOffset=false
    Write Material Random=true
   )
   /SetVariables(Particles.MaterialRandom:NiagaraFloat=$Particles.RibbonRandom) {isSetParameters=true, module=SetVariables_#4}
   &SkeletalMeshLocation(
    Bone / Socket Index=&SelectIntFromArray(Direct Array Index=$Particles.SpawnGroup, Direct Array Mode=::NewEnumerator1, Int Selection Array=$Emitter.StartBoneIndexArray)
    Bone Sampling Mode=::NewEnumerator4
    MaskSpawnGroup=false
    Orientation Sampling=::NewEnumerator1
    Position Sampling=::NewEnumerator2
    Sampled Position Offset=&SelectVectorFromArray(Direct Array Index=$Particles.SpawnGroup, Direct Array Mode=::NewEnumerator1, Vector Selection Array=$Emitter.StartBoneOffsetArray)
    Skeletal Mesh=$User.SkeletalMesh
    Use Sampled Mesh Attribute as Ribbon ID=false
   )
   &FN_GooCalculateGoalPosition(
    Axis:Vector3f=$Emitter.RibbonAxis
    Scale:Vector3f=&Multiply_VectorByFloat(
     Vector=&Multiply_VectorByFloat(
      Vector=&VectorFromFloat(Value=&Multiply_Float(A=$Emitter.RibbonLength, B=0))
      Float=&Lerp_Float(A=$Emitter.RibbonLengthRandMin, B=$Emitter.RibbonLengthRandMax, Alpha=$Particles.RibbonRandom)
     )
     Float=$Particles.RibbonLinkOrder
    )
    StartPosition:Vector3f=$Particles.SkeletalMeshLocation.SampledPosition
   ) {module=FN_CalculateGoalPosition}
   /SetVariables(Particles.Position=$Particles.GoalPosition) {isSetParameters=true, module=SetVariables_#5}
   /SetVariables(Particles.PreviousRotation:Quat4f=$Engine.Owner.Rotation) {isSetParameters=true, module=SetVariables_#6}
   &DynamicMaterialParameters(
    Index 0 Param 1=&Multiply_Float(A=$Particles.RibbonRandom, B=10)
    Module.Param0WriteEnabled=true
    Module.Param1WriteEnabled=true
    Module.Param2WriteEnabled=true
    Module.Param3WriteEnabled=true
   )
   /SetVariables(Particles.SpriteSize=<10 10>) {isSetParameters=true, module=SetVariables_#7}
  ]
  particleUpdateScript [
   &ParticleState()
   &SkeletalMeshLocation(
    Bone / Socket Index=&SelectIntFromArray(Direct Array Index=$Particles.SpawnGroup, Direct Array Mode=::NewEnumerator1, Int Selection Array=$Emitter.StartBoneIndexArray)
    Bone Sampling Mode=::NewEnumerator4
    MaskSpawnGroup=false
    Orientation Sampling=::NewEnumerator1
    Position Sampling=::NewEnumerator2
    Sampled Position Offset=&SelectVectorFromArray(Direct Array Index=$Particles.SpawnGroup, Direct Array Mode=::NewEnumerator1, Vector Selection Array=$Emitter.StartBoneOffsetArray)
    Skeletal Mesh=$User.SkeletalMesh
    Use Sampled Mesh Attribute as Ribbon ID=false
   ) {module=SkeletalMeshLocation001}
   &SolveForcesAndVelocity(Clamp Velocity=false, Limit Acceleration=false)
   &FN_GooCalculateGoalPosition(
    Axis:Vector3f=&SlerpVector(
     Vector A:Vector3f=&SlerpVector(
      Vector A:Vector3f=<-2 -1 0>
      Vector B:Vector3f=&TransformVector(Vector=$User.WorldSpaceWindVector, Source Space=::World, Destination Space=::Local)
      Alpha:NiagaraFloat=$User.WorldSpaceWindVectorFactor
      Scale:NiagaraFloat=1
     )
     Vector B:Vector3f=<0 0 1>
     Alpha:NiagaraFloat=&FloatFromCurve(
      FloatCurve=DataInterface() {propertyValues {Curve {keys [{tangentMode=RCTM_User, time=0, value=0}, {tangentMode=RCTM_User, time=1, value=0.5}]}}}
      CurveIndex=$Particles.RibbonLinkOrder
     )
     Scale:NiagaraFloat=1
    )
    CustomOrientation:Quat4f=$Engine.Owner.Rotation
    Scale:Vector3f=&Multiply_VectorByFloat(
     Vector=&Multiply_VectorByFloat(
      Vector=&VectorFromFloat(Value=&Multiply_Float(
        A=$Emitter.RibbonLength
        B=&FloatFromCurve(FloatCurve=DataInterface() {propertyValues {Curve {keys [
             {arriveTangent=30.27, interpMode=RCIM_Cubic, leaveTangent=30.27, tangentMode=RCTM_User, time=0, value=0}
             {interpMode=RCIM_Cubic, tangentMode=RCTM_User, time=0.15, value=1.5}
             <at 0.2 1 RCIM_Cubic>
             {arriveTangent=-3.4902, interpMode=RCIM_Cubic, leaveTangent=-3.4902, tangentMode=RCTM_User, time=1, value=0}
            ]}}})
       ))
      Float=&Lerp_Float(A=$Emitter.RibbonLengthRandMin, B=$Emitter.RibbonLengthRandMax, Alpha=$Particles.RibbonRandom)
     )
     Float=$Particles.RibbonLinkOrder
    )
    StartPosition:Vector3f=$Particles.SkeletalMeshLocation001.SampledPosition
    Use Quaternion Rotation=true
   ) {module=FN_CalculateGoalPosition001}
   &FN_GooApproachGoalPosition(
    AttractionForce:NiagaraFloat=&Lerp_Float(A=$Emitter.ResolveSpeedLow, B=$Emitter.ResolveSpeedHigh, Alpha=$Particles.RibbonRandom)
    GoalPosition:NiagaraPosition=$Particles.GoalPosition
    MaximumDistance:NiagaraFloat=&Multiply_Float(
     A=&FloatFromCurve(
      FloatCurve=DataInterface() {propertyValues {Curve {keys [
          {arriveTangent=4.1785, interpMode=RCIM_Cubic, time=0, value=0}
          {arriveTangent=2.3315, interpMode=RCIM_Cubic, leaveTangent=2.3315, tangentMode=RCTM_User, time=1, value=1}
         ]}}}
      CurveIndex=$Particles.RibbonLinkOrder
      Scale Curve=&Multiply_Float(
       A=$Emitter.MaxDistance
       B=&FloatFromCurve(FloatCurve=DataInterface() {propertyValues {Curve {keys [<at 0.75 1>,<at 1 0>]}}})
      )
     )
     B=&Lerp_Float(A=$Emitter.MaxDistance_RandFactorLow, B=$Emitter.MaxDistance_RandFactorHigh, Alpha=$Particles.RibbonRandom)
    )
    RibbonAxis:Vector3f=$Particles.RibbonAxis
   )
   /SetVariables(Particles.PreviousRotation:Quat4f=$Engine.Owner.Rotation) {isSetParameters=true, module=SetVariables_#8}
   &ScaleColor(
    Scale Alpha=&FloatFromCurve(FloatCurve=DataInterface() {propertyValues {Curve {keys [<at 0 0>, {arriveTangent=6.6358, leaveTangent=6.6358, time=0.5, value=1}, <at 0.7 1>, <at 1 0>]}}})
    Scale RGB=&VectorFromFloat(Value=&FloatFromCurve(FloatCurve=DataInterface() {propertyValues {Curve {keys [{arriveTangent=-319.691, interpMode=RCIM_Cubic, leaveTangent=-319.691, tangentMode=RCTM_User, time=0, value=800}, <at 0.4 15>, <at 1 1>]}}}))
    ScaleA=true
    ScaleRGB=true
   )
   &DynamicMaterialParameters(
    Index 0 Param 1=&Multiply_Float(A=$Particles.RibbonRandom, B=10)
    Index 0 Param 2=$User.EnergyFillFactor
    Module.Param0WriteEnabled=false
    Module.Param1WriteEnabled=true
    Module.Param2WriteEnabled=false
    Module.Param3WriteEnabled=false
   ) {module=DynamicMaterialParameters001}
  ]
  rendererClasses [&NiagaraRibbonRendererProperties]
  renderers [{rendererClass=&NiagaraRibbonRendererProperties}]
 }
 Right_Hip001 {
  emitterSpawnScript [&FN_GooGlobalProperties(
    EndBoneIndexArray=DataInterface() {propertyValues {GpuSyncMode=SyncCpuToGpu, MaxElements=0, IntData [0]}}
    EndBoneOffsetArray=DataInterface() {propertyValues {GpuSyncMode=SyncCpuToGpu, MaxElements=0, FloatData [<0 0 0>]}}
    MaxDistance:NiagaraFloat=$User.HipRibbonLength
    MaxDistanceCompressed=10
    MaxDistance_RandFactorHigh=1
    MaxDistance_RandFactorLow=1
    ParticlesPerRibbon=2
    ResolveSpeedHigh=3
    RibbonLength:NiagaraFloat=$User.HipRibbonLength
    RibbonLengthRandHigh=1
    RibbonLengthRandLow=1
    StartBoneIndexArray=DataInterface() {propertyValues {GpuSyncMode=SyncCpuToGpu, MaxElements=0, IntData [11]}}
    StartBoneOffsetArray=DataInterface() {propertyValues {GpuSyncMode=SyncCpuToGpu, MaxElements=0, FloatData [<2 0 3>]}}
   )]
  emitterUpdateScript [
   &EmitterState()
   &SpawnBurst_Instantaneous(
    Spawn Count=&Multiply_Int(A=$Emitter.ParticlesPerRibbon, B=$Emitter.NumberOfRibbons)
    Use Spawn Probability=false
    UseLoopCountLimit=false
   )
  ]
  particleSpawnScript [
   /SetVariables(Particles.SpawnGroup=&Divide_Int(A:NiagaraInt32=&ReturnExecIndex(), B:NiagaraInt32=$Emitter.ParticlesPerRibbon)) {isSetParameters=true, module=SetVariables_#9}
   /SetVariables(
    Particles.RibbonID=&MakeRibbonIDFromSpawnGroup()
    Particles.RibbonLinkOrder:NiagaraFloat=&Divide_Float(
     A=&MakeFloatFromInt(Int=&ModuloInt(A=&ReturnExecIndex(), B=$Emitter.ParticlesPerRibbon))
     B=&Subtract_Float(A=&MakeFloatFromInt(Int=$Emitter.ParticlesPerRibbon), B=1)
    )
   ) {isSetParameters=true, module=SetVariables_#10}
   /SetVariables(Particles.RibbonRandom:NiagaraFloat=&RandomRangeFloat(
     Minimum=0
     Maximum=1
     Randomness Mode=::NewEnumerator1
     Random Seed=&Add_Integer(A=$Engine.System.RandomSeed, B=$Particles.SpawnGroup)
     Fixed Random Seed=true
    )) {isSetParameters=true, module=SetVariables_#3}
   &InitializeRibbon(
    Color=&MultiplyLinearColor(
     Linear Color=$User.BariumPaintColor
     Scale Factor=&MakeVector4_FromLinearColor(LinearColor=&ColorFromCurve(
       ColorCurve=DataInterface() {propertyValues {
         AlphaCurve {keys [<at 0.574 1 RCIM_Cubic>,<at 0.9977 0>]}
         BlueCurve {keys [<at 0 1 RCIM_Cubic>,<at 1 1 RCIM_Cubic>]}
         GreenCurve {keys [<at 0 1 RCIM_Cubic>,<at 1 1 RCIM_Cubic>]}
         RedCurve {keys [<at 0 1 RCIM_Cubic>,<at 1 1 RCIM_Cubic>]}
        }}
       CurveIndex=$Particles.RibbonLinkOrder
      ))
    )
    Lifetime=&Lerp_Float(A=2, B=3, Alpha=$Particles.RibbonRandom)
    Ribbon Width=$User.HipRibbonWidth
    UsePositionOffset=false
    Write Material Random=true
   )
   /SetVariables(Particles.MaterialRandom:NiagaraFloat=$Particles.RibbonRandom) {isSetParameters=true, module=SetVariables_#4}
   &SkeletalMeshLocation(
    Bone / Socket Index=&SelectIntFromArray(Direct Array Index=$Particles.SpawnGroup, Direct Array Mode=::NewEnumerator1, Int Selection Array=$Emitter.StartBoneIndexArray)
    Bone Sampling Mode=::NewEnumerator4
    MaskSpawnGroup=false
    Orientation Sampling=::NewEnumerator1
    Position Sampling=::NewEnumerator2
    Sampled Position Offset=&SelectVectorFromArray(Direct Array Index=$Particles.SpawnGroup, Direct Array Mode=::NewEnumerator1, Vector Selection Array=$Emitter.StartBoneOffsetArray)
    Skeletal Mesh=$User.SkeletalMesh
    Use Sampled Mesh Attribute as Ribbon ID=false
   )
   &FN_GooCalculateGoalPosition(
    Axis:Vector3f=$Emitter.RibbonAxis
    Scale:Vector3f=&Multiply_VectorByFloat(
     Vector=&Multiply_VectorByFloat(
      Vector=&VectorFromFloat(Value=&Multiply_Float(A=$Emitter.RibbonLength, B=0))
      Float=&Lerp_Float(A=$Emitter.RibbonLengthRandMin, B=$Emitter.RibbonLengthRandMax, Alpha=$Particles.RibbonRandom)
     )
     Float=$Particles.RibbonLinkOrder
    )
    StartPosition:Vector3f=$Particles.SkeletalMeshLocation.SampledPosition
   ) {module=FN_CalculateGoalPosition}
   /SetVariables(Particles.Position=$Particles.GoalPosition) {isSetParameters=true, module=SetVariables_#5}
   /SetVariables(Particles.PreviousRotation:Quat4f=$Engine.Owner.Rotation) {isSetParameters=true, module=SetVariables_#6}
   &DynamicMaterialParameters(
    Index 0 Param 1=&Multiply_Float(A=$Particles.RibbonRandom, B=10)
    Module.Param0WriteEnabled=true
    Module.Param1WriteEnabled=true
    Module.Param2WriteEnabled=true
    Module.Param3WriteEnabled=true
   )
   /SetVariables(Particles.SpriteSize=<10 10>) {isSetParameters=true, module=SetVariables_#7}
  ]
  particleUpdateScript [
   &ParticleState()
   &SkeletalMeshLocation(
    Bone / Socket Index=&SelectIntFromArray(Direct Array Index=$Particles.SpawnGroup, Direct Array Mode=::NewEnumerator1, Int Selection Array=$Emitter.StartBoneIndexArray)
    Bone Sampling Mode=::NewEnumerator4
    MaskSpawnGroup=false
    Orientation Sampling=::NewEnumerator1
    Position Sampling=::NewEnumerator2
    Sampled Position Offset=&SelectVectorFromArray(Direct Array Index=$Particles.SpawnGroup, Direct Array Mode=::NewEnumerator1, Vector Selection Array=$Emitter.StartBoneOffsetArray)
    Skeletal Mesh=$User.SkeletalMesh
    Use Sampled Mesh Attribute as Ribbon ID=false
   ) {module=SkeletalMeshLocation001}
   &SolveForcesAndVelocity(Clamp Velocity=false, Limit Acceleration=false)
   &FN_GooCalculateGoalPosition(
    Axis:Vector3f=&SlerpVector(
     Vector A:Vector3f=&SlerpVector(
      Vector A:Vector3f=<-2 -1 0>
      Vector B:Vector3f=&TransformVector(Vector=$User.WorldSpaceWindVector, Source Space=::World, Destination Space=::Local)
      Alpha:NiagaraFloat=$User.WorldSpaceWindVectorFactor
      Scale:NiagaraFloat=1
     )
     Vector B:Vector3f=<0 0 1>
     Alpha:NiagaraFloat=&FloatFromCurve(
      FloatCurve=DataInterface() {propertyValues {Curve {keys [{tangentMode=RCTM_User, time=0, value=0}, {tangentMode=RCTM_User, time=1, value=0.5}]}}}
      CurveIndex=$Particles.RibbonLinkOrder
     )
     Scale:NiagaraFloat=1
    )
    CustomOrientation:Quat4f=$Engine.Owner.Rotation
    Scale:Vector3f=&Multiply_VectorByFloat(
     Vector=&Multiply_VectorByFloat(
      Vector=&VectorFromFloat(Value=&Multiply_Float(
        A=$Emitter.RibbonLength
        B=&FloatFromCurve(FloatCurve=DataInterface() {propertyValues {Curve {keys [
             {arriveTangent=30.27, interpMode=RCIM_Cubic, leaveTangent=30.27, tangentMode=RCTM_User, time=0, value=0}
             {interpMode=RCIM_Cubic, tangentMode=RCTM_User, time=0.15, value=1.5}
             <at 0.2 1 RCIM_Cubic>
             {arriveTangent=-3.4902, interpMode=RCIM_Cubic, leaveTangent=-3.4902, tangentMode=RCTM_User, time=1, value=0}
            ]}}})
       ))
      Float=&Lerp_Float(A=$Emitter.RibbonLengthRandMin, B=$Emitter.RibbonLengthRandMax, Alpha=$Particles.RibbonRandom)
     )
     Float=$Particles.RibbonLinkOrder
    )
    StartPosition:Vector3f=$Particles.SkeletalMeshLocation001.SampledPosition
    Use Quaternion Rotation=true
   ) {module=FN_CalculateGoalPosition001}
   &FN_GooApproachGoalPosition(
    AttractionForce:NiagaraFloat=&Lerp_Float(A=$Emitter.ResolveSpeedLow, B=$Emitter.ResolveSpeedHigh, Alpha=$Particles.RibbonRandom)
    GoalPosition:NiagaraPosition=$Particles.GoalPosition
    MaximumDistance:NiagaraFloat=&Multiply_Float(
     A=&FloatFromCurve(
      FloatCurve=DataInterface() {propertyValues {Curve {keys [
          {arriveTangent=4.1785, interpMode=RCIM_Cubic, time=0, value=0}
          {arriveTangent=2.3315, interpMode=RCIM_Cubic, leaveTangent=2.3315, tangentMode=RCTM_User, time=1, value=1}
         ]}}}
      CurveIndex=$Particles.RibbonLinkOrder
      Scale Curve=&Multiply_Float(
       A=$Emitter.MaxDistance
       B=&FloatFromCurve(FloatCurve=DataInterface() {propertyValues {Curve {keys [<at 0.75 1>,<at 1 0>]}}})
      )
     )
     B=&Lerp_Float(A=$Emitter.MaxDistance_RandFactorLow, B=$Emitter.MaxDistance_RandFactorHigh, Alpha=$Particles.RibbonRandom)
    )
    RibbonAxis:Vector3f=$Particles.RibbonAxis
   )
   /SetVariables(Particles.PreviousRotation:Quat4f=$Engine.Owner.Rotation) {isSetParameters=true, module=SetVariables_#8}
   &ScaleColor(
    Scale Alpha=&FloatFromCurve(FloatCurve=DataInterface() {propertyValues {Curve {keys [<at 0 0>, {arriveTangent=6.6358, leaveTangent=6.6358, time=0.5, value=1}, <at 0.7 1>, <at 1 0>]}}})
    Scale RGB=&VectorFromFloat(Value=&FloatFromCurve(FloatCurve=DataInterface() {propertyValues {Curve {keys [{arriveTangent=-319.691, interpMode=RCIM_Cubic, leaveTangent=-319.691, tangentMode=RCTM_User, time=0, value=800}, <at 0.4 15>, <at 1 1>]}}}))
    ScaleA=true
    ScaleRGB=true
   )
   &DynamicMaterialParameters(
    Index 0 Param 1=&Multiply_Float(A=$Particles.RibbonRandom, B=10)
    Index 0 Param 2=$User.EnergyFillFactor
    Module.Param0WriteEnabled=false
    Module.Param1WriteEnabled=true
    Module.Param2WriteEnabled=false
    Module.Param3WriteEnabled=false
   ) {module=DynamicMaterialParameters001}
  ]
  rendererClasses [&NiagaraRibbonRendererProperties]
  renderers [{rendererClass=&NiagaraRibbonRendererProperties}]
 }
 Right_Leg {
  emitterSpawnScript [&FN_GooGlobalProperties(
    EndBoneIndexArray=DataInterface() {propertyValues {GpuSyncMode=SyncCpuToGpu, MaxElements=0, IntData [0]}}
    EndBoneOffsetArray=DataInterface() {propertyValues {GpuSyncMode=SyncCpuToGpu, MaxElements=0, FloatData [<0 0 0>]}}
    MaxDistance:NiagaraFloat=$User.LegRibbonLength
    MaxDistanceCompressed=10
    MaxDistance_RandFactorHigh=1
    MaxDistance_RandFactorLow=1
    NumberOfRibbons=2
    ParticlesPerRibbon=3
    ResolveSpeedHigh=3
    RibbonLength:NiagaraFloat=$User.LegRibbonLength
    RibbonLengthRandHigh=1
    RibbonLengthRandLow=1
    StartBoneIndexArray=DataInterface() {propertyValues {GpuSyncMode=SyncCpuToGpu, MaxElements=0, IntData [8,9]}}
    StartBoneOffsetArray=DataInterface() {propertyValues {GpuSyncMode=SyncCpuToGpu, MaxElements=0, FloatData [<20 0 5>, <-15 0 5>]}}
   )]
  emitterUpdateScript [
   &EmitterState()
   &SpawnBurst_Instantaneous(
    Spawn Count=&Multiply_Int(A=$Emitter.ParticlesPerRibbon, B=$Emitter.NumberOfRibbons)
    Use Spawn Probability=false
    UseLoopCountLimit=false
   )
  ]
  particleSpawnScript [
   /SetVariables(Particles.SpawnGroup=&Divide_Int(A:NiagaraInt32=&ReturnExecIndex(), B:NiagaraInt32=$Emitter.ParticlesPerRibbon)) {isSetParameters=true, module=SetVariables_#9}
   /SetVariables(
    Particles.RibbonID=&MakeRibbonIDFromSpawnGroup()
    Particles.RibbonLinkOrder:NiagaraFloat=&Divide_Float(
     A=&MakeFloatFromInt(Int=&ModuloInt(A=&ReturnExecIndex(), B=$Emitter.ParticlesPerRibbon))
     B=&Subtract_Float(A=&MakeFloatFromInt(Int=$Emitter.ParticlesPerRibbon), B=1)
    )
   ) {isSetParameters=true, module=SetVariables_#10}
   /SetVariables(Particles.RibbonRandom:NiagaraFloat=&RandomRangeFloat(
     Minimum=0
     Maximum=1
     Randomness Mode=::NewEnumerator1
     Random Seed=&Add_Integer(A=$Engine.System.RandomSeed, B=$Particles.SpawnGroup)
     Fixed Random Seed=true
    )) {isSetParameters=true, module=SetVariables_#3}
   &InitializeRibbon(
    Color=&MultiplyLinearColor(
     Linear Color=$User.BariumPaintColor
     Scale Factor=&MakeVector4_FromLinearColor(LinearColor=&ColorFromCurve(
       ColorCurve=DataInterface() {propertyValues {
         AlphaCurve {keys [<at 0.574 1 RCIM_Cubic>,<at 0.9977 0>]}
         BlueCurve {keys [<at 0 1 RCIM_Cubic>,<at 1 1 RCIM_Cubic>]}
         GreenCurve {keys [<at 0 1 RCIM_Cubic>,<at 1 1 RCIM_Cubic>]}
         RedCurve {keys [<at 0 1 RCIM_Cubic>,<at 1 1 RCIM_Cubic>]}
        }}
       CurveIndex=$Particles.RibbonLinkOrder
      ))
    )
    Lifetime=&Lerp_Float(A=2, B=3, Alpha=$Particles.RibbonRandom)
    Ribbon Width=$User.LegRibbonWidth
    UsePositionOffset=false
    Write Material Random=true
   )
   /SetVariables(Particles.MaterialRandom:NiagaraFloat=$Particles.RibbonRandom) {isSetParameters=true, module=SetVariables_#4}
   &SkeletalMeshLocation(
    Bone / Socket Index=&SelectIntFromArray(Direct Array Index=$Particles.SpawnGroup, Direct Array Mode=::NewEnumerator1, Int Selection Array=$Emitter.StartBoneIndexArray)
    Bone Sampling Mode=::NewEnumerator4
    MaskSpawnGroup=false
    Orientation Sampling=::NewEnumerator1
    Position Sampling=::NewEnumerator2
    Sampled Position Offset=&SelectVectorFromArray(Direct Array Index=$Particles.SpawnGroup, Direct Array Mode=::NewEnumerator1, Vector Selection Array=$Emitter.StartBoneOffsetArray)
    Skeletal Mesh=$User.SkeletalMesh
    Use Sampled Mesh Attribute as Ribbon ID=false
   )
   &FN_GooCalculateGoalPosition(
    Axis:Vector3f=$Emitter.RibbonAxis
    Scale:Vector3f=&Multiply_VectorByFloat(
     Vector=&Multiply_VectorByFloat(
      Vector=&VectorFromFloat(Value=&Multiply_Float(A=$Emitter.RibbonLength, B=0))
      Float=&Lerp_Float(A=$Emitter.RibbonLengthRandMin, B=$Emitter.RibbonLengthRandMax, Alpha=$Particles.RibbonRandom)
     )
     Float=$Particles.RibbonLinkOrder
    )
    StartPosition:Vector3f=$Particles.SkeletalMeshLocation.SampledPosition
   ) {module=FN_CalculateGoalPosition}
   /SetVariables(Particles.Position=$Particles.GoalPosition) {isSetParameters=true, module=SetVariables_#5}
   /SetVariables(Particles.PreviousRotation:Quat4f=$Engine.Owner.Rotation) {isSetParameters=true, module=SetVariables_#6}
   &DynamicMaterialParameters(
    Index 0 Param 1=&Multiply_Float(A=$Particles.RibbonRandom, B=10)
    Module.Param0WriteEnabled=true
    Module.Param1WriteEnabled=true
    Module.Param2WriteEnabled=true
    Module.Param3WriteEnabled=true
   )
   /SetVariables(Particles.SpriteSize=<10 10>) {isSetParameters=true, module=SetVariables_#7}
  ]
  particleUpdateScript [
   &ParticleState()
   &SkeletalMeshLocation(
    Bone / Socket Index=&SelectIntFromArray(Direct Array Index=$Particles.SpawnGroup, Direct Array Mode=::NewEnumerator1, Int Selection Array=$Emitter.StartBoneIndexArray)
    Bone Sampling Mode=::NewEnumerator4
    MaskSpawnGroup=false
    Orientation Sampling=::NewEnumerator1
    Position Sampling=::NewEnumerator2
    Sampled Position Offset=&SelectVectorFromArray(Direct Array Index=$Particles.SpawnGroup, Direct Array Mode=::NewEnumerator1, Vector Selection Array=$Emitter.StartBoneOffsetArray)
    Skeletal Mesh=$User.SkeletalMesh
    Use Sampled Mesh Attribute as Ribbon ID=false
   ) {module=SkeletalMeshLocation001}
   &SolveForcesAndVelocity(Clamp Velocity=false, Limit Acceleration=false)
   &FN_GooCalculateGoalPosition(
    Axis:Vector3f=&SlerpVector(
     Vector A:Vector3f=&SlerpVector(
      Vector A:Vector3f=<-1 0 0>
      Vector B:Vector3f=&TransformVector(Vector=$User.WorldSpaceWindVector, Source Space=::World, Destination Space=::Local)
      Alpha:NiagaraFloat=$User.WorldSpaceWindVectorFactor
      Scale:NiagaraFloat=1
     )
     Vector B:Vector3f=<0 0 1>
     Alpha:NiagaraFloat=&FloatFromCurve(
      FloatCurve=DataInterface() {propertyValues {Curve {keys [{tangentMode=RCTM_User, time=0, value=0}, {tangentMode=RCTM_User, time=1, value=0.5}]}}}
      CurveIndex=$Particles.RibbonLinkOrder
     )
     Scale:NiagaraFloat=1
    )
    CustomOrientation:Quat4f=$Engine.Owner.Rotation
    Scale:Vector3f=&Multiply_VectorByFloat(
     Vector=&Multiply_VectorByFloat(
      Vector=&VectorFromFloat(Value=&Multiply_Float(
        A=$Emitter.RibbonLength
        B=&FloatFromCurve(FloatCurve=DataInterface() {propertyValues {Curve {keys [
             {arriveTangent=30.27, interpMode=RCIM_Cubic, leaveTangent=30.27, tangentMode=RCTM_User, time=0, value=0}
             {interpMode=RCIM_Cubic, tangentMode=RCTM_User, time=0.15, value=1.5}
             <at 0.2 1 RCIM_Cubic>
             {arriveTangent=-3.4902, interpMode=RCIM_Cubic, leaveTangent=-3.4902, tangentMode=RCTM_User, time=1, value=0}
            ]}}})
       ))
      Float=&Lerp_Float(A=$Emitter.RibbonLengthRandMin, B=$Emitter.RibbonLengthRandMax, Alpha=$Particles.RibbonRandom)
     )
     Float=$Particles.RibbonLinkOrder
    )
    StartPosition:Vector3f=$Particles.SkeletalMeshLocation001.SampledPosition
    Use Quaternion Rotation=true
   ) {module=FN_CalculateGoalPosition001}
   &FN_GooApproachGoalPosition(
    AttractionForce:NiagaraFloat=&Lerp_Float(A=$Emitter.ResolveSpeedLow, B=$Emitter.ResolveSpeedHigh, Alpha=$Particles.RibbonRandom)
    GoalPosition:NiagaraPosition=$Particles.GoalPosition
    MaximumDistance:NiagaraFloat=&Multiply_Float(
     A=&FloatFromCurve(
      FloatCurve=DataInterface() {propertyValues {Curve {keys [
          {arriveTangent=4.1785, interpMode=RCIM_Cubic, time=0, value=0}
          {arriveTangent=2.3315, interpMode=RCIM_Cubic, leaveTangent=2.3315, tangentMode=RCTM_User, time=1, value=1}
         ]}}}
      CurveIndex=$Particles.RibbonLinkOrder
      Scale Curve=&Multiply_Float(
       A=$Emitter.MaxDistance
       B=&FloatFromCurve(FloatCurve=DataInterface() {propertyValues {Curve {keys [<at 0.75 1>,<at 1 0>]}}})
      )
     )
     B=&Lerp_Float(A=$Emitter.MaxDistance_RandFactorLow, B=$Emitter.MaxDistance_RandFactorHigh, Alpha=$Particles.RibbonRandom)
    )
    RibbonAxis:Vector3f=$Particles.RibbonAxis
   )
   /SetVariables(Particles.PreviousRotation:Quat4f=$Engine.Owner.Rotation) {isSetParameters=true, module=SetVariables_#8}
   &ScaleColor(
    Scale Alpha=&FloatFromCurve(FloatCurve=DataInterface() {propertyValues {Curve {keys [
         {arriveTangent=0.6269, interpMode=RCIM_Cubic, leaveTangent=0.6269, tangentMode=RCTM_User, time=0, value=0}
         {arriveTangent=1.4286, leaveTangent=1.4286, time=0.5, value=1}
         <at 0.7 1>
         <at 1 0>
        ]}}})
    Scale RGB=&VectorFromFloat(Value=&FloatFromCurve(FloatCurve=DataInterface() {propertyValues {Curve {keys [{arriveTangent=-319.691, interpMode=RCIM_Cubic, leaveTangent=-319.691, tangentMode=RCTM_User, time=0, value=800}, <at 0.4 15>, <at 1 1>]}}}))
    ScaleA=true
    ScaleRGB=true
   )
   &DynamicMaterialParameters(
    Index 0 Param 1=&Multiply_Float(A=$Particles.RibbonRandom, B=10)
    Index 0 Param 2=$User.EnergyFillFactor
    Module.Param0WriteEnabled=false
    Module.Param1WriteEnabled=true
    Module.Param2WriteEnabled=false
    Module.Param3WriteEnabled=false
   ) {module=DynamicMaterialParameters001}
  ]
  rendererClasses [&NiagaraRibbonRendererProperties]
  renderers [{rendererClass=&NiagaraRibbonRendererProperties}]
 }
 Right_Leg001 {
  emitterSpawnScript [&FN_GooGlobalProperties(
    EndBoneIndexArray=DataInterface() {propertyValues {GpuSyncMode=SyncCpuToGpu, MaxElements=0, IntData [0]}}
    EndBoneOffsetArray=DataInterface() {propertyValues {GpuSyncMode=SyncCpuToGpu, MaxElements=0, FloatData [<0 0 0>]}}
    MaxDistance:NiagaraFloat=$User.LegRibbonLength
    MaxDistanceCompressed=10
    MaxDistance_RandFactorHigh=1
    MaxDistance_RandFactorLow=1
    NumberOfRibbons=2
    ParticlesPerRibbon=2
    ResolveSpeedHigh=3
    RibbonLength:NiagaraFloat=$User.LegRibbonLength
    RibbonLengthRandHigh=1
    RibbonLengthRandLow=1
    StartBoneIndexArray=DataInterface() {propertyValues {GpuSyncMode=SyncCpuToGpu, MaxElements=0, IntData [8,9]}}
    StartBoneOffsetArray=DataInterface() {propertyValues {GpuSyncMode=SyncCpuToGpu, MaxElements=0, FloatData [<20 0 5>, <-15 0 5>]}}
   )]
  emitterUpdateScript [
   &EmitterState()
   &SpawnBurst_Instantaneous(
    Spawn Count=&Multiply_Int(A=$Emitter.ParticlesPerRibbon, B=$Emitter.NumberOfRibbons)
    Use Spawn Probability=false
    UseLoopCountLimit=false
   )
  ]
  particleSpawnScript [
   /SetVariables(Particles.SpawnGroup=&Divide_Int(A:NiagaraInt32=&ReturnExecIndex(), B:NiagaraInt32=$Emitter.ParticlesPerRibbon)) {isSetParameters=true, module=SetVariables_#9}
   /SetVariables(
    Particles.RibbonID=&MakeRibbonIDFromSpawnGroup()
    Particles.RibbonLinkOrder:NiagaraFloat=&Divide_Float(
     A=&MakeFloatFromInt(Int=&ModuloInt(A=&ReturnExecIndex(), B=$Emitter.ParticlesPerRibbon))
     B=&Subtract_Float(A=&MakeFloatFromInt(Int=$Emitter.ParticlesPerRibbon), B=1)
    )
   ) {isSetParameters=true, module=SetVariables_#10}
   /SetVariables(Particles.RibbonRandom:NiagaraFloat=&RandomRangeFloat(
     Minimum=0
     Maximum=1
     Randomness Mode=::NewEnumerator1
     Random Seed=&Add_Integer(A=$Engine.System.RandomSeed, B=$Particles.SpawnGroup)
     Fixed Random Seed=true
    )) {isSetParameters=true, module=SetVariables_#3}
   &InitializeRibbon(
    Color=&MultiplyLinearColor(
     Linear Color=$User.BariumPaintColor
     Scale Factor=&MakeVector4_FromLinearColor(LinearColor=&ColorFromCurve(
       ColorCurve=DataInterface() {propertyValues {
         AlphaCurve {keys [<at 0.574 1 RCIM_Cubic>,<at 0.9977 0>]}
         BlueCurve {keys [<at 0 1 RCIM_Cubic>,<at 1 1 RCIM_Cubic>]}
         GreenCurve {keys [<at 0 1 RCIM_Cubic>,<at 1 1 RCIM_Cubic>]}
         RedCurve {keys [<at 0 1 RCIM_Cubic>,<at 1 1 RCIM_Cubic>]}
        }}
       CurveIndex=$Particles.RibbonLinkOrder
      ))
    )
    Lifetime=&Lerp_Float(A=2, B=3, Alpha=$Particles.RibbonRandom)
    Ribbon Width=$User.LegRibbonWidth
    UsePositionOffset=false
    Write Material Random=true
   )
   /SetVariables(Particles.MaterialRandom:NiagaraFloat=$Particles.RibbonRandom) {isSetParameters=true, module=SetVariables_#4}
   &SkeletalMeshLocation(
    Bone / Socket Index=&SelectIntFromArray(Direct Array Index=$Particles.SpawnGroup, Direct Array Mode=::NewEnumerator1, Int Selection Array=$Emitter.StartBoneIndexArray)
    Bone Sampling Mode=::NewEnumerator4
    MaskSpawnGroup=false
    Orientation Sampling=::NewEnumerator1
    Position Sampling=::NewEnumerator2
    Sampled Position Offset=&SelectVectorFromArray(Direct Array Index=$Particles.SpawnGroup, Direct Array Mode=::NewEnumerator1, Vector Selection Array=$Emitter.StartBoneOffsetArray)
    Skeletal Mesh=$User.SkeletalMesh
    Use Sampled Mesh Attribute as Ribbon ID=false
   )
   &FN_GooCalculateGoalPosition(
    Axis:Vector3f=$Emitter.RibbonAxis
    Scale:Vector3f=&Multiply_VectorByFloat(
     Vector=&Multiply_VectorByFloat(
      Vector=&VectorFromFloat(Value=&Multiply_Float(A=$Emitter.RibbonLength, B=0))
      Float=&Lerp_Float(A=$Emitter.RibbonLengthRandMin, B=$Emitter.RibbonLengthRandMax, Alpha=$Particles.RibbonRandom)
     )
     Float=$Particles.RibbonLinkOrder
    )
    StartPosition:Vector3f=$Particles.SkeletalMeshLocation.SampledPosition
   ) {module=FN_CalculateGoalPosition}
   /SetVariables(Particles.Position=$Particles.GoalPosition) {isSetParameters=true, module=SetVariables_#5}
   /SetVariables(Particles.PreviousRotation:Quat4f=$Engine.Owner.Rotation) {isSetParameters=true, module=SetVariables_#6}
   &DynamicMaterialParameters(
    Index 0 Param 1=&Multiply_Float(A=$Particles.RibbonRandom, B=10)
    Module.Param0WriteEnabled=true
    Module.Param1WriteEnabled=true
    Module.Param2WriteEnabled=true
    Module.Param3WriteEnabled=true
   )
   /SetVariables(Particles.SpriteSize=<10 10>) {isSetParameters=true, module=SetVariables_#7}
  ]
  particleUpdateScript [
   &ParticleState()
   &SkeletalMeshLocation(
    Bone / Socket Index=&SelectIntFromArray(Direct Array Index=$Particles.SpawnGroup, Direct Array Mode=::NewEnumerator1, Int Selection Array=$Emitter.StartBoneIndexArray)
    Bone Sampling Mode=::NewEnumerator4
    MaskSpawnGroup=false
    Orientation Sampling=::NewEnumerator1
    Position Sampling=::NewEnumerator2
    Sampled Position Offset=&SelectVectorFromArray(Direct Array Index=$Particles.SpawnGroup, Direct Array Mode=::NewEnumerator1, Vector Selection Array=$Emitter.StartBoneOffsetArray)
    Skeletal Mesh=$User.SkeletalMesh
    Use Sampled Mesh Attribute as Ribbon ID=false
   ) {module=SkeletalMeshLocation001}
   &SolveForcesAndVelocity(Clamp Velocity=false, Limit Acceleration=false)
   &FN_GooCalculateGoalPosition(
    Axis:Vector3f=&SlerpVector(
     Vector A:Vector3f=&SlerpVector(
      Vector A:Vector3f=<-1 0 0>
      Vector B:Vector3f=&TransformVector(Vector=$User.WorldSpaceWindVector, Source Space=::World, Destination Space=::Local)
      Alpha:NiagaraFloat=$User.WorldSpaceWindVectorFactor
      Scale:NiagaraFloat=1
     )
     Vector B:Vector3f=<0 0 1>
     Alpha:NiagaraFloat=&FloatFromCurve(
      FloatCurve=DataInterface() {propertyValues {Curve {keys [{tangentMode=RCTM_User, time=0, value=0}, {tangentMode=RCTM_User, time=1, value=0.5}]}}}
      CurveIndex=$Particles.RibbonLinkOrder
     )
     Scale:NiagaraFloat=1
    )
    CustomOrientation:Quat4f=$Engine.Owner.Rotation
    Scale:Vector3f=&Multiply_VectorByFloat(
     Vector=&Multiply_VectorByFloat(
      Vector=&VectorFromFloat(Value=&Multiply_Float(
        A=$Emitter.RibbonLength
        B=&FloatFromCurve(FloatCurve=DataInterface() {propertyValues {Curve {keys [
             {arriveTangent=30.27, interpMode=RCIM_Cubic, leaveTangent=30.27, tangentMode=RCTM_User, time=0, value=0}
             {interpMode=RCIM_Cubic, tangentMode=RCTM_User, time=0.15, value=1.5}
             <at 0.2 1 RCIM_Cubic>
             {arriveTangent=-3.4902, interpMode=RCIM_Cubic, leaveTangent=-3.4902, tangentMode=RCTM_User, time=1, value=0}
            ]}}})
       ))
      Float=&Lerp_Float(A=$Emitter.RibbonLengthRandMin, B=$Emitter.RibbonLengthRandMax, Alpha=$Particles.RibbonRandom)
     )
     Float=$Particles.RibbonLinkOrder
    )
    StartPosition:Vector3f=$Particles.SkeletalMeshLocation001.SampledPosition
    Use Quaternion Rotation=true
   ) {module=FN_CalculateGoalPosition001}
   &FN_GooApproachGoalPosition(
    AttractionForce:NiagaraFloat=&Lerp_Float(A=$Emitter.ResolveSpeedLow, B=$Emitter.ResolveSpeedHigh, Alpha=$Particles.RibbonRandom)
    GoalPosition:NiagaraPosition=$Particles.GoalPosition
    MaximumDistance:NiagaraFloat=&Multiply_Float(
     A=&FloatFromCurve(
      FloatCurve=DataInterface() {propertyValues {Curve {keys [
          {arriveTangent=4.1785, interpMode=RCIM_Cubic, time=0, value=0}
          {arriveTangent=2.3315, interpMode=RCIM_Cubic, leaveTangent=2.3315, tangentMode=RCTM_User, time=1, value=1}
         ]}}}
      CurveIndex=$Particles.RibbonLinkOrder
      Scale Curve=&Multiply_Float(
       A=$Emitter.MaxDistance
       B=&FloatFromCurve(FloatCurve=DataInterface() {propertyValues {Curve {keys [<at 0.75 1>,<at 1 0>]}}})
      )
     )
     B=&Lerp_Float(A=$Emitter.MaxDistance_RandFactorLow, B=$Emitter.MaxDistance_RandFactorHigh, Alpha=$Particles.RibbonRandom)
    )
    RibbonAxis:Vector3f=$Particles.RibbonAxis
   )
   /SetVariables(Particles.PreviousRotation:Quat4f=$Engine.Owner.Rotation) {isSetParameters=true, module=SetVariables_#8}
   &ScaleColor(
    Scale Alpha=&FloatFromCurve(FloatCurve=DataInterface() {propertyValues {Curve {keys [
         {arriveTangent=0.6269, interpMode=RCIM_Cubic, leaveTangent=0.6269, tangentMode=RCTM_User, time=0, value=0}
         {arriveTangent=1.4286, leaveTangent=1.4286, time=0.5, value=1}
         <at 0.7 1>
         <at 1 0>
        ]}}})
    Scale RGB=&VectorFromFloat(Value=&FloatFromCurve(FloatCurve=DataInterface() {propertyValues {Curve {keys [{arriveTangent=-319.691, interpMode=RCIM_Cubic, leaveTangent=-319.691, tangentMode=RCTM_User, time=0, value=800}, <at 0.4 15>, <at 1 1>]}}}))
    ScaleA=true
    ScaleRGB=true
   )
   &DynamicMaterialParameters(
    Index 0 Param 1=&Multiply_Float(A=$Particles.RibbonRandom, B=10)
    Index 0 Param 2=$User.EnergyFillFactor
    Module.Param0WriteEnabled=false
    Module.Param1WriteEnabled=true
    Module.Param2WriteEnabled=false
    Module.Param3WriteEnabled=false
   ) {module=DynamicMaterialParameters001}
  ]
  rendererClasses [&NiagaraRibbonRendererProperties]
  renderers [{rendererClass=&NiagaraRibbonRendererProperties}]
 }
}
summary {
 systemName=NS_BariumKillBurst
 emitters [
  {emitterName=Left_Arm, rendererClasses [&NiagaraRibbonRendererProperties]}
  {emitterName=Right_Arm, rendererClasses [&NiagaraRibbonRendererProperties]}
  {emitterName=Left_Leg, rendererClasses [&NiagaraRibbonRendererProperties]}
  {emitterName=Right_Leg, rendererClasses [&NiagaraRibbonRendererProperties]}
  {emitterName=Left_Hip, rendererClasses [&NiagaraRibbonRendererProperties]}
  {emitterName=Right_Hip, rendererClasses [&NiagaraRibbonRendererProperties]}
  {emitterName=Left_Hand, rendererClasses [&NiagaraRibbonRendererProperties]}
  {emitterName=Right_Hand, rendererClasses [&NiagaraRibbonRendererProperties]}
  {emitterName=Helmet, rendererClasses [&NiagaraRibbonRendererProperties]}
  {emitterName=Left_Arm001, rendererClasses [&NiagaraRibbonRendererProperties]}
  {emitterName=Right_Arm001, rendererClasses [&NiagaraRibbonRendererProperties]}
  {emitterName=Left_Leg001, rendererClasses [&NiagaraRibbonRendererProperties]}
  {emitterName=Right_Leg001, rendererClasses [&NiagaraRibbonRendererProperties]}
  {emitterName=Left_Hip001, rendererClasses [&NiagaraRibbonRendererProperties]}
  {emitterName=Right_Hip001, rendererClasses [&NiagaraRibbonRendererProperties]}
  {emitterName=Left_Hand001, rendererClasses [&NiagaraRibbonRendererProperties]}
  {emitterName=Right_Hand001, rendererClasses [&NiagaraRibbonRendererProperties]}
  {emitterName=Helmet001, rendererClasses [&NiagaraRibbonRendererProperties]}
 ]
 userVariables [
  User.ArmRibbonLength {type=NiagaraFloat, defaultValue=25, description=""}
  User.ArmRibbonWidth {type=NiagaraFloat, defaultValue=15, description=""}
  User.BariumPaintColor {type=LinearColor, description="", defaultValue {struct=LinearColor, value=<rgba 1 1 1 1>}}
  User.EnergyFillFactor {type=NiagaraFloat, defaultValue=0.5, description=""}
  User.HipRibbonLength {type=NiagaraFloat, defaultValue=20, description=""}
  User.HipRibbonWidth {type=NiagaraFloat, defaultValue=15, description=""}
  User.LegRibbonLength {type=NiagaraFloat, defaultValue=27, description=""}
  User.LegRibbonWidth {type=NiagaraFloat, defaultValue=20, description=""}
  User.SkeletalMesh {
   type=NiagaraDataInterfaceSkeletalMesh
   defaultValue {
    struct=NiagaraExt_VariableValue_DataInterface
    value {dataInterface=&self:NiagaraDataInterfaceSkeletalMesh_0}
   }
  }
  User.SpawnHelmetFX {type=NiagaraFloat, defaultValue=0, description=""}
  User.SpawnJacketFX {type=NiagaraFloat, defaultValue=0, description=""}
  User.WorldSpaceWindVector {type=Vector3f, description="", defaultValue {struct=Vector3f, value=<1 0 0>}}
  User.WorldSpaceWindVectorFactor {type=NiagaraFloat, defaultValue=0, description=""}
 ]
}
