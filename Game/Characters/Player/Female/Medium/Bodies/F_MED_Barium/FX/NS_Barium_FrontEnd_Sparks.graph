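# Niagara system for the Barium character's front-end sparks, pairing a short orange flash with falling wrist sparks.
# WristSparks bursts a few short-lived sprites downward in a cone under gravity and drag, stretched by speed as their bright color fades.
# The single-particle flash emitter feeds mesh, sprite, light and component renderers, with a growing mesh, curve-dimmed color and an offset light position.
# A debug gnomon emitter is disabled.
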NiagaraSystem /Game/Characters/Player/Female/Medium/Bodies/F_MED_Barium/FX/NS_Barium_FrontEnd_Sparks
dependencies {
 usedDataInterfaces [&NiagaraDataInterfaceVectorCurve, &NiagaraDataInterfaceCurve]
 usedDynamicInputs []
 usedModules [&SystemState]
}
emitters {
 FN_DebugGnomon {
  bEnabled=false
  emitterUpdateScript [
   &EmitterState()
   &SpawnBurst_Instantaneous(Use Spawn Probability=false, UseLoopCountLimit=false)
  ]
  particleSpawnScript [&InitializeParticle(Color Mode=::NewEnumerator0, Lifetime=2, Mesh Scale Mode=::NewEnumerator3, UsePositionOffset=false, Write Material Random=true)]
  particleUpdateScript [&ParticleState()]
  rendererClasses [&NiagaraMeshRendererProperties]
  renderers [{rendererClass=&NiagaraMeshRendererProperties, rendererIndex=0}]
 }
 FN_Weapons_Core_Flash_BETA {
  emitterUpdateScript [
   &EmitterState()
   &SpawnBurst_Instantaneous(Spawn Count=1, Use Spawn Probability=false)
  ]
  particleSpawnScript [
   /Niagara/Modules/Spawn/Initialization/InitializeParticle(
    Lifetime=&UniformRangedFloat(Minimum=0.1, Maximum=0.1)
    Mesh Scale=&UniformRangedVector(Minimum=<0.6 0.6 0.6>, Maximum=<0.9 0.8 0.8>)
    Module.Write Color=true
    Module.Write Lifetime=true
    Module.Write Mass=true
    Module.Write Material Random=true
    Module.Write Position=true
    Module.Write Scale=true
    Module.Write SpriteRotation=false
    Module.Write SpriteSize=true
    Position=&Add_Vector(A=/Niagara/DynamicInputs/Helpers/SimulationPosition(), B=<0 0 0>)
    Sprite Size=&Vector2DFromFloat(Value=&RandomRangeFloat(Minimum=10, Maximum=50))
   )
   /Game/Effects/Niagara/Modules/Orientation/FN_Fan(Spin:NiagaraFloat=&UniformRangedFloat(Minimum=0, Maximum=1))
   /SetVariables(Particles.LightPositionNew:Vector3f=&Add_Vector(
     A=Linked() {linkedVariable {name=Particles.Position, type=NiagaraPosition}}
     B=/Niagara/DynamicInputs/Transforms/TransformPosition(Position=<25 0 0>, Source Space=::Local, Destination Space=::Simulation)
    )) {isSetParameters=true}
   &CameraOffset(Camera Offset Amount=10)
  ]
  particleUpdateScript [
   &ParticleState()
   &Color(
    Color=<rgba 1 0.3636 0.01 1>
    Scale Color=&VectorFromCurve(VectorCurve=DataInterface() {propertyValues {XCurve {keys [<at 0.7 12>,<at 1 1>]}, YCurve {keys [<at 0.7 12>,<at 1 1>]}, ZCurve {keys [<at 0.7 12>,<at 1 1>]}}})
   )
   &ScaleMeshSize(Scale Factor=&VectorFromFloat(Value=&FloatFromCurve(FloatCurve=DataInterface() {propertyValues {Curve {keys [{arriveTangent=2.2457, interpMode=RCIM_Cubic, leaveTangent=2.2457, tangentMode=RCTM_User, time=0, value=0}, <at 1 1>]}}})))
   &FN_Sprite_Attributes(
    Create SpriteSize Attribute=true
    SpriteSize=&Vector2DFromFloat(Value=&RandomRangeFloat(Minimum=5, Maximum=35))
    Write to Module Namespace=true
   ) {module=FN_Sprite}
   /SetVariables(Particles.SpriteColor=&MultiplyLinearColor(Linear Color=$Particles.Color, Scale Factor=<1 1 1 0.2>)) {isSetParameters=true}
   &DynamicMaterialParameters(Write Parameter Index 0=false)
   &FN_Light_Attributes(
    Color=<rgba 1 0.6601 0.0729 1>
    Create Color Attribute=true
    Create Exponent Attribute=true
    Enable Particle Light Enable=true
    Exponent=4
    Light Enabled=true
    Scale Color=5
   )
   &FN_Light_Attributes(Color=<rgba 1 0.7728 0.204 1>, Create Color Attribute=true, Enable Particle Light Enable=true, Light Enabled=true) {module=FN_Light_Attributes001}
  ]
  rendererClasses [
   &NiagaraMeshRendererProperties
   &NiagaraLightRendererProperties
   &NiagaraComponentRendererProperties
   &NiagaraSpriteRendererProperties
  ]
  renderers [
   {rendererClass=&NiagaraMeshRendererProperties, rendererIndex=0}
   {rendererClass=&NiagaraLightRendererProperties, rendererIndex=1}
   {rendererClass=&NiagaraComponentRendererProperties, rendererIndex=2}
   {rendererClass=&NiagaraSpriteRendererProperties, rendererIndex=3}
  ]
 }
 WristSparks {
  emitterUpdateScript [
   &EmitterState()
   &SpawnBurst_Instantaneous(Spawn Count=&UniformRangedInt(Minimum=2, Maximum=6), Use Spawn Probability=false)
  ]
  particleSpawnScript [
   &InitializeParticle(
    Color=<rgba 1 0.5 0.1 1>
    Lifetime=&UniformRangedFloat(Minimum=0.05, Maximum=0.25)
    Module.Write Color=false
    Module.Write Lifetime=true
    Module.Write Mass=true
    Module.Write Material Random=true
    Module.Write Position=true
    Module.Write Scale=false
    Module.Write SpriteRotation=false
    Module.Write SpriteSize=true
    Sprite Size=&UniformRangedVector2D(Minimum=<1 1>, Maximum=<2 2>)
   )
   &AddVelocityInCone(
    Cone Angle=35
    Cone Axis=<0 0 -1>
    Module.Use Velocity Falloff On Cone Axis=false
    Velocity Strength=&UniformRangedFloat(Minimum=30, Maximum=350)
   )
   &SystemLocation(MaskSpawnGroup=false, Offset=<2 0 0>)
  ]
  particleUpdateScript [
   &ParticleState()
   &AccelerationForce(Acceleration=<0 0 -980>)
   &Drag(Use Linear Drag=true, Use Rotational Drag=false)
   &SolveForcesAndVelocity(Clamp Velocity=false, Limit Acceleration=false)
   &Color(
    Color=<rgba 1 0.442 0.0945 1>
    Scale Alpha=0.5
    Scale Color=&VectorFromCurve(
     VectorCurve=DataInterface() {propertyValues {
       XCurve {keys [<at -0.0011 1.0067>,<at 0.9989 0>]}
       YCurve {keys [<at -0.0011 1.0067>,<at 0.9989 0>]}
       ZCurve {keys [<at -0.0011 1.0067>,<at 0.9989 0>]}
      }}
     Scale Curve=<50 50 10>
    )
   )
   &ScaleSpriteSizeBySpeed(
    Max Scale Factor=<0.6 15>
    Min Scale Factor=<0.2 1>
    Sample Scale Factor By Curve=true
    Scale Factor Curve=DataInterface() {propertyValues {Curve {keys [
        <at -0.0077 0.0765 RCIM_Cubic>
        {arriveTangent=1.2463, interpMode=RCIM_Cubic, leaveTangent=1.2463, tangentMode=RCTM_User, time=0.9985, value=0.9549}
       ]}}}
    Velocity Threshold=600
   )
  ]
  rendererClasses [&NiagaraSpriteRendererProperties]
  renderers [{rendererClass=&NiagaraSpriteRendererProperties}]
 }
}
summary {
 systemName=NS_Barium_FrontEnd_Sparks
 emitters [
  {emitterName=WristSparks, rendererClasses [&NiagaraSpriteRendererProperties]}
  {
   emitterName=FN_Weapons_Core_Flash_BETA
   rendererClasses [
    &NiagaraMeshRendererProperties
    &NiagaraLightRendererProperties
    &NiagaraComponentRendererProperties
    &NiagaraSpriteRendererProperties
   ]
  }
  {bEnabled=false, emitterName=FN_DebugGnomon, rendererClasses [&NiagaraMeshRendererProperties]}
 ]
 userVariables []
}
