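# Material function for the F_MED_Barium player character body that layers an animated fluid-flow normal onto incoming material attributes.
# A warp texture, biplanar-projected from pre-skinned position and normal with time-based scrolling, distorts the UVs of a tiled, scrolling flow normal map.
# A BariumMask channel controls flow normal strength, and angle-corrected blending plus an Alpha input combine it with the base normal.
# Outputs the updated attributes, the normal and the raw warp color.
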MaterialFunction /Game/Characters/Player/Female/Medium/Bodies/F_MED_Barium/Materials/MaterialFunctions/MF_BariumFlow
graph {nodes [
  FunctionOutput(None=SetMaterialAttributes!(MaterialAttributes=NamedRerouteUsage(), Normal=NamedRerouteUsage()) {attributeSetTypes [#1]})
  FunctionOutput(None=NamedRerouteUsage()) {outputName=Normal}
  n1 = TextureObjectParameter() {
   group=BariumFlow
   parameterName=WarpTexture
   samplerType=SAMPLERTYPE_Masks
   texture=/Game/Characters/Player/Female/Medium/Bodies/F_MED_Barium/Textures/TechArt/T_Barium_Warp
  }
  n2 = Time()
  n3 = VectorParameter() {group=BariumFlow, parameterName=WarpUVSpeed, defaultValue=<rgba 0 0 0 1>}
  n4 = MaterialFunctionCall!(
   ColorTex=@n1
   NormalTex
   Pos=Add(A=NamedRerouteUsage(), B=Multiply(A=@n2, B=@n3.RGB))
   Nrm=NamedRerouteUsage()
   ColorUVScale=ScalarParameter() {defaultValue=128, group=BariumFlow, parameterName=WarpTextureSize}
   ColorBlendFactor=ScalarParameter() {defaultValue=4, group=BariumFlow, parameterName=WarpColorBlendFactor}
   NormalStrength
   NormalUVScale
   NormalBlendFactor
   Local Space Inputs=StaticBool() {value=true}
  ) {materialFunction=/Game/Global/MaterialFunctions/MF_BiPlanar}
  FunctionOutput(None=@n4.Color) {outputName=Warp}
  n5 = ScalarParameter() {group=BariumFlow, parameterName=NormalStrengthMin}
  n6 = ScalarParameter() {defaultValue=1, group=BariumFlow, parameterName=NormalStrengthMax}
  n7 = FunctionInput() {bUsePreviewValueAsDefault=true, inputName=BariumMask, inputType=FunctionInput_Vector4, previewValue=<1 1 1 1>}
  n8 = ChannelMaskParameter(None=NamedRerouteDeclaration(None=@n7) {name=BariumMask, nodeColor=<rgba 0.5 0 1 1>}) {group=BariumFlow, maskChannel=Blue, parameterName=BariumNormalMaskChannel, defaultValue=<rgba 0 0 1 0>}
  NamedRerouteDeclaration(None=LinearInterpolate(A=Constant3Vector() {constant=<rgba 0 0 1 0>}, B=NamedRerouteUsage(), Alpha=LinearInterpolate(A=@n5, B=@n6, Alpha=@n8.R))) {name=Normal_OUT, nodeColor=<rgba 0.0078 0 1 1>}
  n9 = GetMaterialAttributes(None=FunctionInput() {inputType=FunctionInput_MaterialAttributes}) {attributeGetTypes [#1]}
  NamedRerouteDeclaration(None=@n9.Normal) {name=Normal_IN, nodeColor=<rgba 1 0 0.1406 1>}
  n10 = FunctionInput() {bUsePreviewValueAsDefault=true, inputName=Alpha, inputType=FunctionInput_Scalar, previewValue=<1 0 0 1>}
  NamedRerouteDeclaration(None=@n10) {name=Alpha_IN, nodeColor=<rgba 1 0 0.9844 1>}
  n11 = NamedRerouteUsage()
  n12 = NamedRerouteUsage()
  NamedRerouteDeclaration(None=LinearInterpolate(
    A=@n11
    B=MaterialFunctionCall(BaseNormal=@n11, AdditionalNormal=@n12) {materialFunction=&BlendAngleCorrectedNormals}.Result
    Alpha=NamedRerouteUsage()
   )) {name=Normal, nodeColor=<rgba 1 0.0703 0 1>}
  NamedRerouteDeclaration(None=@n9.MaterialAttributes) {name=Material_IN, nodeColor=<rgba 0 1 0.5547 1>}
  NamedRerouteDeclaration(None=VertexInterpolator(VS=PreSkinnedNormal()).PS) {name=PS_Normal, nodeColor=<rgba 1 0 0.8437 1>}
  NamedRerouteDeclaration(None=VertexInterpolator(VS=PreSkinnedPosition()).PS) {name=PS_Position, nodeColor=<rgba 1 0 0.1641 1>}
  NamedRerouteDeclaration(None=ConstantBiasScale(None=@n4.Color) {bias=-0.5, scale=2}) {name=UVWarp, nodeColor=<rgba 0.3828 0 1 1>}
  n13 = Multiply(A=ComponentMask(None=NamedRerouteUsage()) {g=true, r=true}, B=ScalarParameter() {group=BariumFlow, parameterName=FlowNormalUVWarp})
  n14 = ComponentMask(None=NamedRerouteUsage()) {g=true, r=true}
  n15 = ComponentMask(None=VectorParameter() {group=BariumFlow, parameterName=FlowNormalUVSpeed, defaultValue=<rgba 0 0.1 0 1>}.RGB) {g=true, r=true}
  n16 = Add(
   A=Multiply(A=TextureCoordinate(), B=ScalarParameter() {defaultValue=1, group=BariumFlow, parameterName=FlowNormalUVTile})
   B=Multiply(A=Multiply(A=@n14, B=@n15), B=Time())
  )
  n17 = TextureObjectParameter() {
   group=BariumFlow
   parameterName=FlowNormalTexture
   samplerType=SAMPLERTYPE_Normal
   texture=/Game/Characters/Player/Female/Medium/Bodies/F_MED_Barium/Textures/TechArt/T_Barium_Fluid_01_N
  }
  NamedRerouteDeclaration(None=TextureSample(UVs=Add(A=@n13, B=@n16), Tex=@n17).RGB) {name=FlowNormal, nodeColor=<rgba 1 0.7266 0 1>}
 ]}
