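# Material function for the Barium player character's body paint, layering animated paint color and emissive onto incoming material attributes and outputting a mask.
# A warped FX map provides pattern channels blended by a threshold growth mask, plus optional scrolling paint swirls.
# Clamped velocity and a timed kill boost, with sphere-mask reveal, fade and emissive flash, form a velocity mask.
# Scrolling noise modulates emissive, which passes through DeriveHDRfromLDR.
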MaterialFunction /Game/Characters/Player/Female/Medium/Bodies/F_MED_Barium/Materials/MaterialFunctions/MF_BariumPaint
graph {nodes [
  Constant3Vector()
  FunctionOutput(None=SetMaterialAttributes!(MaterialAttributes=NamedRerouteUsage(), BaseColor=NamedRerouteUsage(), EmissiveColor=NamedRerouteUsage()) {attributeSetTypes [#1, #2]})
  n1 = NamedRerouteUsage()
  n2 = OneMinus(None=NamedRerouteUsage())
  n3 = Add(A=@n2, B=ScalarParameter() {defaultValue=0.03, group=BariumPaint, parameterName=FXPatternGrowthThreshold})
  n4 = ComponentMask(None=NamedRerouteUsage()) {a=true}
  n5 = LinearInterpolate(
   A=ComponentMask(None=NamedRerouteUsage()) {g=true}
   B=ComponentMask(None=NamedRerouteUsage()) {r=true}
   Alpha=Saturate(None=SmoothStep(Min=@n2, Max=@n3, Value=@n4))
  )
  n6 = Saturate(None=StaticSwitchParameter(True=Add(A=@n1, B=@n5), False=@n5) {defaultValue=true, group=BariumPaint, parameterName=UsePaintSwirls})
  FunctionOutput(None=NamedRerouteDeclaration(None=@n6) {name=FinalMask, nodeColor=<rgba 0 0.1328 1 1>}) {outputName=Mask}
  n7 = GetMaterialAttributes(None=FunctionInput() {inputType=FunctionInput_MaterialAttributes}) {attributeGetTypes [#1, #2]}
  NamedRerouteDeclaration(None=@n7.BaseColor) {name=IN_BaseColor, nodeColor=<rgba 0.6875 0 1 1>}
  n8 = NamedRerouteUsage()
  NamedRerouteDeclaration(None=LinearInterpolate(A=NamedRerouteUsage(), B=NamedRerouteUsage(), Alpha=Multiply(A=@n6, B=@n8))) {name=OUT_BaseColor, nodeColor=<rgba 0.2656 0 1 1>}
  n9 = Power(
   Base=ComponentMask(None=NamedRerouteUsage()) {a=true}
   Exp=ScalarParameter() {defaultValue=2, group=BariumPaint, parameterName=PaintMaskDFExponent}
  )
  n10 = Multiply(
   A=NamedRerouteDeclaration(None=VectorParameter() {group=BariumPaint, parameterName=BariumPaintColor, defaultValue=<rgba 0.974 0.7434 0 1>}.RGB) {name=LightningColor, nodeColor=<rgba 0.2656 1 0 1>}
   B=LinearInterpolate(B=@n9, Alpha=ScalarParameter() {group=BariumPaint, parameterName=Paint_DF_Influence}) {constA=1}
  )
  n11 = TextureCoordinate()
  n12 = ScalarParameter() {defaultValue=1, group=BariumPaint, parameterName=EmissiveNoiseTile}
  n13 = ComponentMask(None=VectorParameter() {group=BariumPaint, parameterName=EmissiveNoiseDirection, defaultValue=<rgba 0 1 0 1>}.RGB) {g=true, r=true}
  n14 = Multiply(A=Time(), B=ScalarParameter() {defaultValue=1, group=BariumPaint, parameterName=EmissiveNoiseSpeed})
  n15 = TextureSampleParameter2D(UVs=Add(A=Multiply(A=@n11, B=@n12), B=Multiply(A=@n13, B=@n14))) {
   group=BariumPaint
   parameterName=EmissiveNoiseTexture
   samplerType=SAMPLERTYPE_LinearColor
   texture=/Game/Characters/Player/Male/Medium/Bodies/M_MED_SRDrift/Textures/T_ElectricNoise_Mask
  }
  n16 = ScalarParameter() {defaultValue=2, group=BariumPaint, parameterName=EmissiveNoiseVelocityMultiplier}
  n17 = Multiply(A=LinearInterpolate(A=@n15.R, B=Multiply(A=@n16, B=@n15.R), Alpha=NamedRerouteUsage()), B=@n9)
  n18 = LinearInterpolate(
   A=ScalarParameter() {group=BariumPaint, parameterName=EmissiveStrengthMin}
   B=ScalarParameter() {defaultValue=1, group=BariumPaint, parameterName=EmissiveStrengthMax}
   Alpha=NamedRerouteUsage()
  )
  n19 = NamedRerouteUsage()
  n20 = Add(A=Add(A=Add(A=@n18, B=@n19), B=NamedRerouteUsage()), B=ScalarParameter() {group=BariumPaint, parameterName=FE_EmissiveBoost})
  n21 = LinearInterpolate(
   A=NamedRerouteUsage()
   B=Multiply(A=NamedRerouteDeclaration(None=@n10) {name=BaseColor, nodeColor=<rgba 0 1 0.5781 1>}, B=Multiply(A=@n17, B=@n20))
   Alpha=Multiply(A=NamedRerouteUsage(), B=NamedRerouteUsage())
  )
  NamedRerouteDeclaration(None=MaterialFunctionCall(LDR Input=Saturate(None=@n21), HDR Tint=@n21) {materialFunction=&DeriveHDRfromLDR}.RemappedRange) {name=Emissive, nodeColor=<rgba 0.5 0 1 1>}
  n22 = ScalarParameter() {defaultValue=1, group=BariumPaint, parameterName=PaintVelocityClamp}
  n23 = Saturate(None=Divide(
    A=ScalarParameter() {group=BariumPaint, parameterName=PaintVelocity}
    B=ScalarParameter() {defaultValue=20, group=BariumPaint, parameterName=PaintVelocityMax}
   ) {constB=600})
  n24 = Saturate(None=Multiply(A=NamedRerouteUsage(), B=NamedRerouteUsage()))
  n25 = Power(
   Base=Max(A=Min(A=@n22, B=@n23), B=NamedRerouteDeclaration(None=@n24) {name=KillBoost, nodeColor=<rgba 1 0.5859 0 1>})
   Exp=ScalarParameter() {defaultValue=0.5, group=BariumPaint, parameterName=VelocityMaskExponent}
  )
  NamedRerouteDeclaration(None=@n25) {name=VelocityMask, nodeColor=<rgba 1 0 0.0469 1>}
  NamedRerouteDeclaration(None=@n7.EmissiveColor) {name=IN_Emissive, nodeColor=<rgba 0 0.8594 1 1>}
  n26 = PreSkinnedLocalBounds()
  n27 = Add(A=Multiply(A=Subtract(A=@n26.Max, B=@n26.Max)) {constB=0.5}, B=@n26.Min)
  n28 = ComponentMask(None=@n26.Max) {b=true}
  n29 = ScalarParameter() {defaultValue=200, group=BariumPaint, parameterName=KillBoostSphereMaskRadius}
  n30 = AppendVector(
   A=ComponentMask(None=@n27) {g=true, r=true}
   B=LinearInterpolate(A=Add(A=@n28, B=@n29), B=ComponentMask(None=@n27) {b=true}, Alpha=NamedRerouteUsage())
  )
  n31 = NamedRerouteUsage()
  n32 = ScalarParameter() {defaultValue=0.25, group=BariumPaint, parameterName=KillBoostSphereMaskHardness}
  NamedRerouteDeclaration(None=Saturate(None=SphereMask(A=@n30, B=@n31, Radius=@n29, Hardness=@n32))) {name=KillBoostSphereMask, nodeColor=<rgba 0 1 0.6016 1>}
  n33 = Time()
  n34 = ScalarParameter() {group=BariumPaint, parameterName=KillBoostTime}
  n35 = ScalarParameter() {defaultValue=0.3, group=BariumPaint, parameterName=KillBoostRevealSpeed}
  n36 = Add(A=Add(A=@n34, B=@n35), B=ScalarParameter() {defaultValue=1, group=BariumPaint, parameterName=KillBoostDelay})
  n37 = Saturate(None=Divide(A=Subtract(A=@n33, B=@n36), B=ScalarParameter() {defaultValue=2, group=BariumPaint, parameterName=KillBoostFadeLength}) {constB=3})
  NamedRerouteDeclaration(None=Power(Base=OneMinus(None=@n37), Exp=ScalarParameter() {defaultValue=0.75, group=BariumPaint, parameterName=KillBoostExp})) {name=KillBoostFade, nodeColor=<rgba 0 0.0156 1 1>}
  n38 = Subtract(A=Time(), B=@n34)
  NamedRerouteDeclaration(None=Saturate(None=Divide(A=@n38, B=@n35))) {name=KillBoostMask, nodeColor=<rgba 1 0 0.6562 1>}
  NamedRerouteDeclaration(None=FunctionInput() {bUsePreviewValueAsDefault=true, inputName=Warp}) {name=Warp, nodeColor=<rgba 1 0 0.7266 1>}
  n39 = TextureCoordinate()
  n40 = ScalarParameter() {defaultValue=1, group=BariumPaint, parameterName=PaintSwirlTile}
  n41 = Multiply(A=Time(), B=VectorParameter() {group=BariumPaint, parameterName=PaintSwirlSpeed, defaultValue=<rgba 0 0 0 1>}.RGB)
  n42 = ComponentMask(None=NamedRerouteUsage()) {g=true, r=true}
  n43 = Add(
   A=Add(A=Multiply(A=@n39, B=@n40), B=ComponentMask(None=@n41) {g=true, r=true})
   B=Multiply(
    A=ScalarParameter() {defaultValue=10, group=BariumPaint, parameterName=PaintSwirlWarpStrength}
    B=ConstantBiasScale(None=@n42) {bias=-0.5, scale=2}
   )
  )
  n44 = TextureObjectParameter() {
   group=BariumPaint
   parameterName=PaintSwirlTexture
   samplerType=SAMPLERTYPE_LinearGrayscale
   texture=/Game/Characters/Player/Female/Medium/Bodies/F_MED_Barium/Textures/TechArt/T_Barium_Paint
  }
  NamedRerouteDeclaration(None=Multiply(A=TextureSample(UVs=@n43, Tex=@n44).R, B=ScalarParameter() {defaultValue=0.1, group=BariumPaint, parameterName=PaintSwirlStrength})) {name=PaintSwirls, nodeColor=<rgba 1 0.2812 0 1>}
  n45 = FunctionInput() {bUsePreviewValueAsDefault=true, inputName=Alpha, inputType=FunctionInput_Scalar, previewValue=<1 0 0 1>}
  NamedRerouteDeclaration(None=@n45) {name=Alpha, nodeColor=<rgba 1 0.3047 0 1>}
  NamedRerouteDeclaration(None=FunctionInput() {bUsePreviewValueAsDefault=true, inputName=Warp2}) {name=Warp2, nodeColor=<rgba 0.4062 1 0 1>}
  NamedRerouteDeclaration(None=VertexInterpolator(VS=PreSkinnedPosition()).PS) {name=PSPosition, nodeColor=<rgba 0 0.9297 1 1>}
  NamedRerouteDeclaration(None=VertexInterpolator(VS=PreSkinnedNormal()).PS) {name=PSNormal, nodeColor=<rgba 1 0.1172 0 1>}
  NamedRerouteDeclaration(None=@n7.MaterialAttributes) {name=IN_Material, nodeColor=<rgba 0.4297 0 1 1>}
  n46 = TextureCoordinate()
  n47 = ComponentMask(None=VectorParameter() {group=BariumPaint, parameterName=FXPatternWarpAmount, defaultValue=<rgba 0 0.04 0 1>}.RGB) {g=true, r=true}
  n48 = Multiply(A=ScalarParameter() {defaultValue=1, group=BariumPaint, parameterName=KillBoostWarpMultiplier}, B=NamedRerouteUsage())
  n49 = Multiply(A=ComponentMask(None=NamedRerouteUsage()) {b=true}, B=Add(A=@n47, B=Multiply(A=@n47, B=@n48)))
  n50 = TextureSampleParameter2D(UVs=StaticSwitchParameter(True=Add(A=@n46, B=@n49), False=@n46) {defaultValue=true, group=BariumPaint, parameterName=UsePatternWarp}) {
   group=BariumPaint
   parameterName=FX Map
   samplerType=SAMPLERTYPE_Masks
   texture=/Game/Characters/Player/Female/Medium/Bodies/F_MED_Barium/Textures/T_F_MED_Barium_Body_FX2
  }
  NamedRerouteDeclaration(None=@n50.RGBA) {name=LightningMask, nodeColor=<rgba 0 0.25 1 1>}
  n51 = ScalarParameter() {defaultValue=2, group=BariumPaint, parameterName=KillBoostEmissiveLength}
  n52 = OneMinus(None=Saturate(None=Divide(A=@n38, B=@n51)))
  NamedRerouteDeclaration(None=Multiply(A=@n52, B=ScalarParameter() {defaultValue=20, group=BariumPaint, parameterName=KillBoostEmissiveStrength})) {name=KillBoostEmissive, nodeColor=<rgba 0 0.3672 1 1>}
  NamedRerouteDeclaration(None=Saturate(None=@n25)) {name=WarpBlend, nodeColor=<rgba 0.0078 1 0 1>}
  NamedRerouteDeclaration(None=@n52) {name=KillBoostWarp, nodeColor=<rgba 1 0 0.8906 1>}
  NamedRerouteUsage()
  NamedRerouteUsage()
  Power(Exp=ScalarParameter() {defaultValue=2, group=BariumPaint, parameterName=DF_MaskExponent})
  ScalarParameter() {defaultValue=0.75, group=BariumPaint, parameterName=Debug_PaintAmount}
 ]}
