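# Material function that fakes cloth simulation with World Position Offset on the F_MED_Barium player body.
# Blends a wind vector toward a velocity parameter by speed, scaling movement and noise strength along a velocity curve.
# Directional push and scrolling world-space texture noise, with facing adjustment, form an offset masked by texture or vertex color.
# The offset adds to the incoming material's WPO by Alpha, with static-switch debug views of mask, magnitude, direction, velocity and noise.
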MaterialFunction /Game/Characters/Player/Female/Medium/Bodies/F_MED_Barium/Materials/MaterialFunctions/MF_WPOClothSim
graph {nodes [
  n1 = ChannelMaskParameter(None=NamedRerouteUsage()) {group=FakeClothSim, parameterName=NoiseChannel, sortPriority=15}
  ConstantBiasScale(None=@n1.R) {bias=-0.5, scale=2}
  n2 = Constant3Vector() {constant=<rgba 1 0 0 0>}
  CrossProduct(B=If(A=ComponentMask() {b=true}, B=Constant() {r=0.999}, "A > B"=@n2, "A == B"=@n2, "A < B"=Constant3Vector() {constant=<rgba 0 0 1 0>}))
  n3 = FunctionInput() {inputName=Material, inputType=FunctionInput_MaterialAttributes}
  n4 = Constant()
  n5 = GetMaterialAttributes(None=@n3) {attributeGetTypes [#1]}
  n6 = Add(A=@n5.WorldPositionOffset, B=NamedRerouteUsage())
  n7 = SetMaterialAttributes!(
   MaterialAttributes=@n3
   BaseColor=NamedRerouteUsage()
   Metallic=@n4
   Specular=@n4
   Roughness=Constant() {r=1}
   EmissiveColor=@n4
   WorldPositionOffset=@n6
  ) {attributeSetTypes [
    #2
    #3
    #4
    #5
    #6
    #1
   ]}
  n8 = FunctionInput() {bUsePreviewValueAsDefault=true, inputName=Alpha, inputType=FunctionInput_Scalar, previewValue=<1 0 0 1>}
  n9 = NamedRerouteDeclaration(None=@n8) {name=Alpha, nodeColor=<rgba 0.0781 0 1 1>}
  n10 = LinearInterpolate(A=@n5.WorldPositionOffset, B=@n6, Alpha=@n9)
  FunctionOutput(None=StaticSwitchParameter(
    True=BlendMaterialAttributes(A=@n3, B=@n7, Alpha=@n9)
    False=SetMaterialAttributes!(MaterialAttributes=@n5.MaterialAttributes, WorldPositionOffset=@n10) {attributeSetTypes [#1]}
   ) {group=FakeClothSim, parameterName=UseClothSimDebug})
  n11 = NamedRerouteUsage()
  n12 = Multiply(A=ScalarParameter() {group=FakeClothSim, parameterName=NegativeFacingAdjustment, sortPriority=20}, B=@n11)
  n13 = NamedRerouteUsage()
  NamedRerouteDeclaration(None=Multiply(A=NamedRerouteUsage(), B=LinearInterpolate(A=@n12, B=@n11, Alpha=@n13))) {name=WindNoise, nodeColor=<rgba 1 0 0.1641 1>}
  n14 = Multiply(
   A=Multiply(A=NamedRerouteUsage()) {constB=-1}
   B=MaterialFunctionCall(Speed=NamedRerouteUsage()) {materialFunction=&TimeWithSpeedVariable}.Time
  )
  n15 = NamedRerouteDeclaration(None=Divide(A=WorldPosition().XYZ, B=ScalarParameter() {defaultValue=128, group=FakeClothSim, parameterName=ClothSimNoiseTileDivisor, sortPriority=14})) {name=ClothSimPos, nodeColor=<rgba 0.5938 0 1 1>}
  NamedRerouteDeclaration(None=TextureSample(UVs=Add(A=@n14, B=@n15)) {samplerType=SAMPLERTYPE_LinearColor, texture=/Game/Global/MaterialFunctions/VT_ClothSimWPO}.RGB) {name=WindNoiseSample_1, nodeColor=<rgba 1 0.75 0 1>}
  n16 = VectorParameter() {group=FakeClothSim, parameterName=Velocity, sortPriority=25, defaultValue=<rgba 0 1 0 1>}
  NamedRerouteDeclaration(None=MaterialFunctionCall(Vector 3=@n16.RGB) {materialFunction=/Engine/Functions/Engine_MaterialFunctions02/Utility/VectorLength}.V3 Length) {name=VelocityMagnitude, nodeColor=<rgba 0 0.4141 1 1>}
  n17 = ScalarParameter() {group=FakeClothSim, parameterName=VelocityMin, sortPriority=3}
  n18 = ScalarParameter() {defaultValue=1000, group=FakeClothSim, parameterName=VelocityMax, sortPriority=4}
  n19 = NamedRerouteUsage()
  NamedRerouteDeclaration(None=Saturate(None=InverseLinearInterpolate(A=@n17, B=@n18, Value=@n19))) {name=VelocityNormalizedMagnitude, nodeColor=<rgba 1 0.0703 0 1>}
  NamedRerouteDeclaration(None=Multiply(
    A=NamedRerouteUsage()
    B=VectorParameter() {group=DebugFakeClothSim, parameterName=ClothSimDebugVelocityChannel, defaultValue=<rgba 1 1 1 1>}.RGB
   )) {name=DebugVelocity, nodeColor=<rgba 0.1719 0 1 1>}
  n20 = Add(A=@n17, B=ScalarParameter() {defaultValue=100, group=FakeClothSim, parameterName=VelocityWindBlendBias, sortPriority=5})
  NamedRerouteDeclaration(None=Saturate(None=InverseLinearInterpolate(A=@n17, B=@n20, Value=@n19))) {name=WindBlend, nodeColor=<rgba 0 0.5078 1 1>}
  n21 = NamedRerouteDeclaration(None=Normalize(VectorInput=VectorParameter() {group=FakeClothSim, parameterName=WindVector, sortPriority=0, defaultValue=<rgba 0 1 0 1>}.RGB)) {name=WindVector, nodeColor=<rgba 0 1 0.3203 1>}
  n22 = NamedRerouteDeclaration(None=Normalize(VectorInput=@n16.RGB)) {name=VelocityNormalizedVector, nodeColor=<rgba 0.3125 0 1 1>}
  n23 = NamedRerouteUsage()
  NamedRerouteDeclaration(None=MaterialFunctionCall(Vector=LinearInterpolate(A=@n21, B=@n22, Alpha=@n23), Default=Constant3Vector() {constant=<rgba 0 1 0 0>}) {materialFunction=&SafeNormalize}.Result) {name=Direction, nodeColor=<rgba 0.3594 0 1 1>}
  n24 = NamedRerouteUsage()
  n25 = ScalarParameter() {defaultValue=1, group=FakeClothSim, parameterName=ClothSimWindNoiseExponent, sortPriority=12}
  n26 = NamedRerouteDeclaration(None=Saturate(None=Power(Base=@n24, Exp=@n25))) {name=velocityCurve, nodeColor=<rgba 0 0.7891 1 1>}
  NamedRerouteDeclaration(None=LinearInterpolate(
    A=ScalarParameter() {group=FakeClothSim, parameterName=ClothSimNoiseStrengthMin, sortPriority=10}
    B=ScalarParameter() {defaultValue=5, group=FakeClothSim, parameterName=ClothSimNoiseStrengthMax, sortPriority=11}
    Alpha=@n26
   )) {name=ClothSimNoiseStrength, nodeColor=<rgba 0.1719 0 1 1>}
  NamedRerouteDeclaration(None=ScalarParameter() {defaultValue=1, group=FakeClothSim, parameterName=ClothSimNoiseSpeed, sortPriority=13}) {name=ClothSimNoiseSpeed, nodeColor=<rgba 0 1 0.5078 1>}
  NamedRerouteDeclaration(None=Multiply(
    A=NamedRerouteUsage()
    B=VectorParameter() {group=DebugFakeClothSim, parameterName=ClothSimDebugDirectionChannel, defaultValue=<rgba 1 1 1 1>}.RGB
   )) {name=DebugDirection, nodeColor=<rgba 0 0.9297 1 1>}
  n27 = NamedRerouteUsage()
  n28 = NamedRerouteUsage()
  n29 = Constant()
  n30 = StaticSwitchParameter(
   True=NamedRerouteUsage()
   False=StaticSwitchParameter(
    True=NamedRerouteUsage()
    False=StaticSwitchParameter(True=@n28, False=@n29) {defaultValue=true, group=DebugFakeClothSim, parameterName=DebugClothSimNoise}
   ) {group=DebugFakeClothSim, parameterName=DebugClothSimVelocity}
  ) {group=DebugFakeClothSim, parameterName=DebugClothSimDirection}
  NamedRerouteDeclaration(None=StaticSwitchParameter(True=NamedRerouteUsage(), False=StaticSwitchParameter(True=@n27, False=@n30) {group=DebugFakeClothSim, parameterName=DebugClothSimMagnitude}) {group=DebugFakeClothSim, parameterName=DebugClothSimMask}) {name=DebugColor, nodeColor=<rgba 1 0 0.6094 1>}
  n31 = NamedRerouteUsage()
  n32 = Multiply(A=ScalarParameter() {group=FakeClothSim, parameterName=NegativeFacingAdjustment}, B=@n31)
  n33 = NamedRerouteUsage()
  NamedRerouteDeclaration(None=Multiply(A=NamedRerouteUsage(), B=LinearInterpolate(A=@n32, B=@n31, Alpha=@n33))) {name=WindVectorPush, nodeColor=<rgba 0.4062 0 1 1>}
  NamedRerouteDeclaration(None=LinearInterpolate(
    A=ScalarParameter() {group=FakeClothSim, parameterName=ClothMovementMin, sortPriority=6}
    B=ScalarParameter() {defaultValue=2, group=FakeClothSim, parameterName=ClothMovementMax, sortPriority=7}
    Alpha=@n26
   )) {name=ClothSimMovement, nodeColor=<rgba 0.3594 1 0 1>}
  n34 = NamedRerouteUsage()
  n35 = NamedRerouteUsage()
  n36 = FunctionInput() {bUsePreviewValueAsDefault=true, inputName=ClothMask, inputType=FunctionInput_Vector4, previewValue=<1 1 1 1>}
  n37 = StaticSwitchParameter(
   True=ChannelMaskParameter(None=@n36) {group=FakeClothSim, parameterName=ClothMaskChannel, sortPriority=-10}.R
   False=ChannelMaskParameter(None=VertexColor()) {group=FakeClothSim, parameterName=VertexColorMaskChannel, sortPriority=-9}.R
  ) {group=FakeClothSim, parameterName=UseTextureMask, sortPriority=-20}
  NamedRerouteDeclaration(None=Multiply(A=Add(A=@n34, B=@n35), B=NamedRerouteDeclaration(None=@n37) {name=DebugClothMask, nodeColor=<rgba 0.4297 1 0 1>})) {name=WPO, nodeColor=<rgba 1 0 0.8203 1>}
  n38 = DotProduct(A=NamedRerouteUsage(), B=VertexNormalWS())
  NamedRerouteDeclaration(None=Saturate(None=ConstantBiasScale(None=@n38))) {name=WindDotN, nodeColor=<rgba 0 0.8828 1 1>}
  NamedRerouteDeclaration(None=@n1.R) {name=WindNoiseSample, nodeColor=<rgba 1 0.2578 0 1>}
  n39 = NamedRerouteUsage()
  n40 = LinearInterpolate(
   A=VertexNormalWS()
   B=NamedRerouteDeclaration(None=Multiply(A=@n39) {constB=-1}) {name=InverseDirection, nodeColor=<rgba 1 0.6563 0 1>}
   Alpha=ScalarParameter() {group=FakeClothSim, parameterName=WindDirOffsetInfluence, sortPriority=21}
  )
  NamedRerouteDeclaration(None=Multiply(A=Normalize(VectorInput=@n40), B=NamedRerouteUsage())) {name=WindNoiseDirection, nodeColor=<rgba 0.1484 1 0 1>}
 ]}
