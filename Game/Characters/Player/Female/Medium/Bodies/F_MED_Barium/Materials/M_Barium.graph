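# Master character material for the Barium female medium player body, using material attributes with shading model chosen by expression.
# Between shared character PreFX and PostFX functions, static switches chain Barium flow, clear coat, silk, thin film, metal LUT, Barium paint and WPO fake cloth sim.
# Channel-selectable SkinFX_Mask parameters supply effect alphas through named reroutes, alongside a Barium cloth mask texture and the flow's warp and normal outputs.
# A cinematic PostFX function finishes the chain.
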Material /Game/Characters/Player/Female/Medium/Bodies/F_MED_Barium/Materials/M_Barium
graph {
 properties {
  bUseMaterialAttributes=true
  bUsedWithClothing=true
  bUsedWithMeshDeformer=true
  bUsedWithMorphTargets=true
  bUsedWithNanite=true
  bUsedWithSkeletalMesh=true
  pixelDepthOffsetMode=PDOM_Legacy
  shadingModel=MSM_FromMaterialExpression
  subsurfaceProfile=&SS_HeroSkin_02
  usedShadingModels=Default Lit | Clear Coat | Subsurface Profile
 }
 nodes [
  n1 = MaterialFunctionCall() {materialFunction=&BaseFN_CharacterFunction_PreFX}
  n2 = NamedRerouteUsage()
  n3 = MaterialFunctionCall!(Alpha=NamedRerouteUsage(), BariumMask=NamedRerouteUsage(), In=@n1.Result) {materialFunction=/Game/Characters/Player/Female/Medium/Bodies/F_MED_Barium/Materials/MaterialFunctions/MF_BariumFlow}
  n4 = StaticSwitchParameter(True=@n3.Result, False=@n1.Result) {defaultValue=true, group=Skin FX, parameterName=UseBariumFlow}
  n5 = StaticSwitchParameter(
   True=MaterialFunctionCall!(Alpha=NamedRerouteUsage(), In=@n4) {materialFunction=/Game/Weapons/WeaponSkins/MaterialFunction/Clearcoat}.Result
   False=@n4
  ) {group=Skin FX, parameterName=Use Clear Coat}
  n6 = StaticSwitchParameter(
   True=MaterialFunctionCall!(Alpha=NamedRerouteUsage(), Material=@n5) {materialFunction=/Game/Weapons/WeaponSkins/MaterialFunction/MF_Silk}.Result
   False=@n5
  ) {defaultValue=true, group=Skin FX, parameterName=UseSilk}
  n7 = StaticSwitchParameter(
   True=MaterialFunctionCall!(Alpha=NamedRerouteUsage(), Material Attributes=@n6) {materialFunction=/Game/Weapons/WeaponSkins/MaterialFunction/ThinFilmFunction}.Result
   False=@n6
  ) {defaultValue=true, group=Skin FX, parameterName=UseThinFilm}
  n8 = StaticSwitchParameter(
   True=MaterialFunctionCall(Mask=NamedRerouteUsage(), Materials=@n7) {materialFunction=&MF_MetalsLUT}.Result
   False=@n7
  ) {defaultValue=true, group=Skin FX, parameterName=UseMetalLUT}
  n9 = StaticSwitchParameter(
   True=MaterialFunctionCall!(Warp=NamedRerouteUsage(), Alpha=NamedRerouteUsage(), Warp2=NamedRerouteUsage(), In=@n8) {materialFunction=/Game/Characters/Player/Female/Medium/Bodies/F_MED_Barium/Materials/MaterialFunctions/MF_BariumPaint}.Result
   False=@n8
  ) {defaultValue=true, group=Skin FX, parameterName=UseBariumPaint}
  n10 = MaterialFunctionCall(
   AO=@n1.AO
   Distance Blend=@n1.AO
   Fresnel=@n1.AO
   Gmap=@n1.AO
   Pre Skinned Local Position=@n1.AO
   Material=StaticSwitchParameter(
    True=MaterialFunctionCall!(Alpha=@n2, ClothMask, Material=@n9) {materialFunction=/Game/Characters/Player/Female/Medium/Bodies/F_MED_Barium/Materials/MaterialFunctions/MF_WPOClothSim}.Result
    False=@n9
   ) {defaultValue=true, group=Skin FX, parameterName=UseFakeClothSim}
  ) {materialFunction=&BaseFN_CharacterFunction_PostFX}
  MaterialFunctionCall(In=@n10.Result) {materialFunction=&CINE_MF_CharPostFX}
  NamedRerouteDeclaration(None=@n3.Normal) {name=BariumFlowNormal, nodeColor=<rgba 0.5703 1 0 1>}
  NamedRerouteDeclaration(None=@n3.Warp) {name=BariumFlowWarp, nodeColor=<rgba 0 1 0.7891 1>}
  n11 = TextureSample(Tex=TextureObjectParameter() {group=Skin FX, parameterName=SkinFX_Mask})
  NamedRerouteDeclaration(None=ChannelMaskParameter(None=@n11.RGBA) {group=Skin FX, maskChannel=Green, parameterName=BariumMaskChannel, defaultValue=<rgba 0 1 0 0>}.R) {name=BariumMaskChannel, nodeColor=<rgba 0 1 0.2031 1>}
  NamedRerouteDeclaration(None=ChannelMaskParameter(None=@n11.RGBA) {group=Skin FX, maskChannel=Green, parameterName=BariumMaskChannel2, defaultValue=<rgba 0 1 0 0>}.R) {name=BariumMaskChannel2, nodeColor=<rgba 0 1 0.2031 1>}
  NamedRerouteDeclaration(None=ChannelMaskParameter(None=@n11.RGBA) {group=Skin FX, parameterName=FakeClothSimChannel}.R) {name=ClothSimAlpha, nodeColor=<rgba 0.8281 0 1 1>}
  n12 = NamedRerouteUsage()
  n13 = Multiply(A=NamedRerouteUsage(), B=ScalarParameter() {group=Skin FX, parameterName=BariumMaskChannel2Influence})
  NamedRerouteDeclaration(None=Saturate(None=Add(A=@n12, B=@n13))) {name=BariumMaskAlpha, nodeColor=<rgba 0.7578 0 1 1>}
  n14 = TextureSampleParameter2D() {
   parameterName=BariumMask
   samplerType=SAMPLERTYPE_Masks
   texture=/Game/Characters/Player/Female/Medium/Bodies/F_MED_Barium/Textures/TechArt/T_Barium_Cloth_Mask
  }
  NamedRerouteDeclaration(None=@n14.RGBA) {name=BariumMask, nodeColor=<rgba 0.6172 1 0 1>}
  NamedRerouteDeclaration(None=ChannelMaskParameter(None=@n11.RGBA) {group=Skin FX, parameterName=BariumFlowChannel}.R) {name=BariumFlowAlpha, nodeColor=<rgba 1 0.5625 0 1>}
  NamedRerouteDeclaration(None=ChannelMaskParameter(None=@n11.RGBA) {group=Skin FX, parameterName=ClearCoatChannel}.R) {name=ClearCoatAlpha, nodeColor=<rgba 0.9688 1 0 1>}
  NamedRerouteDeclaration(None=ChannelMaskParameter(None=@n11.RGBA) {group=Skin FX, parameterName=Silk Channel}.R) {name=SilkAlpha, nodeColor=<rgba 0.5234 0 1 1>}
  NamedRerouteDeclaration(None=ChannelMaskParameter(None=@n11.RGBA) {group=Skin FX, parameterName=ThinFilm Channel}.R) {name=ThinFilmAlpha, nodeColor=<rgba 0.0312 0 1 1>}
  NamedRerouteDeclaration(None=ChannelMaskParameter(None=@n11.RGBA) {group=Skin FX, parameterName=MetalLUT Channel}.R) {name=MetalLUTAlpha, nodeColor=<rgba 1 0.0703 0 1>}
 ]
}
