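# Control Rig for the F_MED_Barium female medium player body that applies corrective adjustments to accessory and deform bones.
# A sequence first parent-constrains the necklace and lower zipper master bones to weighted blends of spine bones, with neck and clavicles also influencing the necklace.
# It then nudges and rotates the necklace, weighted by looking_down.
# Finally, when is_playing_character_idle is true, it offsets and enlarges both glute deform bones.
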ControlRigBlueprint /Game/Characters/Player/Female/Medium/Bodies/F_MED_Barium/Meshes/F_MED_Barium_CtrlRig
graphs [
 RigVMFunctionLibrary {links [], nodes []}
 RigVMModel {nodes [
   RigUnit_BeginExecution = &RigUnit_BeginExecution()
   n1 = /Script/ControlRig.RigUnit_SequenceExecution(ExecuteContext=@RigUnit_BeginExecution.ExecutePin) {is=struct RigVMUnitNode}
   &RigUnit_ParentConstraint(
    AdvancedSettings=(RotationOrderForFilter=XZY)
    Child=(Type=Bone,Name="dfrm_necklace")
    Parents=((Item=(Type=Bone,Name="neck_01"),Weight=0.32),(Item=(Type=Bone,Name="spine_05"),Weight=0.79),(Item=(Type=Bone,Name="clavicle_l"),Weight=0.013),(Item=(Type=Bone,Name="clavicle_r"),Weight=0.013),(Item=(Type=Bone,Name="spine_04"),Weight=0.154))
    Weight=1
    ExecutePin=@n1.A
   )
   > &RigUnit_ParentConstraint(
    AdvancedSettings=(RotationOrderForFilter=XZY)
    Child=(Type=Bone,Name="dyn_zipper_02_mstr")
    Parents=((Item=(Type=Bone,Name="spine_04"),Weight=0.78),(Item=(Type=Bone,Name="spine_05"),Weight=0.22))
    Weight=1
   )
   Branch_RigVMUnitNode = &RigVMFunction_ControlFlowBranch(ExecuteContext=@n1.C, Condition=RigVMVariableNode() {variable=is_playing_character_idle}.Value)
   &RigUnit_ModifyTransforms(
    ItemToModify=((Item=(Type=Bone,Name="deform_glute_l"),Transform=(Rotation=(X=0,Y=0,Z=0,W=1),Translation=(X=-2.5,Y=-2,Z=0),Scale3D=(X=1.25,Y=1.25,Z=1.25))))
    Weight=1
    ExecutePin=@Branch_RigVMUnitNode.True
   )
   > &RigUnit_ModifyTransforms(
    ItemToModify=((Item=(Type=Bone,Name="deform_glute_r"),Transform=(Rotation=(X=0,Y=0,Z=0,W=1),Translation=(X=2.5,Y=2,Z=0),Scale3D=(X=1.25,Y=1.25,Z=1.25))))
    Weight=1
   )
   &RigUnit_ModifyTransforms(
    ItemToModify=((Item=(Type=Bone,Name="dfrm_necklace"),Transform=(Rotation=(X=0,Y=0,Z=0,W=1),Translation=(X=0.25,Y=0,Z=0),Scale3D=(X=1,Y=1,Z=1))))
    ItemToModify.0.Transform.Rotation=&RigVMFunction_MathQuaternionFromEuler(Euler=(X=0,Y=-2,Z=2)).Result
    Weight=RigVMVariableNode() {variable=looking_down}.Value
    ExecutePin=@n1.B
   )
  ]}
]
hierarchy [
 root {loc=<0 0 0>, rot=<0 0 0 1>}
 pelvis {parent=root, loc=<0 -0.5301 98.0571>, rot=<0.0127 0.707 -0.0127 -0.707>}
 spine_01 {parent=pelvis, loc=<6.1208 -0.9328 0>, rot=<0 0 0.0576 -0.9983>}
 spine_02 {parent=spine_01, loc=<6.6022 -0.0153 0>, rot=<0 0 0.0022 -1>}
 spine_03 {parent=spine_02, loc=<6.4919 -0.0105 0>, rot=<0 0 0.0015 -1>}
 spine_04 {parent=spine_03, loc=<6.9596 -0.007 0>, rot=<0 0 0.0009 -1>}
 spine_05 {parent=spine_04, loc=<6.9981 -0.0017 0>, rot=<0 0 -0.0042 -1>}
 clavicle_l {parent=spine_05, loc=<5.4831 -0.1053 -3.2946>, rot=<0.0938 -0.7388 0.0307 -0.6666>}
 upperarm_l {parent=clavicle_l, loc=<9.9486 0 0>, rot=<0.0371 -0.2992 -0.0705 -0.9509>}
 lowerarm_l {parent=upperarm_l, loc=<26.4793 0 0>, rot=<0 0 -0.2382 -0.9712>}
 hand_l {parent=lowerarm_l, loc=<22.7458 0 0>, rot=<-0.6747 -0.0182 0.0605 -0.7354>}
 thumb_01_l {parent=hand_l, loc=<1.3118 -0.3191 -2.459>, rot=<0.6546 -0.4042 0.0664 -0.6354>}
 thumb_02_l {parent=thumb_01_l, loc=<5.9322 -0.0253 0.1416>, rot=<0.0121 0.0046 0.2964 -0.955>}
 thumb_03_l {parent=thumb_02_l, loc=<2.2223 0.3682 0>, rot=<0 0 0.0172 -0.9999>}
 index_metacarpal_l {parent=hand_l, loc=<1.6535 0.0183 -2.96>, rot=<0.0158 -0.0124 -0.0173 -0.9996>}
 index_01_l {parent=index_metacarpal_l, loc=<8.3371 0.1718 -0.2461>, rot=<0.0363 0.0499 0.142 -0.9879>}
 index_02_l {parent=index_01_l, loc=<3.4919 -0.3004 0>, rot=<0 0 0.2948 -0.9556>}
 index_03_l {parent=index_02_l, loc=<2.5578 -0.123 0>, rot=<0 0 0.1968 -0.9804>}
 middle_metacarpal_l {parent=hand_l, loc=<2.1856 -0.1725 -0.5984>, rot=<0.006 0.0168 -0.0329 -0.9993>}
 middle_01_l {parent=middle_metacarpal_l, loc=<7.6374 0.346 -0.1508>, rot=<-0.0085 0.0328 0.1813 -0.9829>}
 middle_02_l {parent=middle_01_l, loc=<3.9108 0 0>, rot=<0 0 0.2808 -0.9598>}
 middle_03_l {parent=middle_02_l, loc=<2.6103 0.0517 0>, rot=<0 0 0.1446 -0.9895>}
 ring_metacarpal_l {parent=hand_l, loc=<2.2074 -0.3625 1.3638>, rot=<-0.0284 0.0722 -0.0458 -0.9959>}
 ring_01_l {parent=ring_metacarpal_l, loc=<7.2483 0.0329 -0.1426>, rot=<-0.0456 -0.0277 0.2951 -0.954>}
 ring_02_l {parent=ring_01_l, loc=<3.7208 0.4647 0>, rot=<0 0 0.1583 -0.9874>}
 ring_03_l {parent=ring_02_l, loc=<2.4152 0.114 0>, rot=<0 0 0.1504 -0.9886>}
 pinky_metacarpal_l {parent=hand_l, loc=<1.7616 -0.9753 3.3814>, rot=<-0.0397 0.0814 -0.069 -0.9935>}
 pinky_01_l {parent=pinky_metacarpal_l, loc=<7.1933 -0.4205 -0.2601>, rot=<-0.0487 -0.0063 0.2556 -0.9655>}
 pinky_02_l {parent=pinky_01_l, loc=<2.8274 0.0893 0>, rot=<0 0 0.2474 -0.9689>}
 pinky_03_l {parent=pinky_02_l, loc=<2.2046 0.0307 0>, rot=<0 0 0.125 -0.9922>}
 deform_wrist_dn_base_l {parent=hand_l, loc=<0 -1.5326 0>, rot=<0 0 0 -1>}
 deform_wrist_dn_l {parent=deform_wrist_dn_base_l, loc=<0 0 0>, rot=<0 0 0 1>}
 deform_wrist_up_base_l {parent=hand_l, loc=<0 2.3009 0>, rot=<0 0 0 -1>}
 deform_wrist_up_l {parent=deform_wrist_up_base_l, loc=<0 0 0>, rot=<0 0 0 1>}
 lowerarm_twist_02_l {parent=lowerarm_l, loc=<9.6199 0 0>, rot=<0 0 0 1>}
 lowerarm_twist_01_l {parent=lowerarm_l, loc=<17.8549 0 0>, rot=<0 0 0 1>}
 deform_elbow_in_base_l {parent=lowerarm_l, loc=<0.1385 3.6052 -0.2182>, rot=<0.0059 0.0117 -0.0071 -0.9999>}
 deform_elbow_in_l {parent=deform_elbow_in_base_l, loc=<0 0 0>, rot=<0 0 0 1>}
 deform_elbow_out_base_l {parent=lowerarm_l, loc=<0.0565 -3.9993 0.1259>, rot=<0.0059 0.0117 -0.0071 -0.9999>}
 deform_elbow_out_l {parent=deform_elbow_out_base_l, loc=<0 0 0>, rot=<0 0 0 1>}
 upperarm_twist_01_l {parent=upperarm_l, loc=<0.0041 0 0>, rot=<0 0 0 1>}
 deform_upperarm_bk_base_l {parent=upperarm_twist_01_l, loc=<0.9473 -4.46 0.3113>, rot=<0.0004 -0.0034 -0.0158 -0.9999>}
 deform_upperarm_bk_l {parent=deform_upperarm_bk_base_l, loc=<0 0 0>, rot=<0 0 0 1>}
 deform_upperarm_fr_base_l {parent=upperarm_twist_01_l, loc=<1.4228 5.0263 0.3212>, rot=<0.0004 -0.0034 -0.0158 -0.9999>}
 deform_upperarm_fr_l {parent=deform_upperarm_fr_base_l, loc=<0 0 0>, rot=<0 0 0 1>}
 deform_upperarm_up_base_l {parent=upperarm_twist_01_l, loc=<-1.6653 0.3195 3.9629>, rot=<0.0004 -0.0034 -0.0158 -0.9999>}
 deform_upperarm_up_l {parent=deform_upperarm_up_base_l, loc=<0 0 0>, rot=<0 0 0 1>}
 upperarm_twist_02_l {parent=upperarm_l, loc=<13.7941 0 0>, rot=<0 0 0 1>}
 neck_01 {parent=spine_05, loc=<11.3138 0.1529 0>, rot=<0 0 -0.2003 -0.9797>}
 neck_02 {parent=neck_01, loc=<4.5627 0 0>, rot=<0 0 0.0504 -0.9987>}
 head {parent=neck_02, loc=<5.3914 0 0>, rot=<0 0 0.115 -0.9934>}
 dyn_zipper_01_mstr {parent=neck_01, loc=<1.7873 6.746 -0.0134>, rot=<0 0 0.9827 0.1854>}
 dyn_zipper_01 {parent=dyn_zipper_01_mstr, loc=<0 0 0>, rot=<0 0 0 -1>}
 clavicle_r {parent=spine_05, loc=<5.4831 -0.1053 3.2946>, rot=<0.7388 0.0938 -0.6666 -0.0307>}
 upperarm_r {parent=clavicle_r, loc=<-9.9486 0 0>, rot=<0.0371 -0.2992 -0.0705 -0.9509>}
 lowerarm_r {parent=upperarm_r, loc=<-26.4793 0 0>, rot=<0 0 -0.2382 -0.9712>}
 hand_r {parent=lowerarm_r, loc=<-22.7458 0 0>, rot=<-0.6747 -0.0182 0.0605 -0.7354>}
 thumb_01_r {parent=hand_r, loc=<-1.3118 0.3191 2.459>, rot=<0.6546 -0.4042 0.0664 -0.6354>}
 thumb_02_r {parent=thumb_01_r, loc=<-5.9322 0.0253 -0.1416>, rot=<0.0121 0.0046 0.2964 -0.955>}
 thumb_03_r {parent=thumb_02_r, loc=<-2.2223 -0.3682 0>, rot=<0 0 0.0172 -0.9999>}
 index_metacarpal_r {parent=hand_r, loc=<-1.6535 -0.0183 2.96>, rot=<0.0158 -0.0124 -0.0173 -0.9996>}
 index_01_r {parent=index_metacarpal_r, loc=<-8.3371 -0.1718 0.2461>, rot=<0.0363 0.0499 0.142 -0.9879>}
 index_02_r {parent=index_01_r, loc=<-3.4919 0.3004 0>, rot=<0 0 0.2948 -0.9556>}
 index_03_r {parent=index_02_r, loc=<-2.5578 0.123 0>, rot=<0 0 0.1968 -0.9804>}
 middle_metacarpal_r {parent=hand_r, loc=<-2.1856 0.1725 0.5984>, rot=<0.006 0.0168 -0.0329 -0.9993>}
 middle_01_r {parent=middle_metacarpal_r, loc=<-7.6374 -0.346 0.1508>, rot=<-0.0085 0.0328 0.1813 -0.9829>}
 middle_02_r {parent=middle_01_r, loc=<-3.9108 0 0>, rot=<0 0 0.2808 -0.9598>}
 middle_03_r {parent=middle_02_r, loc=<-2.6103 -0.0517 0>, rot=<0 0 0.1446 -0.9895>}
 ring_metacarpal_r {parent=hand_r, loc=<-2.2074 0.3625 -1.3638>, rot=<-0.0284 0.0722 -0.0458 -0.9959>}
 ring_01_r {parent=ring_metacarpal_r, loc=<-7.2483 -0.0329 0.1426>, rot=<-0.0456 -0.0277 0.2951 -0.954>}
 ring_02_r {parent=ring_01_r, loc=<-3.7208 -0.4647 0>, rot=<0 0 0.1583 -0.9874>}
 ring_03_r {parent=ring_02_r, loc=<-2.4152 -0.114 0>, rot=<0 0 0.1504 -0.9886>}
 pinky_metacarpal_r {parent=hand_r, loc=<-1.7616 0.9753 -3.3814>, rot=<-0.0397 0.0814 -0.069 -0.9935>}
 pinky_01_r {parent=pinky_metacarpal_r, loc=<-7.1933 0.4205 0.2601>, rot=<-0.0487 -0.0063 0.2556 -0.9655>}
 pinky_02_r {parent=pinky_01_r, loc=<-2.8274 -0.0893 0>, rot=<0 0 0.2474 -0.9689>}
 pinky_03_r {parent=pinky_02_r, loc=<-2.2046 -0.0307 0>, rot=<0 0 0.125 -0.9922>}
 deform_wrist_dn_base_r {parent=hand_r, loc=<-0.0001 1.5328 0>, rot=<0 0 0 -1>}
 deform_wrist_dn_r {parent=deform_wrist_dn_base_r, loc=<0 0 0>, rot=<0 0 0 1>}
 deform_wrist_up_base_r {parent=hand_r, loc=<-0.0002 -2.3007 -0.0001>, rot=<0 0 0 -1>}
 deform_wrist_up_r {parent=deform_wrist_up_base_r, loc=<0 0 0>, rot=<0 0 0 1>}
 lowerarm_twist_02_r {parent=lowerarm_r, loc=<-9.6199 0 0>, rot=<0 0 0 -1>}
 lowerarm_twist_01_r {parent=lowerarm_r, loc=<-17.8549 0 0>, rot=<0 0 0 -1>}
 deform_elbow_in_base_r {parent=lowerarm_r, loc=<-0.1385 -3.6052 0.2182>, rot=<0.0059 0.0117 -0.0071 -0.9999>}
 deform_elbow_in_r {parent=deform_elbow_in_base_r, loc=<0 0 0>, rot=<0 0 0 1>}
 deform_elbow_out_base_r {parent=lowerarm_r, loc=<-0.0565 3.9993 -0.126>, rot=<0.0059 0.0117 -0.0071 -0.9999>}
 deform_elbow_out_r {parent=deform_elbow_out_base_r, loc=<0 0 0>, rot=<0 0 0 1>}
 upperarm_twist_01_r {parent=upperarm_r, loc=<-0.0041 0 0>, rot=<0 0 0 -1>}
 deform_upperarm_bk_base_r {parent=upperarm_twist_01_r, loc=<-0.9473 4.46 -0.3113>, rot=<0.0004 -0.0034 -0.0158 -0.9999>}
 deform_upperarm_bk_r {parent=deform_upperarm_bk_base_r, loc=<0 0 0>, rot=<0 0 0 1>}
 deform_upperarm_fr_base_r {parent=upperarm_twist_01_r, loc=<-1.4228 -5.0263 -0.3212>, rot=<0.0004 -0.0034 -0.0158 -0.9999>}
 deform_upperarm_fr_r {parent=deform_upperarm_fr_base_r, loc=<0 0 0>, rot=<0 0 0 1>}
 deform_upperarm_up_base_r {parent=upperarm_twist_01_r, loc=<1.6653 -0.3195 -3.9628>, rot=<0.0004 -0.0034 -0.0158 -0.9999>}
 deform_upperarm_up_r {parent=deform_upperarm_up_base_r, loc=<0 0 0>, rot=<0 0 0 1>}
 upperarm_twist_02_r {parent=upperarm_r, loc=<-13.7941 0 0>, rot=<0 0 0 -1>}
 dfrm_necklace {parent=spine_05, loc=<6.3465 6.4171 -0.0012>, rot=<-0.5911 -0.3881 -0.3881 -0.5911>}
 deform_pec_aim_base_l {parent=spine_04, loc=<3.1942 9.3086 -1.2237>, rot=<0.0614 -0.4031 0.3295 -0.8516>}
 deform_pec_aim_l {parent=deform_pec_aim_base_l, loc=<0 0 0>, rot=<0 0 0 1>}
 deform_pec_base_l {parent=deform_pec_aim_l, loc=<9.3665 0 0>, rot=<0 0 0 1>}
 deform_pec_l {parent=deform_pec_base_l, loc=<0 0 0>, rot=<0 0 0 1>}
 deform_pec_aim_base_r {parent=spine_04, loc=<3.1938 9.3085 1.2237>, rot=<-0.4031 -0.0614 0.8516 0.3295>}
 deform_pec_aim_r {parent=deform_pec_aim_base_r, loc=<0 0 0>, rot=<0 0 0 1>}
 deform_pec_base_r {parent=deform_pec_aim_r, loc=<-9.3667 -0.0003 -0.0004>, rot=<0 0 0 1>}
 deform_pec_r {parent=deform_pec_base_r, loc=<0 0 0>, rot=<0 0 0 1>}
 dyn_zipper_02_mstr {parent=spine_04, loc=<2.0765 11.3012 -0.0134>, rot=<0 0 0.9887 0.1501>}
 dyn_zipper_02 {parent=dyn_zipper_02_mstr, loc=<0 0 0>, rot=<0 0 0 -1>}
 thigh_l {parent=pelvis, loc=<-1.0855 -0.2109 -8.7053>, rot=<0.0009 0.0345 0.0155 -0.9993>}
 calf_l {parent=thigh_l, loc=<-42.5268 0 0>, rot=<0 0 -0.0657 -0.9978>}
 foot_l {parent=calf_l, loc=<-46.8037 0 0>, rot=<0.0422 -0.04 0.0613 -0.9964>}
 ball_l {parent=foot_l, loc=<-3.5658 10.8066 -0.1544>, rot=<0.0001 -0.0001 -0.7063 -0.7079>}
 dyn_lower_pants_l {parent=foot_l, loc=<0 0 0>, rot=<0 0 0 1>}
 calf_twist_02_l {parent=calf_l, loc=<-18.8064 -0.0967 0.0397>, rot=<0.0237 0.0095 -0.0035 -0.9997>}
 calf_twist_01_l {parent=calf_l, loc=<-35.813 -0.0967 0.0397>, rot=<0.0237 0.0095 -0.0035 -0.9997>}
 deform_knee_in_base_l {parent=calf_l, loc=<0.5405 -5.0592 -0.5944>, rot=<-0.0003 -0.0151 0.0267 -0.9995>}
 deform_knee_in_l {parent=deform_knee_in_base_l, loc=<0 0 0>, rot=<0 0 0 1>}
 deform_knee_out_base_l {parent=calf_l, loc=<0.2138 3.9943 -0.8167>, rot=<-0.0003 -0.0151 0.0267 -0.9995>}
 deform_knee_out_l {parent=deform_knee_out_base_l, loc=<0 0 0>, rot=<0 0 0 1>}
 thigh_twist_01_l {parent=thigh_l, loc=<-21.629 -0.0919 -0.0002>, rot=<0.0013 0 0 -1>}
 deform_glute_base_l {parent=thigh_l, loc=<-6.0004 -8.9997 -0.0128>, rot=<0.0004 0.0009 -0.4617 -0.887>}
 deform_glute_l {parent=deform_glute_base_l, loc=<-0.0006 -0.0014 0>, rot=<0 0 0 1>}
 deform_groin_base_l {parent=thigh_l, loc=<0.4808 5.6776 -0.2472>, rot=<0 0.001 0 -1>}
 deform_groin_l {parent=deform_groin_base_l, loc=<-0.0058 0.0233 0>, rot=<0 0 0 1>, scale=<0.999 1 1>}
 thigh_r {parent=pelvis, loc=<-1.0855 -0.2109 8.7053>, rot=<0.0345 -0.0009 0.9993 0.0155>}
 calf_r {parent=thigh_r, loc=<42.5268 0 0>, rot=<0 0 -0.0657 -0.9978>}
 foot_r {parent=calf_r, loc=<46.8037 0 0>, rot=<0.0422 -0.04 0.0613 -0.9964>}
 ball_r {parent=foot_r, loc=<3.5658 -10.8066 0.1544>, rot=<0.0001 -0.0001 -0.7063 -0.7079>}
 dyn_lower_pants_r {parent=foot_r, loc=<0 0 0>, rot=<0 0 0 1>}
 calf_twist_02_r {parent=calf_r, loc=<18.8064 0.0967 -0.0397>, rot=<0.0237 0.0095 -0.0035 -0.9997>}
 calf_twist_01_r {parent=calf_r, loc=<35.813 0.0967 -0.0397>, rot=<0.0237 0.0095 -0.0035 -0.9997>}
 deform_knee_in_base_r {parent=calf_r, loc=<-0.5405 5.0592 0.5943>, rot=<-0.0003 -0.0151 0.0267 -0.9995>}
 deform_knee_in_r {parent=deform_knee_in_base_r, loc=<0 0 0>, rot=<0 0 0 1>}
 deform_knee_out_base_r {parent=calf_r, loc=<-0.2137 -3.9943 0.8168>, rot=<-0.0003 -0.0151 0.0267 -0.9995>}
 deform_knee_out_r {parent=deform_knee_out_base_r, loc=<0 0 0>, rot=<0 0 0 1>}
 thigh_twist_01_r {parent=thigh_r, loc=<21.629 0.0919 0.0002>, rot=<0.0013 0 0 -1>}
 deform_glute_base_r {parent=thigh_r, loc=<6.0002 8.9997 0.0129>, rot=<0.0004 0.0009 -0.4617 -0.887>}
 deform_glute_r {parent=deform_glute_base_r, loc=<0.0045 0.0114 0>, rot=<0 0 0 1>}
 deform_groin_base_r {parent=thigh_r, loc=<-0.4808 -5.6776 0.2472>, rot=<0 0.001 0 -1>}
 deform_groin_r {parent=deform_groin_base_r, loc=<0 0 0>, rot=<0 0 0 1>}
 dfrm_belt_buckle {parent=pelvis, loc=<2.6345 8.2712 0.0004>, rot=<0 0 -0.6943 -0.7197>}
]
variables [
 is_playing_character_idle {type=bool, default=False}
 looking_down {type=double, default=0}
]
